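# Additive, unlit, two-sided Niagara particle material for the FareSpore character's ear glow.
# Two opposite-panning samples of the backpack's audio-reactivity gradient, mapped by local position, blend along local X and gain slight saturation via negative desaturation.
# A scalar blends white toward these colors, then particle color and a radial gradient shape the emissive.
# Opacity combines particle alpha, that gradient, view-angle fade and switchable depth fade, which Low quality skips.
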Material /BRCosmetics/Characters/Player/Male/Medium/Bodies/M_MED_FareSpore/FX/Material/M_FX_FareSpore_EarGlow
graph {
 properties {
  bUsedWithNiagaraMeshParticles=true
  bUsedWithNiagaraSprites=true
  blendMode=BLEND_Additive
  pixelDepthOffsetMode=PDOM_Legacy
  shadingModel=MSM_Unlit
  twoSided=true
  usedShadingModels=Unlit
 }
 nodes [
  n1 = ComponentMask(None=Divide(A=LocalPosition().XYZ) {constB=100}) {b=true, r=true}
  n2 = Constant2Vector() {r=0.2}
  n3 = TextureSample(UVs=Panner(Coordinate=@n1, Speed=@n2) {speedX=0.5}) {
   samplerSource=SSM_Wrap_WorldGroupSettings
   texture=/BRCosmetics/Accessories/FORT_Backpacks/Backpack_M_MED_FareSpore/Textures/T_FareSpore_Backpack_AudioReactivityGradient
  }
  n4 = ScalarParameter() {defaultValue=3, parameterName=Pink/Blue_ScrollingSpeed}
  n5 = Time()
  n6 = TextureSample(UVs=Panner(Coordinate=@n1, Time=Add(A=@n4, B=@n5), Speed=Multiply(A=@n2) {constB=-1}) {speedX=-0.5}) {
   samplerSource=SSM_Wrap_WorldGroupSettings
   texture=&T_FareSpore_Backpack_AudioReactivityGradient
  }
  n7 = ComponentMask(None=LocalPosition().XYZ) {r=true}
  n8 = Saturate(None=Add(A=Multiply(A=@n7) {constB=0.1}) {constB=0.4})
  n9 = NamedRerouteDeclaration(None=Desaturation(None=LinearInterpolate(A=@n3.RGB, B=@n6.RGB, Alpha=@n8), Fraction=Constant() {r=-0.1})) {name="Pink&&BlueColors", nodeColor=<rgba 0.4531 0 1 1>}
  n10 = ScalarParameter() {parameterName=UseFareSporeColors?}
  n11 = ParticleColor()
  n12 = Multiply(A=MaterialFunctionCall(UVs=TextureCoordinate(), Density=Constant() {r=1}) {materialFunction=&RadialGradientExponential}.RadialGradientExponential)
  MaterialFunctionCall(Emissive=Multiply(A=LinearInterpolate(B=@n9, Alpha=@n10) {constA=1}, B=Multiply(A=@n11.RGB, B=@n12))) {materialFunction=&MF_FX_EmissiveAdjustments}
  n13 = Multiply(A=@n11.A, B=@n12)
  n14 = StaticSwitchParameter(True=DepthFade(Opacity=@n13, FadeDistance=ScalarParameter() {defaultValue=20, parameterName=Depth Fade}), False=@n13) {parameterName=Use_DepthFade}
  Multiply(
   A=QualitySwitch(Default=@n14, Low=@n13, High=@n14, Medium=@n14, Epic=@n14)
   B=MaterialFunctionCall("FadeSensitivity (optional)"=ScalarParameter() {defaultValue=1, parameterName=AngleFade}) {materialFunction=&FadeBasedOnViewAngle}.Result
  )
  NamedRerouteUsage()
  NamedRerouteUsage()
 ]
}
