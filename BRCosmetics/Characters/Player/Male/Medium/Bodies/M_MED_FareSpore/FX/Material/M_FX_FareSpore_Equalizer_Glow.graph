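# Additive unlit, two-sided Niagara mesh-particle material for the FareSpore body cosmetic, drawing an audio-equalizer glow.
# A dynamic-parameter spectrum sample, modulated along U by a curve-atlas distortion borrowed from HeavyRoar and offset by a sine wave, sets the bar height scaled by Open/Close.
# A smoothstep on V forms the Shape mask.
# Color comes from a time-oscillating curve-atlas gradient times particle color, with edge fades and particle alpha shaping opacity.
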Material /BRCosmetics/Characters/Player/Male/Medium/Bodies/M_MED_FareSpore/FX/Material/M_FX_FareSpore_Equalizer_Glow
graph {
 properties {
  bUsedWithNiagaraMeshParticles=true
  blendMode=BLEND_Additive
  pixelDepthOffsetMode=PDOM_Legacy
  shadingModel=MSM_Unlit
  twoSided=true
  usedShadingModels=Unlit
 }
 nodes [
  n1 = MaterialFunctionCall(Float2=TextureCoordinate()) {materialFunction=&BreakOutFloat2Components}
  n2 = ScalarParameter() {defaultValue=360, parameterName=ColorStep}
  Divide(A=Round(None=Multiply(A=@n1.R, B=@n2) {constB=5}), B=@n2) {constB=5}
  MaterialFunctionCall(Emissive=NamedRerouteUsage()) {materialFunction=&MF_FX_EmissiveAdjustments}
  n3 = Constant()
  n4 = MaterialFunctionCall(Float2=TextureCoordinate()) {materialFunction=&BreakOutFloat2Components}
  n5 = SmoothStep(Min=@n3, Max=Add(A=@n3, B=Constant() {r=0.5}), Value=@n4.G)
  n6 = Constant()
  n7 = SmoothStep(Min=@n6, Max=Add(A=@n6, B=Constant() {r=0.1}), Value=OneMinus(None=@n4.G))
  Multiply(A=Multiply(A=Multiply(A=@n5, B=@n7), B=NamedRerouteUsage()), B=ParticleColor().A)
  n8 = Multiply(A=Time(), B=ScalarParameter() {defaultValue=0.2, parameterName=Time Mult})
  n9 = Add(A=Multiply(A=Sine(None=@n8)) {constB=0.5}) {constB=0.5}
  n10 = CurveAtlasRowParameter(CurveTime=Frac(None=Add(A=@n1.R, B=@n9) {constB=0})) {
   atlas=/BRCosmetics/Characters/Player/Male/Medium/Bodies/M_MED_FareSpore/FX/Misc/CA_FareSpore
   curve=/BRCosmetics/Characters/Player/Male/Medium/Bodies/M_MED_FareSpore/FX/Misc/CC_FareSpore
   parameterName=ColorCurve
  }
  NamedRerouteDeclaration(None=Multiply(A=ParticleColor().RGB, B=@n10)) {name=Color, nodeColor=<rgba 1 0.8906 0 1>}
  n11 = ScalarParameter() {defaultValue=1, parameterName=Wave_01 Smothness}
  n12 = DynamicParameter() {paramNames [SpectrumSample_x, SpectrumSample_y, SpectrumSample_z, SpectrumSample_a]}
  n13 = ScalarParameter() {defaultValue=2, parameterName=Equaliser Noise Tile}
  n14 = MaterialFunctionCall(Float2=Multiply(
    A=TextureCoordinate()
    B=MaterialFunctionCall(X=@n13) {materialFunction=&MakeFloat2}.Result
   )) {materialFunction=&BreakOutFloat2Components}
  n15 = ScalarParameter() {parameterName=Equaliser Noise Offset}
  n16 = CurveAtlasRowParameter(CurveTime=Frac(None=Add(A=@n14.R, B=@n15))) {
   atlas=/BRCosmetics/Accessories/FORT_Backpacks/Backpack_M_MED_HeavyRoar/FX/CA_Backpack_HeavyRoar
   curve=/BRCosmetics/Accessories/FORT_Backpacks/Backpack_M_MED_HeavyRoar/FX/CC_Backpack_HeavyRoar_MusicDistort
   defaultValue=1
   parameterName=ColorDistortion
  }
  n17 = MaterialFunctionCall(Float4=Multiply(
    A=MaterialFunctionCall(X=@n12.SpectrumSample_x, Y=@n12.SpectrumSample_x, Z=@n12.SpectrumSample_x, A=@n12.SpectrumSample_x) {materialFunction=&MakeFloat4}.Result
    B=MaterialFunctionCall(X=@n16.R, Y=@n16.R, Z=@n16.R, A=@n16.R) {materialFunction=&MakeFloat4}.Result
   )) {materialFunction=&BreakOutFloat4Components}
  n18 = Max(A=Max(A=@n17.R, B=@n17.R), B=@n17.B)
  MaterialExpressionRemap_0 = MaterialXRemap!(
   None=NamedRerouteDeclaration(None=Max(A=@n18, B=@n17.A)) {name=SpectrumSample, nodeColor=<rgba 0.8281 0 1 1>}
   InputLow=ScalarParameter() {parameterName=InLow}
   InputHigh=ScalarParameter() {defaultValue=1, parameterName=InHigh}
   TargetLow=ScalarParameter() {parameterName=Param_10}
   TargetHigh=ScalarParameter() {defaultValue=1, parameterName=Param_11}
  )
  n19 = TextureCoordinate()
  n20 = MaterialFunctionCall(Multiplier=Constant() {r=2}) {materialFunction=&Pi}
  n21 = ScalarParameter() {defaultValue=4, parameterName=Sin_01_Period}
  n22 = Add(
   A=Multiply(
    A=MaterialFunctionCall(Float2=@n19) {materialFunction=&BreakOutFloat2Components}.R
    B=ScalarParameter() {defaultValue=-3, parameterName=Sin_01_Freq}
   )
   B=Multiply(A=Time(), B=Divide(A=@n20.Result, B=@n21))
  )
  n23 = Add(
   A=Multiply(A=Sine(None=@n22), B=ScalarParameter() {defaultValue=1, parameterName=Sin_01_Scale})
   B=ScalarParameter() {parameterName=Sin_01_Bias}
  )
  n24 = Multiply(
   A=Power(Base=ConstantBiasScale(None=@n23), Exp=ScalarParameter() {defaultValue=1})
   B=ScalarParameter() {defaultValue=-0.4, parameterName=Wave Distortion Amount}
  ) {constB=0.1}
  n25 = Multiply(
   A=Saturate(None=Add(A=@MaterialExpressionRemap_0, B=@n24) {constA=1, constB=0})
   B=Multiply(
    A=ScalarParameter() {defaultValue=1, parameterName=WavesAmount}
    B=DynamicParameter() {parameterIndex=1, paramNames [Open/Close, Wave Offset, "", ""]}.Open/Close
   )
  )
  n26 = ScalarParameter() {defaultValue=0.8, parameterName=Param_1}
  n27 = DynamicParameter() {parameterIndex=1, paramNames [Open/Close, Wave Offset, "", ""]}
  n28 = Add(
   A=Multiply(A=OneMinus(None=@n25), B=LinearInterpolate(B=@n26, Alpha=@n27.Open/Close) {constA=1})
   B=ScalarParameter() {defaultValue=-0.5, parameterName=Param_14}
  )
  n29 = TextureCoordinate()
  NamedRerouteDeclaration(None=SmoothStep(
    Min=@n28
    Max=Add(A=@n28, B=@n11)
    Value=MaterialFunctionCall(Float2=@n29) {materialFunction=&BreakOutFloat2Components}.G
   )) {name=Shape, nodeColor=<rgba 1 0 0.7031 1>}
 ]
}
