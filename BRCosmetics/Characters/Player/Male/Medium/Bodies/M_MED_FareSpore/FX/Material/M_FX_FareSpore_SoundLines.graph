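# Additive unlit Niagara sprite material for FareSpore body "SoundLines" FX, drawing offset, color-split edges from depth.
# Screen UVs scale around a dynamic-parameter point and shift both ways by camera-distance-scaled LineDir to sample custom or scene depth.
# Reconstructed world positions feed sphere masks, minus the unshifted mask, leaving fringes tinted separately.
# Panning Perlin noise hue-shifts the fringes before particle color and emissive adjustment.
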Material /BRCosmetics/Characters/Player/Male/Medium/Bodies/M_MED_FareSpore/FX/Material/M_FX_FareSpore_SoundLines
graph {
 properties {bUsedWithNiagaraSprites=true, blendMode=BLEND_Additive, pixelDepthOffsetMode=PDOM_Legacy, shadingModel=MSM_Unlit, usedShadingModels=Unlit}
 nodes [
  n1 = TextureObjectParameter() {
   group=HueShift
   parameterName=HueShift_Texture
   samplerType=SAMPLERTYPE_LinearGrayscale
   sortPriority=0
   texture=&T_SimplePerlin_Greyscale
  }
  n2 = VectorParameter() {group=HueShift, parameterName=HueShift_UVControl, channelNames {a=VPan, b=UPan, g=VScale, r=UScale}, defaultValue=<rgba 1 1 0 2>}
  n3 = Multiply(
   A=Add(A=MaterialFunctionCall(Texture=@n1, UVControl_RGB=@n2, UVControl_A=@n2) {materialFunction=&FXCommon_TextureSample}.R) {constB=-0.5}
   B=ScalarParameter() {parameterName=HS}
  )
  n4 = ScreenPosition()
  n5 = DynamicParameter() {parameterIndex=1, defaultValue=<rgba 0 0 0 1>, paramNames [SUVX,Y,Z,A]}
  n6 = MaterialFunctionCall(World Position=AppendVector(A=AppendVector(A=@n5.SUVX, B=@n5.SUVX), B=@n5.Z)) {materialFunction=&TransformToClipSpace}
  n7 = DynamicParameter() {defaultValue=<rgba 0 0 0 1>, paramNames [X,Y,Z,Scale2]}
  n8 = Add(A=Multiply(A=Subtract(A=@n4.ViewportUV, B=@n6.Buffer UVs for Scene Textures), B=@n7.Scale2), B=@n6.Buffer UVs for Scene Textures)
  n9 = VectorParameter() {parameterName=LineDir, defaultValue=<rgba 0.25 0 0 1>}
  n10 = Distance(A=CameraPositionWS(), B=ParticlePositionWS())
  n11 = Multiply(A=AppendVector(A=Divide(A=@n9.R, B=@n10), B=@n9.G), B=@n5.A)
  n12 = Add(A=@n8, B=@n11)
  n13 = SceneDepth(UVs=@n12)
  n14 = ComponentMask(None=SceneTexture(UVs=@n12) {sceneTextureId=PPI_CustomDepth}.Color) {r=true}
  n15 = StaticSwitchParameter(True=@n13.R, False=ShadingPathSwitch(Default=@n14, Deferred=@n14, Forward=@n13.R, Mobile=@n13.R)) {parameterName=UseSceneDepth}
  n16 = AppendVector(A=@n7.X, B=@n7.X)
  n17 = VectorParameter() {parameterName=MaskSize, defaultValue=<rgba 30 0 3 0>}
  n18 = Add(A=@n17.R, B=DynamicParameter() {parameterIndex=2, defaultValue=<rgba 0 0 0 1>, paramNames [AA,0,0,0]}.AA)
  n19 = Power(
   Base=SphereMask(
    A=MaterialFunctionCall(SceneDepth=@n15) {materialFunction=&WorldPositionBehindTranslucency}.Result
    B=AppendVector(A=@n16, B=@n7.Z)
    Radius=@n18
    Hardness=@n17.G
   ) {attenuationRadius=10, hardnessPercent=0}
   Exp=ScalarParameter() {defaultValue=2, parameterName=Power}
  )
  n20 = SceneDepth(UVs=@n8)
  n21 = ComponentMask(None=SceneTexture(UVs=@n8) {sceneTextureId=PPI_CustomDepth}.Color) {r=true}
  n22 = ShadingPathSwitch(Default=@n21, Deferred=@n21, Forward=@n20.R, Mobile=@n20.R)
  n23 = DynamicParameter() {defaultValue=<rgba 0 0 0 1>, paramNames [X,Y,Z,Scale2]}
  n24 = SphereMask(
   A=MaterialFunctionCall(SceneDepth=StaticSwitchParameter(True=@n20.R, False=@n22) {parameterName=UseSceneDepth}) {materialFunction=&WorldPositionBehindTranslucency}.Result
   B=AppendVector(A=AppendVector(A=@n23.X, B=@n23.X), B=@n23.Z)
   Radius=Multiply(A=@n18) {constB=1.5}
   Hardness=@n17.G
  ) {attenuationRadius=10, hardnessPercent=0}
  n25 = ScalarParameter() {defaultValue=2, parameterName=Power}
  n26 = Saturate(None=Ceil(None=Power(Base=@n24, Exp=@n25)))
  n27 = Subtract(A=Saturate(None=Multiply(A=@n19, B=@n17.B)), B=@n26)
  n28 = Add(A=@n8, B=Multiply(A=@n11) {constB=-1})
  n29 = SceneDepth(UVs=@n28)
  n30 = ComponentMask(None=SceneTexture(UVs=@n28) {sceneTextureId=PPI_CustomDepth}.Color) {r=true}
  n31 = StaticSwitchParameter(True=@n29.R, False=ShadingPathSwitch(Default=@n30, Deferred=@n30, Forward=@n29.R, Mobile=@n29.R)) {parameterName=UseSceneDepth}
  n32 = DynamicParameter() {defaultValue=<rgba 0 0 0 1>, paramNames [X,Y,Z,Scale2]}
  n33 = AppendVector(A=@n32.X, B=@n32.X)
  n34 = Power(
   Base=SphereMask(
    A=MaterialFunctionCall(SceneDepth=@n31) {materialFunction=&WorldPositionBehindTranslucency}.Result
    B=AppendVector(A=@n33, B=@n32.Z)
    Radius=@n18
    Hardness=@n17.G
   ) {attenuationRadius=10, hardnessPercent=0}
   Exp=ScalarParameter() {defaultValue=2, parameterName=Power}
  )
  n35 = Subtract(A=Saturate(None=Multiply(A=@n34, B=@n17.B)), B=@n26)
  n36 = Add(
   A=Multiply(A=@n27, B=VectorParameter() {parameterName=Param2, defaultValue=<rgba 3 0 0 1>}.RGB)
   B=Multiply(A=VectorParameter() {defaultValue=<rgba 0 3 0 1>}.RGB, B=@n35)
  )
  n37 = ParticleColor()
  n38 = Multiply(
   A=LinearInterpolate(
    A=MaterialFunctionCall(Hue Shift Percentage=@n3, Texture=@n36) {materialFunction=&HueShift}.Result
    Alpha=Constant() {r=0.75}
   )
   B=@n37.RGB
  )
  Add(A=Divide(A=@n38) {constB=1.977}) {constB=0}
  Constant() {r=1}
  MaterialFunctionCall(Emissive=@n38) {materialFunction=&MF_FX_EmissiveAdjustments}
  Multiply(A=@n9.R, B=@n10)
  Multiply(A=@n37.A, B=Add(A=Saturate(None=@n27), B=Saturate(None=@n35)))
  ParticleColor()
 ]
}
