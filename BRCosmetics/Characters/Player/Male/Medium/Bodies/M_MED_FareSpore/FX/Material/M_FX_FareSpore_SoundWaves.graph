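# Translucent unlit Niagara sprite material for FareSpore sound wave FX, painting hue-shifted particle color and comic dots over scene surfaces.
# Screen UVs, rescaled around a particle-supplied point, sample world normals for an animated hue shift and a camera-facing rim band.
# Sphere masks on depth-reconstructed world position, using custom or scene depth, confine the effect.
# Forward and mobile paths substitute Perlin-noise hue, boosted color and an edge mask for the normal-based terms.
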Material /BRCosmetics/Characters/Player/Male/Medium/Bodies/M_MED_FareSpore/FX/Material/M_FX_FareSpore_SoundWaves
graph {
 properties {bUsedWithNiagaraSprites=true, blendMode=BLEND_Translucent, pixelDepthOffsetMode=PDOM_Legacy, shadingModel=MSM_Unlit, usedShadingModels=Unlit}
 nodes [
  Add(B=NamedRerouteUsage())
  Constant() {r=1}
  n1 = VectorParameter() {defaultValue=<rgba 0.1 1 1 1>}
  n2 = VectorParameter() {parameterName=Angle, defaultValue=<rgba 0 1 0.2 1>}
  n3 = Sine(None=Multiply(A=@n2.B, B=Time() {bOverride_Period=true, period=6666}))
  n4 = Normalize(VectorInput=LinearInterpolate(A=Constant3Vector() {constant=<rgba 0.32 1.0721 0.67 0>}, B=Constant3Vector() {constant=<rgba -0.752 0.0614 0.246 0>}, Alpha=@n3))
  n5 = LinearInterpolate(A=@n2.R, B=@n2.R, Alpha=@n3)
  n6 = Constant3Vector()
  n7 = CameraVectorWS()
  n8 = Normalize(VectorInput=Add(A=RotateAboutAxis(NormalizedRotationAxis=@n4, RotationAngle=@n5, PivotPoint=@n6, Position=@n7), B=@n7))
  n9 = ScreenPosition()
  n10 = DynamicParameter() {parameterIndex=1, defaultValue=<rgba 0 0 0 1>, paramNames [X,Y,Z,A]}
  n11 = AppendVector(A=AppendVector(A=@n10.X, B=@n10.X), B=@n10.Z)
  n12 = MaterialFunctionCall(World Position=@n11) {materialFunction=&TransformToClipSpace}
  n13 = Add(A=Multiply(A=Subtract(A=@n9.ViewportUV, B=@n12.Buffer UVs for Scene Textures), B=@n10.A), B=@n12.Buffer UVs for Scene Textures)
  n14 = SceneTexture(UVs=@n13) {bFiltered=true, sceneTextureId=PPI_WorldNormal}
  n15 = Constant3Vector() {constant=<rgba 0 0 1 0>}
  n16 = ShadingPathSwitch(Default=@n14.Color, Deferred=@n14.Color, Forward=@n15, Mobile=@n15)
  n17 = Multiply(A=@n1.R, B=OneMinus(None=DotProduct(A=@n8, B=@n16)))
  n18 = TextureObjectParameter() {
   group=Hue
   parameterName=Hue_Texture
   samplerType=SAMPLERTYPE_LinearGrayscale
   sortPriority=0
   texture=&T_SimplePerlin_Greyscale
  }
  n19 = VectorParameter() {group=Hue, parameterName=Hue_UVControl, channelNames {a=VPan, b=UPan, g=VScale, r=UScale}, defaultValue=<rgba 0.7 0.7 0.7 0.7>}
  n20 = ComponentMask(None=TextureCoordinate()) {g=true, r=true}
  n21 = Multiply(
   A=Add(
    A=MaterialFunctionCall(Texture=@n18, UVControl_RGB=@n19, UVControl_A=@n19, "[Override]UVInput"=@n20) {materialFunction=&FXCommon_TextureSample}.R
    B=ScalarParameter() {group=Hue, parameterName=Hue_GrayAdjust_Add}
   )
   B=@n1.R
  )
  n22 = ParticleColor()
  n23 = MaterialFunctionCall(Hue Shift Percentage=ShadingPathSwitch(Default=@n17, Deferred=@n17, Forward=@n21, Mobile=@n21), Texture=@n22.RGB) {materialFunction=&HueShift}
  n24 = Multiply(A=@n23.Result) {constB=10}
  n25 = ShadingPathSwitch(Default=@n23.Result, Deferred=@n23.Result, Forward=@n24, Mobile=@n24)
  Divide(A=@n25) {constB=1.977}
  n26 = ScalarParameter() {defaultValue=0.4}
  n27 = DynamicParameter() {parameterIndex=2, defaultValue=<rgba 1 0 0 1>, paramNames [UVScale,Param2,Param3,Param4]}
  n28 = Rotator(Coordinate=TextureCoordinate(), Time=@n27.UVScale)
  n29 = ScalarParameter() {defaultValue=9, parameterName=DotsUVScale}
  n30 = SmoothStep(
   Min=@n26
   Max=Add(A=@n26, B=ScalarParameter() {defaultValue=0.2, parameterName=Param_2})
   Value=TextureSample(UVs=Multiply(A=@n28, B=@n29) {constB=4}) {samplerType=SAMPLERTYPE_LinearColor, texture=/Game/Weapons/WeaponSkins/Wraps/Textures/Comic/T_Wrap_Patterns_Dots}.R
  )
  n31 = ScalarParameter() {defaultValue=0.5, parameterName=DotsIntensity}
  MaterialFunctionCall(Emissive=LinearInterpolate(A=@n25, B=VectorParameter() {parameterName=DotsColor, defaultValue=<rgba 0.1 1.05 2 1>}.RGB, Alpha=Multiply(A=@n30, B=@n31))) {materialFunction=&MF_FX_EmissiveAdjustments}
  Multiply(A=@n29, B=@n27.UVScale)
  NamedRerouteDeclaration(None=@n25) {name=test, nodeColor=<rgba 1 0 0.4922 1>}
  NamedRerouteDeclaration(None=Multiply(A=Multiply(A=@n30), B=ScalarParameter() {defaultValue=2, parameterName=Param_3})) {name=Dots, nodeColor=<rgba 0.7812 1 0 1>}
  NamedRerouteUsage()
  n32 = OneMinus(None=DotProduct(A=@n7, B=@n16))
  n33 = ScalarParameter() {defaultValue=2, parameterName=Exp}
  n34 = Power(Base=@n32, Exp=ScalarParameter() {defaultValue=4, parameterName=Exp2})
  n35 = Multiply(A=Saturate(None=Power(Base=@n32, Exp=@n33)), B=OneMinus(None=Saturate(None=@n34)))
  n36 = SceneDepth(UVs=@n13)
  n37 = ComponentMask(None=SceneTexture(UVs=@n13) {sceneTextureId=PPI_CustomDepth}.Color) {r=true}
  n38 = MaterialFunctionCall(SceneDepth=StaticSwitchParameter(True=@n36.R, False=ShadingPathSwitch(Default=@n37, Deferred=@n37, Forward=@n36.R, Mobile=@n36.R)) {parameterName=UseSceneDepth}) {materialFunction=&WorldPositionBehindTranslucency}
  n39 = VectorParameter() {parameterName=MaskSize, defaultValue=<rgba 20 0 0 0>}
  n40 = DynamicParameter() {defaultValue=<rgba 0 0 0 1>, paramNames [X,Y,Z,Param4]}
  n41 = Add(A=@n39.R, B=@n40.Param4)
  n42 = SphereMask(A=@n38.Result, B=AppendVector(A=AppendVector(A=@n40.X, B=@n40.X), B=@n40.Z), Radius=@n41, Hardness=@n39.G) {attenuationRadius=10, hardnessPercent=0}
  n43 = VectorParameter() {
   group=LowPlatMask
   parameterName=LowPlatMask_EdgeMask_InvertControl
   channelNames {a="", b=Power, g=Multiply, r=Offset}
   defaultValue=<rgba 0 1 3.3 1>
  }
  n44 = VectorParameter() {
   group=LowPlatMask
   parameterName=LowPlatMask_EdgeMask_Control
   channelNames {a="", b=Power, g=Multiply, r=Offset}
   defaultValue=<rgba 0 1 2.5 1>
  }
  n45 = MaterialFunctionCall!(
   UorV=Max(A=SphereMask(A=@n38.Result, B=@n11, Radius=@n41, Hardness=@n39.G) {attenuationRadius=10, hardnessPercent=0}, B=@n42)
   Invert_EdgeControl=@n43
   NoInvert_EdgeControl=@n44
  ) {materialFunction=/Game/Effects/Niagara/MasterMaterials/Functions/Common/FXCommon_EdgeMask_Ex}
  n46 = Multiply(A=Power(Base=@n42)) {constB=0.1}
  Reroute(None=Multiply(A=ShadingPathSwitch(Default=@n35, Deferred=@n35, Forward=@n45.Mask, Mobile=@n45.Mask), B=Multiply(A=@n22.A, B=@n46)))
  SceneDepth(UVs=@n13)
  VectorParameter() {parameterName=RERER, defaultValue=<rgba 1.3 2 0.1 0.005>}
 ]
}
