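# Niagara system for the FareSpore character body cosmetic, drawing pulsing sound-wave and line sprites on the arms and legs.
# System.AudioReactivity, apparently from a fake audio loop module, gates looping burst spawns and modulates alpha, brightness or size.
# Partitioned particles follow indexed skeletal mesh bones or sockets, passing positions and rotated offsets to materials through dynamic parameters.
# User flags hide hand or foot effects and switch front-end preview tuning.
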NiagaraSystem /BRCosmetics/Characters/Player/Male/Medium/Bodies/M_MED_FareSpore/FX/NS_Character_FareSpore_Body
dependencies {
 usedDataInterfaces [
  &NiagaraDataInterfaceArrayFloat
  &NiagaraDataInterfaceArrayInt32
  &NiagaraDataInterfaceCurve
  &NiagaraDataInterfacePlatformSet
  &NiagaraDataInterfaceSkeletalMesh
  &NiagaraDataInterfaceColorCurve
 ]
 usedDynamicInputs [
  /Niagara/DynamicInputs/Lerp/SmoothLerpOverTime_Float
  &MakeFloatFromLinearColor
  &RemapRange
  &ColorFromCurve
  &MakeCustomLinearColorFromBool
  &LerpLinearColors
 ]
 usedModules [
  &self:SystemSpawnScript.SystemScriptSource.SystemScriptGraph.NiagaraNodeAssignment_9.SetVariables_#1
  &self:SystemSpawnScript.SystemScriptSource.SystemScriptGraph.NiagaraNodeAssignment_3.SetVariables_#2
  &self:SystemSpawnScript.SystemScriptSource.SystemScriptGraph.NiagaraNodeAssignment_2.SetVariables_#3
  &SystemState
  &Timeline
  /Game/Effects/Niagara/Modules/Audio/FN_FakeAudioReactivityLoop
  &self:SystemSpawnScript.SystemScriptSource.SystemScriptGraph.NiagaraNodeAssignment_8.SetVariables_#4
  &self:SystemSpawnScript.SystemScriptSource.SystemScriptGraph.NiagaraNodeAssignment_0.SetVariables_#5
  &self:SystemSpawnScript.SystemScriptSource.SystemScriptGraph.NiagaraNodeAssignment_1.SetVariables_#6
  &self:SystemSpawnScript.SystemScriptSource.SystemScriptGraph.NiagaraNodeAssignment_6.SetVariables_#7
  &self:SystemSpawnScript.SystemScriptSource.SystemScriptGraph.NiagaraNodeAssignment_7.SetVariables_#8
  &InterpolateOverTime
 ]
}
emitters {
 Arms_SoundWaves {
  emitterUpdateScript [
   &EmitterState(Inactive Response=::NewEnumerator1, Life Cycle Mode=::NewEnumerator1, Loop Duration=0.5, UseLoopDelay=false)
   &SpawnBurst_Instantaneous(
    Spawn Count=4
    Spawn Probability=&MakeCustomFloatFromBool(Bool=&SetBoolByFloatComparison(A=$System.AudioReactivity, B=0.1), True Float=1, False Float=0)
    Use Spawn Probability=true
    UseLoopCountLimit=false
   )
   &SpawnBurst_Instantaneous(
    Spawn Count=4
    Spawn Probability=&MakeCustomFloatFromBool(Bool=&SetBoolByFloatComparison(A=$System.AudioReactivity, B=0.1), True Float=1, False Float=0)
    Spawn Time=0.2
    Use Spawn Probability=true
    UseLoopCountLimit=false
   ) {module=SpawnBurst_Instantaneous001}
  ]
  particleSpawnScript [
   &PartitionParticles(Number Of Partitions=8, Particles Per Partition=1, Partition Mode:ENiagara_PartitionMode_v2=::NewEnumerator1)
   &InitializeParticle(
    Color=&MergeHelper_ColorSwitch(
     0=<rgba 0 0.6667 1 1>
     1=<rgba 0 0.6667 1 1>
     2=<rgba 0 0.6667 1 1>
     3=<rgba 0 0.6667 1 1>
     Index=$Particles.PartitionParticles.Partition
     Num=::NewEnumerator2
    )
    Lifetime=0.5
    Sprite Size Mode=::NewEnumerator3
    Uniform Sprite Size=&SelectFloatFromArray(
     Direct Array Index=$Particles.PartitionParticles.Partition
     Float Selection Array=DataInterface() {propertyValues {FloatData [40,40,40,40,40,40,40,40]}}
    )
    UsePositionOffset=false
    Write Material Random=true
   )
   /SetVariables(Particles.VisibilityTag=&MakeCustomIntFromBool(Bool=$User.bIsFrontEndPreview, True INT=1, False INT=0)) {isSetParameters=true, module=SetVariables_#9}
  ]
  particleUpdateScript [
   &ParticleState()
   &SkeletalMeshLocation(
    Bone / Socket Index=&SelectIntFromArray(
     Direct Array Index=$Particles.PartitionParticles.Partition
     Direct Array Mode=::NewEnumerator1
     Int Selection Array=DataInterface() {propertyValues {IntData [8,9,10,11]}}
    )
    MaskSpawnGroup=false
    Mesh Sampling Type=::NewEnumerator5
    Sampled Position Offset=<0 0 0>
    Skeletal Mesh=$System.Skeletal Mesh
    Socket Sampling Mode=::NewEnumerator5
   )
   &CameraOffset(Camera Offset Amount=&FN_FrontEndFloatModifier(Gameplay=20, Front End=0))
   /SetVariables(Particles.OffsetPosition:NiagaraPosition=&AddVectorToPosition(
     Position=$Particles.Position
     Vector=&RotateVectorByQuaternion(Vector To Rotate=<0 0 0>, Quaternion=$Particles.SkeletalMeshLocation.SampledRotation)
    )) {isSetParameters=true, module=SetVariables_#10}
   &DynamicMaterialParameters(
    Index 0 Param 1=&MakeFloatFromVector(Vector=Linked() {linkedVariable {name=Particles.Position, type=NiagaraPosition}}, Channel=::NewEnumerator0)
    Index 0 Param 2=&MakeFloatFromVector(Vector=Linked() {linkedVariable {name=Particles.Position, type=NiagaraPosition}}, Channel=::NewEnumerator1)
    Index 0 Param 3=&MakeFloatFromVector(Vector=Linked() {linkedVariable {name=Particles.Position, type=NiagaraPosition}}, Channel=::NewEnumerator2)
    Index 0 Param 4=&RandomRangeFloat(Minimum=0, Maximum=1)
    Index 1 Param 1=&MakeFloatFromVector(Vector=Linked() {linkedVariable {name=Particles.OffsetPosition, type=NiagaraPosition}}, Channel=::NewEnumerator0)
    Index 1 Param 2=&MakeFloatFromVector(Vector=Linked() {linkedVariable {name=Particles.OffsetPosition, type=NiagaraPosition}}, Channel=::NewEnumerator1)
    Index 1 Param 3=&MakeFloatFromVector(Vector=Linked() {linkedVariable {name=Particles.OffsetPosition, type=NiagaraPosition}}, Channel=::NewEnumerator2)
    Index 1 Param 4=&Lerp_Float(
     A=&FloatFromCurve(
      FloatCurve=DataInterface() {propertyValues {Curve {keys [
          {arriveTangent=-0.0762, interpMode=RCIM_Cubic, leaveTangent=-0.0762, tangentMode=RCTM_User, time=0, value=0.8}
          {arriveTangent=-0.2085, interpMode=RCIM_Cubic, leaveTangent=-0.2085, tangentMode=RCTM_User, time=1, value=0.3}
         ]}}}
      Scale Curve=1.35
     )
     B=&FloatFromCurve(
      FloatCurve=DataInterface() {propertyValues {Curve {keys [
          {arriveTangent=-0.8006, interpMode=RCIM_Cubic, leaveTangent=-0.8006, tangentMode=RCTM_User, time=0, value=0.8}
          {arriveTangent=-0.3233, interpMode=RCIM_Cubic, leaveTangent=-0.3233, tangentMode=RCTM_User, time=1, value=0.4}
         ]}}}
      Scale Curve=1.35
     )
     Alpha=1
    )
    Index 2 Param 1=1
    Module.Index0Float0Write=true
    Module.Index0Float1Write=true
    Module.Index0Float2Write=true
    Module.Index0Float3Write=true
    Module.Index1Float0Write=true
    Module.Index1Float1Write=true
    Module.Index1Float2Write=true
    Module.Index1Float3Write=true
    Module.Index2Float0Write=true
    Module.Index2Float1Write=false
    Module.Index2Float2Write=false
    Module.Index2Float3Write=false
    Write Parameter Index 1=::NewEnumerator1
    Write Parameter Index 2=::NewEnumerator1
   )
   &ScaleColor(
    Scale Alpha=&FloatFromCurve(
     FloatCurve=DataInterface() {propertyValues {Curve {keys [
         {arriveTangent=0.5632, interpMode=RCIM_Cubic, leaveTangent=0.5632, tangentMode=RCTM_User, time=0, value=0}
         {arriveTangent=0.168, interpMode=RCIM_Cubic, leaveTangent=0.168, tangentMode=RCTM_User, time=0.25, value=1}
         {arriveTangent=-0.1012, interpMode=RCIM_Cubic, leaveTangent=-0.1012, tangentMode=RCTM_User, time=1, value=0}
        ]}}}
     Scale Curve=&Multiply_Float(A=&FN_FrontEndFloatModifier(Gameplay=10, Front End=20), B=$System.Timeline001.Time)
    )
    Scale RGB=&VectorFromFloat(Value=&FloatFromCurve(FloatCurve=DataInterface() {propertyValues {Curve {keys [
          {arriveTangent=0.168, interpMode=RCIM_Cubic, leaveTangent=0.168, tangentMode=RCTM_User, time=0.2, value=1.5}
          {arriveTangent=-0.1012, interpMode=RCIM_Cubic, leaveTangent=-0.1012, tangentMode=RCTM_User, time=0.4, value=1}
         ]}}}))
    ScaleA=true
    ScaleRGB=true
   )
   &ScaleColor(
    Scale Alpha=$System.AudioReactivity
    Scale RGB=&VectorFromFloat(Value=&Add_Float(
      A=&FloatFromCurve(
       FloatCurve=DataInterface() {propertyValues {Curve {keys [
           {interpMode=RCIM_Cubic, tangentMode=RCTM_Break, time=0, value=0}
           {arriveTangent=3.7964, interpMode=RCIM_Cubic, leaveTangent=1.1533, tangentMode=RCTM_Break, time=0.9, value=1}
           <at 1 0>
          ]}}}
       CurveIndex=$System.AudioReactivity
      )
      B=1
     ))
    ScaleA=true
    ScaleRGB=true
   ) {module=ScaleColor001}
   &ScaleColor(
    Scale Alpha=&MakeCustomFloatFromBool(Bool=$User.bisHandHidden, True Float=0, False Float=1)
    ScaleA=true
    ScaleRGB=false
   ) {module=ScaleColor002}
  ]
  rendererClasses [&NiagaraSpriteRendererProperties]
  renderers [
   {rendererClass=&NiagaraSpriteRendererProperties}
   {rendererClass=&NiagaraSpriteRendererProperties, rendererIndex=1}
   {rendererClass=&NiagaraSpriteRendererProperties, rendererIndex=2}
  ]
 }
 Arms_SoundWaves001 {
  emitterUpdateScript [
   &EmitterState(Inactive Response=::NewEnumerator1, Life Cycle Mode=::NewEnumerator1, Loop Duration=0.5, UseLoopDelay=false)
   &SpawnBurst_Instantaneous(
    Spawn Count=2
    Spawn Probability=&MakeCustomFloatFromBool(Bool=&SetBoolByFloatComparison(A=$System.AudioReactivity, B=0.1), True Float=1, False Float=0)
    Use Spawn Probability=true
    UseLoopCountLimit=false
   )
   &SpawnBurst_Instantaneous(
    Spawn Count=2
    Spawn Probability=&MakeCustomFloatFromBool(Bool=&SetBoolByFloatComparison(A=$System.AudioReactivity, B=0.1), True Float=1, False Float=0)
    Spawn Time=0.2
    Use Spawn Probability=true
    UseLoopCountLimit=false
   ) {module=SpawnBurst_Instantaneous001}
  ]
  particleSpawnScript [
   &PartitionParticles(Particles Per Partition=1, Partition Mode:ENiagara_PartitionMode_v2=::NewEnumerator1)
   &InitializeParticle(
    Color=&MergeHelper_ColorSwitch(
     0=<rgba 0 0.6667 1 1>
     1=<rgba 0 0.6667 1 1>
     2=<rgba 0 0.6667 1 1>
     3=<rgba 0 0.6667 1 1>
     Index=$Particles.PartitionParticles.Partition
     Num=::NewEnumerator2
    )
    Lifetime=0.5
    Sprite Size Mode=::NewEnumerator3
    Uniform Sprite Size=&SelectFloatFromArray(
     Direct Array Index=$Particles.PartitionParticles.Partition
     Float Selection Array=DataInterface() {propertyValues {FloatData [40,40,40,40,40,40,40,40]}}
    )
    UsePositionOffset=false
    Write Material Random=true
   )
   /SetVariables(Particles.VisibilityTag=&MakeCustomIntFromBool(Bool=$User.bIsFrontEndPreview, True INT=1, False INT=0)) {isSetParameters=true, module=SetVariables_#9}
  ]
  particleUpdateScript [
   &ParticleState()
   &SkeletalMeshLocation(
    Bone / Socket Index=&SelectIntFromArray(
     Direct Array Index=$Particles.PartitionParticles.Partition
     Direct Array Mode=::NewEnumerator1
     Int Selection Array=DataInterface() {propertyValues {IntData [12,13]}}
    )
    MaskSpawnGroup=false
    Mesh Sampling Type=::NewEnumerator5
    Sampled Position Offset=<0 0 0>
    Skeletal Mesh=$System.Skeletal Mesh
    Socket Sampling Mode=::NewEnumerator5
   )
   &CameraOffset(Camera Offset Amount=&FN_FrontEndFloatModifier(Gameplay=30, Front End=0))
   /SetVariables(Particles.OffsetPosition:NiagaraPosition=&AddVectorToPosition(
     Position=$Particles.Position
     Vector=&RotateVectorByQuaternion(Vector To Rotate=<0 0 -5>, Quaternion=$Particles.SkeletalMeshLocation.SampledRotation)
    )) {isSetParameters=true, module=SetVariables_#10}
   &DynamicMaterialParameters(
    Index 0 Param 1=&MakeFloatFromVector(Vector=Linked() {linkedVariable {name=Particles.Position, type=NiagaraPosition}}, Channel=::NewEnumerator0)
    Index 0 Param 2=&MakeFloatFromVector(Vector=Linked() {linkedVariable {name=Particles.Position, type=NiagaraPosition}}, Channel=::NewEnumerator1)
    Index 0 Param 3=&MakeFloatFromVector(Vector=Linked() {linkedVariable {name=Particles.Position, type=NiagaraPosition}}, Channel=::NewEnumerator2)
    Index 0 Param 4=&RandomRangeFloat(Minimum=0, Maximum=1)
    Index 1 Param 1=&MakeFloatFromVector(Vector=Linked() {linkedVariable {name=Particles.OffsetPosition, type=NiagaraPosition}}, Channel=::NewEnumerator0)
    Index 1 Param 2=&MakeFloatFromVector(Vector=Linked() {linkedVariable {name=Particles.OffsetPosition, type=NiagaraPosition}}, Channel=::NewEnumerator1)
    Index 1 Param 3=&MakeFloatFromVector(Vector=Linked() {linkedVariable {name=Particles.OffsetPosition, type=NiagaraPosition}}, Channel=::NewEnumerator2)
    Index 1 Param 4=&Lerp_Float(
     A=&FloatFromCurve(
      FloatCurve=DataInterface() {propertyValues {Curve {keys [
          {arriveTangent=-0.0762, interpMode=RCIM_Cubic, leaveTangent=-0.0762, tangentMode=RCTM_User, time=0, value=0.8}
          {arriveTangent=-0.2085, interpMode=RCIM_Cubic, leaveTangent=-0.2085, tangentMode=RCTM_User, time=1, value=0.3}
         ]}}}
      Scale Curve=1.35
     )
     B=&FloatFromCurve(
      FloatCurve=DataInterface() {propertyValues {Curve {keys [
          {arriveTangent=-0.8006, interpMode=RCIM_Cubic, leaveTangent=-0.8006, tangentMode=RCTM_User, time=0, value=0.8}
          {arriveTangent=-0.3233, interpMode=RCIM_Cubic, leaveTangent=-0.3233, tangentMode=RCTM_User, time=1, value=0.4}
         ]}}}
      Scale Curve=1.35
     )
     Alpha=1
    )
    Index 2 Param 1=1
    Module.Index0Float0Write=true
    Module.Index0Float1Write=true
    Module.Index0Float2Write=true
    Module.Index0Float3Write=true
    Module.Index1Float0Write=true
    Module.Index1Float1Write=true
    Module.Index1Float2Write=true
    Module.Index1Float3Write=true
    Module.Index2Float0Write=true
    Module.Index2Float1Write=false
    Module.Index2Float2Write=false
    Module.Index2Float3Write=false
    Write Parameter Index 1=::NewEnumerator1
    Write Parameter Index 2=::NewEnumerator1
   )
   &ScaleColor(
    Scale Alpha=&FloatFromCurve(
     FloatCurve=DataInterface() {propertyValues {Curve {keys [
         {arriveTangent=0.5632, interpMode=RCIM_Cubic, leaveTangent=0.5632, tangentMode=RCTM_User, time=0, value=0}
         {arriveTangent=0.168, interpMode=RCIM_Cubic, leaveTangent=0.168, tangentMode=RCTM_User, time=0.25, value=1}
         {arriveTangent=-0.1012, interpMode=RCIM_Cubic, leaveTangent=-0.1012, tangentMode=RCTM_User, time=1, value=0}
        ]}}}
     Scale Curve=&Multiply_Float(A=&FN_FrontEndFloatModifier(Gameplay=8, Front End=12), B=$System.Timeline001.Time)
    )
    Scale RGB=&VectorFromFloat(Value=&FloatFromCurve(FloatCurve=DataInterface() {propertyValues {Curve {keys [
          {arriveTangent=0.168, interpMode=RCIM_Cubic, leaveTangent=0.168, tangentMode=RCTM_User, time=0.2, value=1.5}
          {arriveTangent=-0.1012, interpMode=RCIM_Cubic, leaveTangent=-0.1012, tangentMode=RCTM_User, time=0.4, value=1}
         ]}}}))
    ScaleA=true
    ScaleRGB=true
   )
   &ScaleColor(
    Scale Alpha=$System.AudioReactivity
    Scale RGB=&VectorFromFloat(Value=&Add_Float(
      A=&FloatFromCurve(
       FloatCurve=DataInterface() {propertyValues {Curve {keys [
           {interpMode=RCIM_Cubic, tangentMode=RCTM_Break, time=0, value=0}
           {arriveTangent=3.7964, interpMode=RCIM_Cubic, leaveTangent=1.1533, tangentMode=RCTM_Break, time=0.9, value=1}
           <at 1 0>
          ]}}}
       CurveIndex=$System.AudioReactivity
      )
      B=1
     ))
    ScaleA=true
    ScaleRGB=true
   ) {module=ScaleColor001}
  ]
  rendererClasses [&NiagaraSpriteRendererProperties]
  renderers [
   {rendererClass=&NiagaraSpriteRendererProperties}
   {rendererClass=&NiagaraSpriteRendererProperties, rendererIndex=1}
   {rendererClass=&NiagaraSpriteRendererProperties, rendererIndex=2}
  ]
 }
 FN_Debug {
  bEnabled=false
  emitterUpdateScript [
   &EmitterState()
   &SpawnBurst_Instantaneous(Spawn Count=1, Use Spawn Probability=false)
  ]
  particleSpawnScript [/Niagara/Modules/Spawn/Initialization/InitializeParticle(
    Lifetime=5
    Module.Write Color=true
    Module.Write Lifetime=true
    Module.Write Mass=true
    Module.Write Material Random=true
    Module.Write Position=false
    Module.Write Scale=false
    Module.Write SpriteRotation=false
    Module.Write SpriteSize=true
    Position=&SimulationPosition()
    Sprite Size=<100 100>
   )]
  particleUpdateScript [
   &ParticleState()
   /SetVariables(Particles.Debug:NiagaraFloat=$System.AudioReactivity) {isSetParameters=true}
   &DynamicMaterialParameters(
    Index 0 Param 1=$Particles.Debug
    Module.Param0WriteEnabled=true
    Module.Param1WriteEnabled=true
    Module.Param2WriteEnabled=true
    Module.Param3WriteEnabled=true
   )
   &SystemLocation(MaskSpawnGroup=false, Offset=<50 0 100>)
  ]
  rendererClasses [&NiagaraSpriteRendererProperties]
  renderers [{rendererClass=&NiagaraSpriteRendererProperties}]
 }
 Legs_SoundLines {
  emitterUpdateScript [
   &EmitterState(Inactive Response=::NewEnumerator1, Life Cycle Mode=::NewEnumerator1, Loop Duration=0.5, UseLoopDelay=false)
   &SpawnBurst_Instantaneous(
    Spawn Count=8
    Spawn Probability=&MakeCustomFloatFromBool(Bool=&SetBoolByFloatComparison(A=$System.AudioReactivity, B=0.1), True Float=1, False Float=0)
    Use Spawn Probability=true
    UseLoopCountLimit=false
   )
   &SpawnBurst_Instantaneous(
    Spawn Count=8
    Spawn Probability=&MakeCustomFloatFromBool(Bool=&SetBoolByFloatComparison(A=$System.AudioReactivity, B=0.1), True Float=1, False Float=0)
    Spawn Time=0.2
    Use Spawn Probability=true
    UseLoopCountLimit=false
   ) {module=SpawnBurst_Instantaneous001}
  ]
  particleSpawnScript [
   &PartitionParticles(Number Of Partitions=8, Particles Per Partition=1, Partition Mode:ENiagara_PartitionMode_v2=::NewEnumerator1)
   /Niagara/Modules/Spawn/Initialization/V2/InitializeParticle(
    Color=&MergeHelper_ColorSwitch(
     0=<rgba 0.6667 0 1 1>
     1=<rgba 0 0.6667 1 1>
     2=<rgba 0.6667 0 1 1>
     3=<rgba 0 0.6667 1 1>
     4=<rgba 0.6667 0 1 1>
     5=<rgba 0 0.6667 1 1>
     6=<rgba 0.6667 0 1 1>
     7=<rgba 0.6667 0 1 1>
     Index=$Particles.PartitionParticles.Partition
     Num=::NewEnumerator6
    )
    Lifetime=0.6
    Mesh Scale Mode=::NewEnumerator3
    Mesh Uniform Scale=0.03
    Sprite Size Mode=::NewEnumerator3
    Uniform Sprite Size=50
    UsePositionOffset=false
    Write Material Random=true
   )
   /SetVariables(Particles.VisibilityTag=&MakeCustomIntFromBool(Bool=$User.bIsFrontEndPreview, True INT=1, False INT=0)) {isSetParameters=true, module=SetVariables_#11}
  ]
  particleUpdateScript [
   &ParticleState()
   &SkeletalMeshLocation(
    Bone / Socket Index=&SelectIntFromArray(
     Direct Array Index=$Particles.PartitionParticles.Partition
     Direct Array Mode=::NewEnumerator1
     Int Selection Array=DataInterface() {propertyValues {IntData [0,1,2,3,4,5,6,7]}}
    )
    MaskSpawnGroup=false
    Mesh Sampling Type=::NewEnumerator5
    Sampled Position Offset=<0 0 0>
    Skeletal Mesh=$System.Skeletal Mesh
    Socket Sampling Mode=::NewEnumerator5
   ) {module=SkeletalMeshLocation002}
   &CameraOffset(Camera Offset Amount=20)
   /SetVariables(Particles.OffsetPosition:NiagaraPosition=&AddVectorToPosition(
     Position=$Particles.Position
     Vector=&RotateVectorByQuaternion(
      Vector To Rotate=&MergeHelper_VectorSwitch(
       0=<0 0 0>
       1=<0 0 0>
       2=<0 0 0>
       3=<0 0 0>
       4=<0 0 0>
       5=<0 0 0>
       6=<0 0 7>
       7=<0 0 -7>
       Index=$Particles.PartitionParticles.Partition
       Num=::NewEnumerator6
      )
      Quaternion=$Particles.SkeletalMeshLocation002.SampledRotation
     )
    )) {isSetParameters=true, module=SetVariables_#10}
   &DynamicMaterialParameters(
    Index 0 Param 1=&MakeFloatFromVector(Vector=Linked() {linkedVariable {name=Particles.Position, type=NiagaraPosition}}, Channel=::NewEnumerator0)
    Index 0 Param 2=&MakeFloatFromVector(Vector=Linked() {linkedVariable {name=Particles.Position, type=NiagaraPosition}}, Channel=::NewEnumerator1)
    Index 0 Param 3=&MakeFloatFromVector(Vector=Linked() {linkedVariable {name=Particles.Position, type=NiagaraPosition}}, Channel=::NewEnumerator2)
    Index 0 Param 4=&Lerp_Float(
     A=&FloatFromCurve(FloatCurve=DataInterface() {propertyValues {Curve {keys [
          {arriveTangent=-0.2971, interpMode=RCIM_Cubic, leaveTangent=-0.2971, tangentMode=RCTM_User, time=0, value=1.1}
          {arriveTangent=-0.1318, interpMode=RCIM_Cubic, leaveTangent=-0.1318, tangentMode=RCTM_User, time=1, value=0.8}
         ]}}})
     B=&FloatFromCurve(FloatCurve=DataInterface() {propertyValues {Curve {keys [
          {arriveTangent=-0.2971, interpMode=RCIM_Cubic, leaveTangent=-0.2971, tangentMode=RCTM_User, time=0, value=1}
          {arriveTangent=-0.1318, interpMode=RCIM_Cubic, leaveTangent=-0.1318, tangentMode=RCTM_User, time=1, value=0.6}
         ]}}})
     Alpha=1
    )
    Index 1 Param 1=&MakeFloatFromVector(Vector=Linked() {linkedVariable {name=Particles.OffsetPosition, type=NiagaraPosition}}, Channel=::NewEnumerator0)
    Index 1 Param 2=&MakeFloatFromVector(Vector=Linked() {linkedVariable {name=Particles.OffsetPosition, type=NiagaraPosition}}, Channel=::NewEnumerator1)
    Index 1 Param 3=&MakeFloatFromVector(Vector=Linked() {linkedVariable {name=Particles.OffsetPosition, type=NiagaraPosition}}, Channel=::NewEnumerator2)
    Index 1 Param 4=&FloatFromCurve(
     FloatCurve=DataInterface() {propertyValues {Curve {keys [
         {arriveTangent=0.1333, interpMode=RCIM_Cubic, leaveTangent=0.1333, tangentMode=RCTM_User, time=0, value=0}
         {interpMode=RCIM_Cubic, leaveTangent=-0.5523, tangentMode=RCTM_Break, time=0.3, value=0.8}
         {arriveTangent=-0.6776, interpMode=RCIM_Cubic, leaveTangent=-0.6776, tangentMode=RCTM_User, time=1, value=0.2}
        ]}}}
     Scale Curve=2
    )
    Index 2 Param 1=&FloatFromCurve(FloatCurve=DataInterface() {propertyValues {Curve {keys [<at 0 0>,<at 1 1>]}}}, Scale Curve=30)
    Module.Index0Param0Write=true
    Module.Index0Param1Write=true
    Module.Index0Param2Write=true
    Module.Index0Param3Write=true
    Module.Index1Param0Write=true
    Module.Index1Param1Write=true
    Module.Index1Param2Write=true
    Module.Index1Param3Write=true
    Module.Index2Param0Write=true
    Module.Index2Param1Write=true
    Module.Index2Param2Write=true
    Module.Index2Param3Write=true
    Write Parameter Index 1=true
    Write Parameter Index 2=true
   )
   &ScaleColor(
    Scale Alpha=&FloatFromCurve(
     FloatCurve=DataInterface() {propertyValues {Curve {keys [
         {arriveTangent=19.3619, interpMode=RCIM_Cubic, leaveTangent=19.3619, tangentMode=RCTM_User, time=0, value=0}
         {arriveTangent=-2.6364, interpMode=RCIM_Cubic, leaveTangent=-2.6364, tangentMode=RCTM_User, time=0.15, value=0.6}
         {arriveTangent=-0.6911, interpMode=RCIM_Cubic, leaveTangent=-0.6911, tangentMode=RCTM_User, time=1, value=0}
        ]}}}
     Scale Curve=&Multiply_Float(A=0.5, B=$System.Timeline001.Time)
    )
    Scale RGB=&VectorFromFloat(Value=&FloatFromCurve(
      FloatCurve=DataInterface() {propertyValues {Curve {keys [
          {arriveTangent=0.168, interpMode=RCIM_Cubic, leaveTangent=0.168, tangentMode=RCTM_User, time=0.2, value=1.5}
          {arriveTangent=-0.1012, interpMode=RCIM_Cubic, leaveTangent=-0.1012, tangentMode=RCTM_User, time=0.4, value=1}
         ]}}}
      Scale Curve=1.2
     ))
    ScaleA=true
    ScaleRGB=true
   )
   &ScaleColor(
    Scale Alpha=&FloatFromCurve(
     FloatCurve=DataInterface() {propertyValues {Curve {keys [
         {interpMode=RCIM_Cubic, leaveTangent=0.431, tangentMode=RCTM_Break, time=0, value=0}
         {arriveTangent=3.1641, interpMode=RCIM_Cubic, leaveTangent=-6.0307, tangentMode=RCTM_Break, time=0.9, value=2}
         <at 1 1>
        ]}}}
     CurveIndex=$System.AudioReactivity
     Scale Curve=0.5
    )
    ScaleA=true
    ScaleRGB=false
   ) {module=ScaleColor001}
   &ScaleSpriteSize(
    Uniform Curve Index=$System.AudioReactivity
    Uniform Curve Sprite Scale=DataInterface() {propertyValues {Curve {keys [{tangentMode=RCTM_User, time=0, value=1}, {tangentMode=RCTM_User, time=0.9, value=2}, <at 1 1>]}}}
   )
   &ScaleColor(
    Scale Alpha=&MakeCustomFloatFromBool(
     Bool=$User.bisFootHidden
     True Float=&MergeHelper_FloatSwitch(0=0, 1=0, 2=0, 3=0, 4=0, 5=0, 6=1, 7=1, Index=$Particles.PartitionParticles.Partition, Num=::NewEnumerator6)
     False Float=1
    )
    ScaleA=true
    ScaleRGB=false
   ) {module=ScaleColor002}
  ]
  rendererClasses [&NiagaraSpriteRendererProperties]
  renderers [
   {rendererClass=&NiagaraSpriteRendererProperties}
   {rendererClass=&NiagaraSpriteRendererProperties, rendererIndex=1}
  ]
 }
 Legs_SoundWaves {
  emitterUpdateScript [
   &EmitterState(Inactive Response=::NewEnumerator1, Life Cycle Mode=::NewEnumerator1, Loop Duration=0.5, UseLoopDelay=false)
   &SpawnBurst_Instantaneous(
    Spawn Count=8
    Spawn Probability=&MakeCustomFloatFromBool(Bool=&SetBoolByFloatComparison(A=$System.AudioReactivity, B=0.1), True Float=1, False Float=0)
    Use Spawn Probability=true
    UseLoopCountLimit=false
   )
   &SpawnBurst_Instantaneous(
    Spawn Count=8
    Spawn Probability=&MakeCustomFloatFromBool(Bool=&SetBoolByFloatComparison(A=$System.AudioReactivity, B=0.1), True Float=1, False Float=0)
    Spawn Time=0.2
    Use Spawn Probability=true
    UseLoopCountLimit=false
   ) {module=SpawnBurst_Instantaneous001}
  ]
  particleSpawnScript [
   &PartitionParticles(Number Of Partitions=8, Particles Per Partition=1, Partition Mode:ENiagara_PartitionMode_v2=::NewEnumerator1)
   &InitializeParticle(
    Color=&MergeHelper_ColorSwitch(
     0=<rgba 0.6667 0 1 1>
     1=<rgba 0 0.6667 1 1>
     2=<rgba 0.6667 0 1 1>
     3=<rgba 0 0.6667 1 1>
     4=<rgba 0.6667 0 1 1>
     5=<rgba 0 0.6667 1 1>
     6=<rgba 0.6667 0 1 1>
     7=<rgba 0.6667 0 1 1>
     Index=$Particles.PartitionParticles.Partition
     Num=::NewEnumerator6
    )
    Lifetime=0.5
    Sprite Size Mode=::NewEnumerator3
    Uniform Sprite Size=&SelectFloatFromArray(
     Direct Array Index=$Particles.PartitionParticles.Partition
     Float Selection Array=DataInterface() {propertyValues {FloatData [40,40,40,40,40,40,40,40]}}
    )
    UsePositionOffset=false
    Write Material Random=true
   )
   /SetVariables(Particles.VisibilityTag=&MakeCustomIntFromBool(Bool=$User.bIsFrontEndPreview, True INT=1, False INT=0)) {isSetParameters=true, module=SetVariables_#9}
  ]
  particleUpdateScript [
   &ParticleState()
   &SkeletalMeshLocation(
    Bone / Socket Index=&SelectIntFromArray(
     Direct Array Index=$Particles.PartitionParticles.Partition
     Direct Array Mode=::NewEnumerator1
     Int Selection Array=DataInterface() {propertyValues {IntData [0,1,2,3,4,5,6,7]}}
    )
    MaskSpawnGroup=false
    Mesh Sampling Type=::NewEnumerator5
    Sampled Position Offset=<0 0 0>
    Skeletal Mesh=$System.Skeletal Mesh
    Socket Sampling Mode=::NewEnumerator5
   )
   &CameraOffset(Camera Offset Amount=&FN_FrontEndFloatModifier(Gameplay=20, Front End=0))
   /SetVariables(Particles.OffsetPosition:NiagaraPosition=&AddVectorToPosition(
     Position=$Particles.Position
     Vector=&RotateVectorByQuaternion(
      Vector To Rotate=&MergeHelper_VectorSwitch(
       0=<0 0 0>
       1=<0 0 0>
       2=<0 0 0>
       3=<0 0 0>
       4=<0 0 0>
       5=<0 0 0>
       6=<0 0 7>
       7=<0 0 -7>
       Index=$Particles.PartitionParticles.Partition
       Num=::NewEnumerator6
      )
      Quaternion=$Particles.SkeletalMeshLocation.SampledRotation
     )
    )) {isSetParameters=true, module=SetVariables_#10}
   &DynamicMaterialParameters(
    Index 0 Param 1=&MakeFloatFromVector(Vector=Linked() {linkedVariable {name=Particles.Position, type=NiagaraPosition}}, Channel=::NewEnumerator0)
    Index 0 Param 2=&MakeFloatFromVector(Vector=Linked() {linkedVariable {name=Particles.Position, type=NiagaraPosition}}, Channel=::NewEnumerator1)
    Index 0 Param 3=&MakeFloatFromVector(Vector=Linked() {linkedVariable {name=Particles.Position, type=NiagaraPosition}}, Channel=::NewEnumerator2)
    Index 0 Param 4=&RandomRangeFloat(Minimum=0, Maximum=1)
    Index 1 Param 1=&MakeFloatFromVector(Vector=Linked() {linkedVariable {name=Particles.OffsetPosition, type=NiagaraPosition}}, Channel=::NewEnumerator0)
    Index 1 Param 2=&MakeFloatFromVector(Vector=Linked() {linkedVariable {name=Particles.OffsetPosition, type=NiagaraPosition}}, Channel=::NewEnumerator1)
    Index 1 Param 3=&MakeFloatFromVector(Vector=Linked() {linkedVariable {name=Particles.OffsetPosition, type=NiagaraPosition}}, Channel=::NewEnumerator2)
    Index 1 Param 4=&Lerp_Float(
     A=&FloatFromCurve(
      FloatCurve=DataInterface() {propertyValues {Curve {keys [
          {arriveTangent=-0.0762, interpMode=RCIM_Cubic, leaveTangent=-0.0762, tangentMode=RCTM_User, time=0, value=0.8}
          {arriveTangent=-0.2085, interpMode=RCIM_Cubic, leaveTangent=-0.2085, tangentMode=RCTM_User, time=1, value=0.3}
         ]}}}
      Scale Curve=1.35
     )
     B=&FloatFromCurve(
      FloatCurve=DataInterface() {propertyValues {Curve {keys [
          {arriveTangent=-0.8006, interpMode=RCIM_Cubic, leaveTangent=-0.8006, tangentMode=RCTM_User, time=0, value=0.8}
          {arriveTangent=-0.3233, interpMode=RCIM_Cubic, leaveTangent=-0.3233, tangentMode=RCTM_User, time=1, value=0.4}
         ]}}}
      Scale Curve=1.35
     )
     Alpha=1
    )
    Index 2 Param 1=1
    Module.Index0Float0Write=true
    Module.Index0Float1Write=true
    Module.Index0Float2Write=true
    Module.Index0Float3Write=true
    Module.Index1Float0Write=true
    Module.Index1Float1Write=true
    Module.Index1Float2Write=true
    Module.Index1Float3Write=true
    Module.Index2Float0Write=true
    Module.Index2Float1Write=false
    Module.Index2Float2Write=false
    Module.Index2Float3Write=false
    Write Parameter Index 1=::NewEnumerator1
    Write Parameter Index 2=::NewEnumerator1
   )
   &ScaleColor(
    Scale Alpha=&FloatFromCurve(
     FloatCurve=DataInterface() {propertyValues {Curve {keys [
         {arriveTangent=0.5632, interpMode=RCIM_Cubic, leaveTangent=0.5632, tangentMode=RCTM_User, time=0, value=0}
         {arriveTangent=0.168, interpMode=RCIM_Cubic, leaveTangent=0.168, tangentMode=RCTM_User, time=0.25, value=1}
         {arriveTangent=-0.1012, interpMode=RCIM_Cubic, leaveTangent=-0.1012, tangentMode=RCTM_User, time=1, value=0}
        ]}}}
     Scale Curve=&Multiply_Float(A=&FN_FrontEndFloatModifier(Gameplay=10, Front End=20), B=$System.Timeline001.Time)
    )
    Scale RGB=&VectorFromFloat(Value=&FloatFromCurve(FloatCurve=DataInterface() {propertyValues {Curve {keys [
          {arriveTangent=0.168, interpMode=RCIM_Cubic, leaveTangent=0.168, tangentMode=RCTM_User, time=0.2, value=1.5}
          {arriveTangent=-0.1012, interpMode=RCIM_Cubic, leaveTangent=-0.1012, tangentMode=RCTM_User, time=0.4, value=1}
         ]}}}))
    ScaleA=true
    ScaleRGB=true
   )
   &ScaleColor(
    Scale Alpha=$System.AudioReactivity
    Scale RGB=&VectorFromFloat(Value=&Add_Float(
      A=&FloatFromCurve(
       FloatCurve=DataInterface() {propertyValues {Curve {keys [
           {interpMode=RCIM_Cubic, tangentMode=RCTM_Break, time=0, value=0}
           {arriveTangent=3.7964, interpMode=RCIM_Cubic, leaveTangent=1.1533, tangentMode=RCTM_Break, time=0.9, value=1}
           <at 1 0>
          ]}}}
       CurveIndex=$System.AudioReactivity
      )
      B=1
     ))
    ScaleA=true
    ScaleRGB=true
   ) {module=ScaleColor003}
   &ScaleColor(
    Scale Alpha=&FN_FrontEndFloatModifier(
     Gameplay=&FN_PickFloatDependOnPlatform(
      TargetPlatform=DataInterface() {propertyValues {Platforms {qualityLevelMask=-2}}}
      TargetPlatformActive=1
      TargetPlatformNoActive=&MergeHelper_FloatSwitch(0=0.2, 1=0.2, 2=0.2, 3=0.2, 4=1, 5=1, 6=1, 7=1, Index=$Particles.PartitionParticles.Partition, Num=::NewEnumerator6)
     )
     Front End=&MergeHelper_FloatSwitch(0=0.2, 1=0.2, 2=0.2, 3=0.2, 4=1, 5=1, 6=1, 7=1, Index=$Particles.PartitionParticles.Partition, Num=::NewEnumerator6)
    )
    ScaleA=true
    ScaleRGB=false
   ) {module=ScaleColor001}
   &ScaleColor(
    Scale Alpha=&MakeCustomFloatFromBool(
     Bool=$User.bisFootHidden
     True Float=&MergeHelper_FloatSwitch(0=0, 1=0, 2=0, 3=0, 4=0, 5=0, 6=1, 7=1, Index=$Particles.PartitionParticles.Partition, Num=::NewEnumerator6)
     False Float=1
    )
    ScaleA=true
    ScaleRGB=false
   ) {module=ScaleColor002}
  ]
  rendererClasses [&NiagaraSpriteRendererProperties]
  renderers [
   {rendererClass=&NiagaraSpriteRendererProperties}
   {rendererClass=&NiagaraSpriteRendererProperties, rendererIndex=1}
   {rendererClass=&NiagaraSpriteRendererProperties, rendererIndex=2}
  ]
 }
 PostLine002 {
  emitterUpdateScript [
   &EmitterState(Inactive Response=::NewEnumerator1, Life Cycle Mode=::NewEnumerator1, Loop Duration=0.5, UseLoopDelay=false)
   &SpawnBurst_Instantaneous(
    Spawn Count=4
    Spawn Probability=&MakeCustomFloatFromBool(Bool=&SetBoolByFloatComparison(A=$System.AudioReactivity, B=0.1), True Float=1, False Float=0)
    Use Spawn Probability=true
    UseLoopCountLimit=false
   )
   &SpawnBurst_Instantaneous(
    Spawn Count=4
    Spawn Probability=&MakeCustomFloatFromBool(Bool=&SetBoolByFloatComparison(A=$System.AudioReactivity, B=0.1), True Float=1, False Float=0)
    Spawn Time=0.2
    Use Spawn Probability=true
    UseLoopCountLimit=false
   ) {module=SpawnBurst_Instantaneous001}
  ]
  particleSpawnScript [
   &PartitionParticles(Number Of Partitions=4, Particles Per Partition=1, Partition Mode:ENiagara_PartitionMode_v2=::NewEnumerator1)
   &InitializeParticle(
    Color=&MergeHelper_ColorSwitch(
     0=<rgba 0 0.6667 1 1>
     1=<rgba 0 0.6667 1 1>
     2=<rgba 0 0.6667 1 1>
     3=<rgba 0 0.6667 1 1>
     Index=$Particles.PartitionParticles.Partition
     Num=::NewEnumerator2
    )
    Lifetime=0.6
    Mesh Scale Mode=::NewEnumerator3
    Mesh Uniform Scale=0.03
    Sprite Size Mode=::NewEnumerator3
    Uniform Sprite Size=25
    UsePositionOffset=false
    Write Material Random=true
   )
   /SetVariables(Particles.VisibilityTag=&MakeCustomIntFromBool(Bool=$User.bIsFrontEndPreview, True INT=1, False INT=0)) {isSetParameters=true, module=SetVariables_#11}
  ]
  particleUpdateScript [
   &ParticleState()
   &SkeletalMeshLocation(
    Bone / Socket Index=&SelectIntFromArray(
     Direct Array Index=$Particles.PartitionParticles.Partition
     Direct Array Mode=::NewEnumerator1
     Int Selection Array=DataInterface() {propertyValues {IntData [8,9,10,11]}}
    )
    MaskSpawnGroup=false
    Mesh Sampling Type=::NewEnumerator5
    Sampled Position Offset=<0 0 0>
    Skeletal Mesh=$System.Skeletal Mesh
    Socket Sampling Mode=::NewEnumerator5
   ) {module=SkeletalMeshLocation002}
   &CameraOffset(Camera Offset Amount=20)
   /SetVariables(Particles.OffsetPosition:NiagaraPosition=&AddVectorToPosition(
     Position=$Particles.Position
     Vector=&RotateVectorByQuaternion(Vector To Rotate=<0 0 -5>, Quaternion=$Particles.SkeletalMeshLocation002.SampledRotation)
    )) {isSetParameters=true, module=SetVariables_#10}
   &DynamicMaterialParameters(
    Index 0 Param 1=&MakeFloatFromVector(Vector=Linked() {linkedVariable {name=Particles.Position, type=NiagaraPosition}}, Channel=::NewEnumerator0)
    Index 0 Param 2=&MakeFloatFromVector(Vector=Linked() {linkedVariable {name=Particles.Position, type=NiagaraPosition}}, Channel=::NewEnumerator1)
    Index 0 Param 3=&MakeFloatFromVector(Vector=Linked() {linkedVariable {name=Particles.Position, type=NiagaraPosition}}, Channel=::NewEnumerator2)
    Index 0 Param 4=&Lerp_Float(
     A=&FloatFromCurve(FloatCurve=DataInterface() {propertyValues {Curve {keys [
          {arriveTangent=-0.2971, interpMode=RCIM_Cubic, leaveTangent=-0.2971, tangentMode=RCTM_User, time=0, value=1.1}
          {arriveTangent=-0.1318, interpMode=RCIM_Cubic, leaveTangent=-0.1318, tangentMode=RCTM_User, time=1, value=0.8}
         ]}}})
     B=&FloatFromCurve(
      FloatCurve=DataInterface() {propertyValues {Curve {keys [
          {arriveTangent=-0.2971, interpMode=RCIM_Cubic, leaveTangent=-0.2971, tangentMode=RCTM_User, time=0, value=1}
          {arriveTangent=-0.1318, interpMode=RCIM_Cubic, leaveTangent=-0.1318, tangentMode=RCTM_User, time=1, value=0.6}
         ]}}}
      Scale Curve=0.9
     )
     Alpha=1
    )
    Index 1 Param 1=&MakeFloatFromVector(Vector=Linked() {linkedVariable {name=Particles.SkeletalMeshLocation002.SampledPosition, type=NiagaraPosition}}, Channel=::NewEnumerator0)
    Index 1 Param 2=&MakeFloatFromVector(Vector=Linked() {linkedVariable {name=Particles.SkeletalMeshLocation002.SampledPosition, type=NiagaraPosition}}, Channel=::NewEnumerator1)
    Index 1 Param 3=&MakeFloatFromVector(Vector=Linked() {linkedVariable {name=Particles.SkeletalMeshLocation002.SampledPosition, type=NiagaraPosition}}, Channel=::NewEnumerator2)
    Index 1 Param 4=&FloatFromCurve(
     FloatCurve=DataInterface() {propertyValues {Curve {keys [
         {arriveTangent=0.1333, interpMode=RCIM_Cubic, leaveTangent=0.1333, tangentMode=RCTM_User, time=0, value=0}
         {interpMode=RCIM_Cubic, leaveTangent=-0.5523, tangentMode=RCTM_Break, time=0.3, value=0.8}
         {arriveTangent=-0.6776, interpMode=RCIM_Cubic, leaveTangent=-0.6776, tangentMode=RCTM_User, time=1, value=0.2}
        ]}}}
     Scale Curve=2
    )
    Index 2 Param 1=&FloatFromCurve(FloatCurve=DataInterface() {propertyValues {Curve {keys [<at 0 0>,<at 1 1>]}}})
    Index 2 Param 4=1000
    Module.Index0Param0Write=true
    Module.Index0Param1Write=true
    Module.Index0Param2Write=true
    Module.Index0Param3Write=true
    Module.Index1Param0Write=true
    Module.Index1Param1Write=true
    Module.Index1Param2Write=true
    Module.Index1Param3Write=true
    Module.Index2Param0Write=true
    Module.Index2Param1Write=true
    Module.Index2Param2Write=true
    Module.Index2Param3Write=true
    Write Parameter Index 1=true
    Write Parameter Index 2=true
   )
   &ScaleColor(
    Scale Alpha=&FloatFromCurve(
     FloatCurve=DataInterface() {propertyValues {Curve {keys [
         {arriveTangent=19.3619, interpMode=RCIM_Cubic, leaveTangent=19.3619, tangentMode=RCTM_User, time=0, value=0}
         {arriveTangent=-2.6364, interpMode=RCIM_Cubic, leaveTangent=-2.6364, tangentMode=RCTM_User, time=0.15, value=0.6}
         {arriveTangent=-0.6911, interpMode=RCIM_Cubic, leaveTangent=-0.6911, tangentMode=RCTM_User, time=1, value=0}
        ]}}}
     Scale Curve=&Multiply_Float(A=0.5, B=$System.Timeline001.Time)
    )
    Scale RGB=&VectorFromFloat(Value=&FloatFromCurve(
      FloatCurve=DataInterface() {propertyValues {Curve {keys [
          {arriveTangent=0.168, interpMode=RCIM_Cubic, leaveTangent=0.168, tangentMode=RCTM_User, time=0.2, value=1.5}
          {arriveTangent=-0.1012, interpMode=RCIM_Cubic, leaveTangent=-0.1012, tangentMode=RCTM_User, time=0.4, value=1}
         ]}}}
      Scale Curve=0.5
     ))
    ScaleA=true
    ScaleRGB=true
   )
   &ScaleColor(
    Scale Alpha=&FloatFromCurve(
     FloatCurve=DataInterface() {propertyValues {Curve {keys [
         {interpMode=RCIM_Cubic, leaveTangent=0.431, tangentMode=RCTM_Break, time=0, value=0}
         {arriveTangent=3.1641, interpMode=RCIM_Cubic, leaveTangent=-6.0307, tangentMode=RCTM_Break, time=0.9, value=2}
         <at 1 1>
        ]}}}
     CurveIndex=$System.AudioReactivity
     Scale Curve=0.5
    )
    ScaleA=true
    ScaleRGB=false
   ) {module=ScaleColor003}
   &ScaleSpriteSize(
    Uniform Curve Index=$System.AudioReactivity
    Uniform Curve Sprite Scale=DataInterface() {propertyValues {Curve {keys [{tangentMode=RCTM_User, time=0, value=1}, {tangentMode=RCTM_User, time=0.9, value=2}, <at 1 1>]}}}
   ) {module=ScaleSpriteSize001}
   &ScaleColor(
    Scale Alpha=&MakeCustomFloatFromBool(Bool=$User.bisHandHidden, True Float=0, False Float=1)
    ScaleA=true
    ScaleRGB=false
   ) {module=ScaleColor002}
  ]
  rendererClasses [&NiagaraSpriteRendererProperties]
  renderers [
   {rendererClass=&NiagaraSpriteRendererProperties}
   {rendererClass=&NiagaraSpriteRendererProperties, rendererIndex=1}
  ]
 }
 PostLine003 {
  emitterUpdateScript [
   &EmitterState(Inactive Response=::NewEnumerator1, Life Cycle Mode=::NewEnumerator1, Loop Duration=0.5, UseLoopDelay=false)
   &SpawnBurst_Instantaneous(
    Spawn Count=2
    Spawn Probability=&MakeCustomFloatFromBool(Bool=&SetBoolByFloatComparison(A=$System.AudioReactivity, B=0.1), True Float=1, False Float=0)
    Use Spawn Probability=true
    UseLoopCountLimit=false
   )
   &SpawnBurst_Instantaneous(
    Spawn Count=2
    Spawn Probability=&MakeCustomFloatFromBool(Bool=&SetBoolByFloatComparison(A=$System.AudioReactivity, B=0.1), True Float=1, False Float=0)
    Spawn Time=0.2
    Use Spawn Probability=true
    UseLoopCountLimit=false
   ) {module=SpawnBurst_Instantaneous001}
  ]
  particleSpawnScript [
   &PartitionParticles(Number Of Partitions=4, Particles Per Partition=1, Partition Mode:ENiagara_PartitionMode_v2=::NewEnumerator1)
   &InitializeParticle(
    Color=&MergeHelper_ColorSwitch(
     0=<rgba 0 0.6667 1 1>
     1=<rgba 0 0.6667 1 1>
     2=<rgba 0 0.6667 1 1>
     3=<rgba 0 0.6667 1 1>
     Index=$Particles.PartitionParticles.Partition
     Num=::NewEnumerator2
    )
    Lifetime=0.6
    Mesh Scale Mode=::NewEnumerator3
    Mesh Uniform Scale=0.03
    Sprite Size Mode=::NewEnumerator3
    Uniform Sprite Size=25
    UsePositionOffset=false
    Write Material Random=true
   )
   /SetVariables(Particles.VisibilityTag=&MakeCustomIntFromBool(Bool=$User.bIsFrontEndPreview, True INT=1, False INT=0)) {isSetParameters=true, module=SetVariables_#11}
  ]
  particleUpdateScript [
   &ParticleState()
   &SkeletalMeshLocation(
    Bone / Socket Index=&SelectIntFromArray(
     Direct Array Index=$Particles.PartitionParticles.Partition
     Direct Array Mode=::NewEnumerator1
     Int Selection Array=DataInterface() {propertyValues {IntData [12,13]}}
    )
    MaskSpawnGroup=false
    Mesh Sampling Type=::NewEnumerator5
    Skeletal Mesh=$System.Skeletal Mesh
    Socket Sampling Mode=::NewEnumerator5
   ) {module=SkeletalMeshLocation002}
   &CameraOffset(Camera Offset Amount=20)
   /SetVariables(Particles.OffsetPosition:NiagaraPosition=&AddVectorToPosition(
     Position=$Particles.Position
     Vector=&RotateVectorByQuaternion(Vector To Rotate=<0 0 -5>, Quaternion=$Particles.SkeletalMeshLocation002.SampledRotation)
    )) {isSetParameters=true, module=SetVariables_#10}
   &DynamicMaterialParameters(
    Index 0 Param 1=&MakeFloatFromVector(Vector=Linked() {linkedVariable {name=Particles.Position, type=NiagaraPosition}}, Channel=::NewEnumerator0)
    Index 0 Param 2=&MakeFloatFromVector(Vector=Linked() {linkedVariable {name=Particles.Position, type=NiagaraPosition}}, Channel=::NewEnumerator1)
    Index 0 Param 3=&MakeFloatFromVector(Vector=Linked() {linkedVariable {name=Particles.Position, type=NiagaraPosition}}, Channel=::NewEnumerator2)
    Index 0 Param 4=&Lerp_Float(
     A=&FloatFromCurve(FloatCurve=DataInterface() {propertyValues {Curve {keys [
          {arriveTangent=-0.2971, interpMode=RCIM_Cubic, leaveTangent=-0.2971, tangentMode=RCTM_User, time=0, value=1.1}
          {arriveTangent=-0.1318, interpMode=RCIM_Cubic, leaveTangent=-0.1318, tangentMode=RCTM_User, time=1, value=0.8}
         ]}}})
     B=&FloatFromCurve(
      FloatCurve=DataInterface() {propertyValues {Curve {keys [
          {arriveTangent=-0.2971, interpMode=RCIM_Cubic, leaveTangent=-0.2971, tangentMode=RCTM_User, time=0, value=1}
          {arriveTangent=-0.1318, interpMode=RCIM_Cubic, leaveTangent=-0.1318, tangentMode=RCTM_User, time=1, value=0.6}
         ]}}}
      Scale Curve=0.9
     )
     Alpha=1
    )
    Index 1 Param 1=&MakeFloatFromVector(Vector=Linked() {linkedVariable {name=Particles.OffsetPosition, type=NiagaraPosition}}, Channel=::NewEnumerator0)
    Index 1 Param 2=&MakeFloatFromVector(Vector=Linked() {linkedVariable {name=Particles.OffsetPosition, type=NiagaraPosition}}, Channel=::NewEnumerator1)
    Index 1 Param 3=&MakeFloatFromVector(Vector=Linked() {linkedVariable {name=Particles.OffsetPosition, type=NiagaraPosition}}, Channel=::NewEnumerator2)
    Index 1 Param 4=&FloatFromCurve(
     FloatCurve=DataInterface() {propertyValues {Curve {keys [
         {arriveTangent=0.1333, interpMode=RCIM_Cubic, leaveTangent=0.1333, tangentMode=RCTM_User, time=0, value=0}
         {interpMode=RCIM_Cubic, leaveTangent=-0.5523, tangentMode=RCTM_Break, time=0.3, value=0.8}
         {arriveTangent=-0.6776, interpMode=RCIM_Cubic, leaveTangent=-0.6776, tangentMode=RCTM_User, time=1, value=0.2}
        ]}}}
     Scale Curve=2
    )
    Index 2 Param 1=&FloatFromCurve(FloatCurve=DataInterface() {propertyValues {Curve {keys [<at 0 0>,<at 1 1>]}}}, Scale Curve=30)
    Module.Index0Param0Write=true
    Module.Index0Param1Write=true
    Module.Index0Param2Write=true
    Module.Index0Param3Write=true
    Module.Index1Param0Write=true
    Module.Index1Param1Write=true
    Module.Index1Param2Write=true
    Module.Index1Param3Write=true
    Module.Index2Param0Write=true
    Module.Index2Param1Write=true
    Module.Index2Param2Write=true
    Module.Index2Param3Write=true
    Write Parameter Index 1=true
    Write Parameter Index 2=true
   )
   &ScaleColor(
    Scale Alpha=&FloatFromCurve(
     FloatCurve=DataInterface() {propertyValues {Curve {keys [
         {arriveTangent=19.3619, interpMode=RCIM_Cubic, leaveTangent=19.3619, tangentMode=RCTM_User, time=0, value=0}
         {arriveTangent=-2.6364, interpMode=RCIM_Cubic, leaveTangent=-2.6364, tangentMode=RCTM_User, time=0.15, value=0.6}
         {arriveTangent=-0.6911, interpMode=RCIM_Cubic, leaveTangent=-0.6911, tangentMode=RCTM_User, time=1, value=0}
        ]}}}
     Scale Curve=&Multiply_Float(A=&FN_FrontEndFloatModifier(Gameplay=0.5, Front End=0.8), B=$System.Timeline001.Time)
    )
    Scale RGB=&VectorFromFloat(Value=&FloatFromCurve(
      FloatCurve=DataInterface() {propertyValues {Curve {keys [
          {arriveTangent=0.168, interpMode=RCIM_Cubic, leaveTangent=0.168, tangentMode=RCTM_User, time=0.2, value=1.5}
          {arriveTangent=-0.1012, interpMode=RCIM_Cubic, leaveTangent=-0.1012, tangentMode=RCTM_User, time=0.4, value=1}
         ]}}}
      Scale Curve=1.2
     ))
    ScaleA=true
    ScaleRGB=true
   )
   &ScaleColor(
    Scale Alpha=&FloatFromCurve(
     FloatCurve=DataInterface() {propertyValues {Curve {keys [
         {interpMode=RCIM_Cubic, leaveTangent=0.431, tangentMode=RCTM_Break, time=0, value=0}
         {arriveTangent=3.1641, interpMode=RCIM_Cubic, leaveTangent=-6.0307, tangentMode=RCTM_Break, time=0.9, value=2}
         <at 1 1>
        ]}}}
     CurveIndex=$System.AudioReactivity
     Scale Curve=0.1
    )
    ScaleA=true
    ScaleRGB=false
   ) {module=ScaleColor003}
   &ScaleSpriteSize(
    Uniform Curve Index=$System.AudioReactivity
    Uniform Curve Sprite Scale=DataInterface() {propertyValues {Curve {keys [{tangentMode=RCTM_User, time=0, value=1}, {tangentMode=RCTM_User, time=0.9, value=2}, <at 1 1>]}}}
   ) {module=ScaleSpriteSize001}
  ]
  rendererClasses [&NiagaraSpriteRendererProperties]
  renderers [
   {rendererClass=&NiagaraSpriteRendererProperties}
   {rendererClass=&NiagaraSpriteRendererProperties, rendererIndex=1}
  ]
 }
}
summary {
 systemName=NS_Character_FareSpore_Body
 emitters [
  {emitterName=Legs_SoundWaves, rendererClasses [&NiagaraSpriteRendererProperties]}
  {emitterName=Legs_SoundLines, rendererClasses [&NiagaraSpriteRendererProperties]}
  {emitterName=Arms_SoundWaves, rendererClasses [&NiagaraSpriteRendererProperties]}
  {emitterName=PostLine002, rendererClasses [&NiagaraSpriteRendererProperties]}
  {emitterName=Arms_SoundWaves001, rendererClasses [&NiagaraSpriteRendererProperties]}
  {emitterName=PostLine003, rendererClasses [&NiagaraSpriteRendererProperties]}
  {bEnabled=false, emitterName=FN_Debug, rendererClasses [&NiagaraSpriteRendererProperties]}
 ]
 userVariables [
  User.bIsADS {type=NiagaraBool}
  User.bisFootHidden {type=NiagaraBool}
  User.bIsFrontEnd {type=NiagaraBool}
  User.bIsFrontEndPreview {type=NiagaraBool}
  User.bisHandHidden {type=NiagaraBool}
  User.bisHeadHidden {type=NiagaraBool}
  User.DebugReactivity {type=NiagaraBool}
  User.EnableFrontendFakeAudioLoop {type=NiagaraFloat, defaultValue=0, description=""}
 ]
}
