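# Niagara system for the FareSpore male character's head cosmetic, drawing an audio-reactive equalizer effect.
# Base and Equalizer spawn Z-mirrored mesh pairs, passing music spectrum and smoothed reactivity to materials through dynamic parameters, while Glow adds two offset sprites.
# System-level audio reactivity drives Equalizer brightness and Glow color, alpha and size, with a fake audio loop module available.
# A shared opacity multiplier fades all emitters, and time-of-day and front-end modifiers tune brightness.
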NiagaraSystem /BRCosmetics/Characters/Player/Male/Medium/Bodies/M_MED_FareSpore/FX/NS_Character_FareSpore_Head
dependencies {
 usedDataInterfaces [&NiagaraDataInterfaceCurve, &NiagaraDataInterfaceColorCurve]
 usedDynamicInputs [
  &SetBoolByFloatComparison
  /Niagara/DynamicInputs/Lerp/SmoothLerpOverTime_Float
  &MakeFloatFromLinearColor
  &RemapRange
  &Lerp_Float
  &ColorFromCurve
  &MakeCustomLinearColorFromBool
  &LerpLinearColors
  &InvertBool
 ]
 usedModules [
  &self:SystemSpawnScript.SystemScriptSource.SystemScriptGraph.NiagaraNodeAssignment_3.SetVariables_#1
  &self:SystemSpawnScript.SystemScriptSource.SystemScriptGraph.NiagaraNodeAssignment_2.SetVariables_#2
  &SystemState
  &Timeline
  /Game/Effects/Niagara/Modules/Audio/FN_FakeAudioReactivityLoop
  &self:SystemSpawnScript.SystemScriptSource.SystemScriptGraph.NiagaraNodeAssignment_8.SetVariables_#3
  &self:SystemSpawnScript.SystemScriptSource.SystemScriptGraph.NiagaraNodeAssignment_0.SetVariables_#4
  &self:SystemSpawnScript.SystemScriptSource.SystemScriptGraph.NiagaraNodeAssignment_1.SetVariables_#5
  &self:SystemSpawnScript.SystemScriptSource.SystemScriptGraph.NiagaraNodeAssignment_6.SetVariables_#6
  &self:SystemSpawnScript.SystemScriptSource.SystemScriptGraph.NiagaraNodeAssignment_7.SetVariables_#7
  &InterpolateOverTime
  &self:SystemSpawnScript.SystemScriptSource.SystemScriptGraph.NiagaraNodeAssignment_4.SetVariables_#8
 ]
}
emitters {
 Base {
  emitterUpdateScript [
   &EmitterState(
    Inactive Response=::NewEnumerator1
    Life Cycle Mode=::NewEnumerator1
    Loop Behavior=::NewEnumerator1
    Loop Duration Mode=::NewEnumerator1
    Scalability Mode=::NewEnumerator1
    UseLoopDelay=false
   )
   &SpawnBurst_Instantaneous(Spawn Count=2, Use Spawn Probability=false, UseLoopCountLimit=false)
  ]
  particleSpawnScript [
   &FN_SetSpawnGroup(Number of Spawn Groups=2)
   &InitializeParticle(
    Color Mode=::NewEnumerator0
    Lifetime=5
    Mesh Scale=&MergeHelper_VectorSwitch(0=<1 1 1>, 1=<1 1 -1>, 2=?, 3=?, Index=$Particles.SpawnGroup, Num=::NewEnumerator0)
    Mesh Scale Mode=::NewEnumerator4
    UsePositionOffset=false
    Write Material Random=true
   )
  ]
  particleUpdateScript [
   &ParticleState(Kill Particles When Lifetime Has Elapsed=false)
   &CameraOffset()
   &DynamicMaterialParameters(
    Index 0 Param 1=&MakeFloatFromVector4(Vector 4=&MakeVector4_FromLinearColor(LinearColor=$System.SpectrumMusic), Channel=::NewEnumerator0)
    Index 0 Param 2=&MakeFloatFromVector4(Vector 4=&MakeVector4_FromLinearColor(LinearColor=$System.SpectrumMusic), Channel=::NewEnumerator1)
    Index 0 Param 3=&MakeFloatFromVector4(Vector 4=&MakeVector4_FromLinearColor(LinearColor=$System.SpectrumMusic), Channel=::NewEnumerator2)
    Index 0 Param 4=&MakeFloatFromVector4(Vector 4=&MakeVector4_FromLinearColor(LinearColor=$System.SpectrumMusic), Channel=::NewEnumerator3)
    Index 1 Param 1=&MakeCustomFloatFromBool(Bool=$User.DebugReactivity, True Float=1, False Float=$System.InterpolateOverTime.InterpolateOverTime_Float.Moving Average)
    Module.Index0Float0Write=true
    Module.Index0Float1Write=true
    Module.Index0Float2Write=true
    Module.Index0Float3Write=true
    Module.Index1Float0Write=true
    Module.Index1Float1Write=true
    Module.Index1Float2Write=true
    Module.Index1Float3Write=true
    Write Parameter Index 1=::NewEnumerator1
   )
   /SetVariables(
    Particles.BaseColor:LinearColor=&MultiplyLinearColor(
     Linear Color=<rgba 0.6 0.6 0.6 0.2>
     Scale Factor=&MakeVector4FromVectorAndFloat("Vector (RGB)"=<1 1 1>, "Float (Alpha)"=$System.OpacityMultiplier)
    )
    Particles.GlowColor=&MultiplyLinearColor(
     Linear Color=<rgba 1.1 1.1 1.1 1>
     Scale Factor=&MakeVector4FromVectorAndFloat("Vector (RGB)"=<1 1 1>, "Float (Alpha)"=$System.OpacityMultiplier)
    )
   ) {isSetParameters=true, module=SetVariables_#9}
  ]
  rendererClasses [&NiagaraMeshRendererProperties]
  renderers [{rendererClass=&NiagaraMeshRendererProperties, rendererIndex=0}]
 }
 Equalizer {
  emitterUpdateScript [
   &EmitterState(
    Inactive Response=::NewEnumerator1
    Life Cycle Mode=::NewEnumerator1
    Loop Behavior=::NewEnumerator1
    Loop Duration Mode=::NewEnumerator1
    Scalability Mode=::NewEnumerator1
    UseLoopDelay=false
   )
   &SpawnBurst_Instantaneous(Spawn Count=2, Use Spawn Probability=false, UseLoopCountLimit=false)
  ]
  particleSpawnScript [
   &FN_SetSpawnGroup(Number of Spawn Groups=2)
   &InitializeParticle(
    Color=<rgba 1 1 1 0.13>
    Lifetime=5
    Mesh Scale=&MergeHelper_VectorSwitch(0=<1 1 1>, 1=<1 1 -1>, 2=?, 3=?, Index=$Particles.SpawnGroup, Num=::NewEnumerator0)
    Mesh Scale Mode=::NewEnumerator4
    UsePositionOffset=false
    Write Material Random=true
   )
  ]
  particleUpdateScript [
   &ParticleState(Kill Particles When Lifetime Has Elapsed=false)
   &CameraOffset()
   &DynamicMaterialParameters(
    Index 0 Param 1=&MakeFloatFromVector4(Vector 4=&MakeVector4_FromLinearColor(LinearColor=$System.SpectrumMusic), Channel=::NewEnumerator0)
    Index 0 Param 2=&MakeFloatFromVector4(Vector 4=&MakeVector4_FromLinearColor(LinearColor=$System.SpectrumMusic), Channel=::NewEnumerator1)
    Index 0 Param 3=&MakeFloatFromVector4(Vector 4=&MakeVector4_FromLinearColor(LinearColor=$System.SpectrumMusic), Channel=::NewEnumerator2)
    Index 0 Param 4=&MakeFloatFromVector4(Vector 4=&MakeVector4_FromLinearColor(LinearColor=$System.SpectrumMusic), Channel=::NewEnumerator3)
    Index 1 Param 1=&MakeCustomFloatFromBool(Bool=$User.DebugReactivity, True Float=1, False Float=$System.InterpolateOverTime.InterpolateOverTime_Float.Moving Average)
    Module.Index0Float0Write=true
    Module.Index0Float1Write=true
    Module.Index0Float2Write=true
    Module.Index0Float3Write=true
    Module.Index1Float0Write=true
    Module.Index1Float1Write=true
    Module.Index1Float2Write=true
    Module.Index1Float3Write=true
    Write Parameter Index 1=::NewEnumerator1
   )
   /SetVariables(
    Particles.BaseColor:LinearColor=&MultiplyLinearColor(
     Linear Color=<rgba 0.5 0.5 0.5 1>
     Scale Factor=&MakeVector4FromVectorAndFloat("Vector (RGB)"=<1 1 1>, "Float (Alpha)"=$System.OpacityMultiplier)
    )
    Particles.GlowColor=&Multiply_LinearColorByFloat(
     Linear Color=&MultiplyLinearColor(
      Linear Color=<rgba 1.1 1.1 1.1 1>
      Scale Factor=&MakeVector4FromVectorAndFloat("Vector (RGB)"=<1 1 1>, "Float (Alpha)"=&FN_FrontEndFloatModifier(Gameplay=1, Front End=0.5))
     )
     Float=&Multiply_Float(
      A=&FNTimeOfDayScalar(Night Scale=0.7, Morning Scale=0.7, Evening Scale=0.7)
      B=$System.OpacityMultiplier
     )
    )
   ) {isSetParameters=true, module=SetVariables_#9}
   &ScaleColor(
    Scale Alpha=&Multiply_Float(
     A=&FNTimeOfDayScalar(Night Scale=1, Morning Scale=1, Evening Scale=1)
     B=$System.OpacityMultiplier
    )
    Scale RGB=&VectorFromFloat(Value=&Add_Float(
      A=&FloatFromCurve(
       FloatCurve=DataInterface() {propertyValues {Curve {keys [
           {interpMode=RCIM_Cubic, tangentMode=RCTM_Break, time=0, value=0}
           {arriveTangent=3.7964, interpMode=RCIM_Cubic, leaveTangent=1.1533, tangentMode=RCTM_Break, time=0.9, value=1}
           <at 1 0>
          ]}}}
       CurveIndex=$System.AudioReactivity
      )
      B=0.5
     ))
    ScaleA=true
    ScaleRGB=true
   ) {module=ScaleColor001}
  ]
  rendererClasses [&NiagaraMeshRendererProperties]
  renderers [
   {rendererClass=&NiagaraMeshRendererProperties, rendererIndex=0}
   {rendererClass=&NiagaraMeshRendererProperties, rendererIndex=1}
   {rendererClass=&NiagaraMeshRendererProperties, rendererIndex=2}
  ]
 }
 Glow {
  emitterUpdateScript [
   &EmitterState(
    Inactive Response=::NewEnumerator1
    Life Cycle Mode=::NewEnumerator1
    Loop Behavior=::NewEnumerator1
    Loop Duration Mode=::NewEnumerator1
    UseLoopDelay=false
   )
   &SpawnBurst_Instantaneous(Spawn Count=2, Use Spawn Probability=false, UseLoopCountLimit=false)
  ]
  particleSpawnScript [
   &FN_SetSpawnGroup(Number of Spawn Groups=2)
   &InitializeParticle(
    Color=&MergeHelper_ColorSwitch(0=<rgba 0.1774 0 0.3 1>, 1=<rgba 0.1774 0 0.3 1>, 2=?, 3=?, Index=$Particles.SpawnGroup, Num=::NewEnumerator0)
    Lifetime=5
    Mesh Scale Mode=::NewEnumerator3
    Position Offset=&MergeHelper_VectorSwitch(0=<24 2 18>, 1=<24 2 -18>, 2=?, 3=?, Index=$Particles.SpawnGroup, Num=::NewEnumerator0)
    Position Offset Coordinate Space=::Local
    Sprite Size Mode=::NewEnumerator3
    Uniform Sprite Size=30
    UsePositionOffset=true
    Write Material Random=true
   )
   &SpriteFacingAndAlignment(Sprite Alignment=<0 0 1>, Sprite Facing=<0 1 0>, WriteAlignment=true, WriteFacing=true)
  ]
  particleUpdateScript [
   &ParticleState(Kill Particles When Lifetime Has Elapsed=false, Let Infinitely Lived Particles Die When Emitter Deactivates=true)
   &CameraOffset(Camera Offset Amount=20)
   &ScaleColor(
    Scale Alpha=&FloatFromCurve(
     FloatCurve=DataInterface() {propertyValues {Curve {keys [
         {interpMode=RCIM_Cubic, leaveTangent=0.2728, tangentMode=RCTM_Break, time=0, value=0}
         {arriveTangent=0.9757, interpMode=RCIM_Cubic, leaveTangent=-0.1879, tangentMode=RCTM_Break, time=0.9, value=1.5}
         {arriveTangent=-4.7154, tangentMode=RCTM_User, time=1, value=0}
        ]}}}
     CurveIndex=$System.AudioReactivity
     Scale Curve=&Multiply_Float(
      A=$System.OpacityMultiplier
      B=&FNTimeOfDayScalar(Night Scale=0.5, Morning Scale=0.7, Evening Scale=0.7)
     )
    )
    Scale RGB=&VectorFromFloat(Value=&Add_Float(
      A=&FloatFromCurve(
       FloatCurve=DataInterface() {propertyValues {Curve {keys [
           {interpMode=RCIM_Cubic, tangentMode=RCTM_Break, time=0, value=0}
           {arriveTangent=9.3425, interpMode=RCIM_Cubic, leaveTangent=-6.938, tangentMode=RCTM_Break, time=0.9, value=3}
           {arriveTangent=-82.1567, tangentMode=RCTM_User, time=1, value=0}
          ]}}}
       CurveIndex=$System.AudioReactivity
       Scale Curve=&FN_FrontEndFloatModifier(Gameplay=1, Front End=0.7)
      )
      B=2
     ))
    ScaleA=true
    ScaleRGB=true
   ) {module=ScaleColor001}
   /SetVariables(
    Particles.Color2=&HueShiftLinearColor(Hue=0.15, Saturation=1, Value=0.15, Linear Color=$Particles.Color)
    Particles.SpriteSize2=&Multiply_Vector2D_ByFloat(Float=2, Vector2D=$Particles.SpriteSize)
   ) {isSetParameters=true}
   &ScaleSpriteSize(
    Uniform Curve Index=$System.AudioReactivity
    Uniform Curve Sprite Scale=DataInterface() {propertyValues {Curve {keys [{tangentMode=RCTM_User, time=0, value=1}, {tangentMode=RCTM_User, time=0.9, value=1.2}, <at 1 1>]}}}
   )
  ]
  rendererClasses [&NiagaraSpriteRendererProperties]
  renderers [
   {rendererClass=&NiagaraSpriteRendererProperties}
   {rendererClass=&NiagaraSpriteRendererProperties, rendererIndex=1}
  ]
 }
}
summary {
 systemName=NS_Character_FareSpore_Head
 emitters [
  {emitterName=Equalizer, rendererClasses [&NiagaraMeshRendererProperties]}
  {emitterName=Glow, rendererClasses [&NiagaraSpriteRendererProperties]}
  {emitterName=Base, rendererClasses [&NiagaraMeshRendererProperties]}
 ]
 userVariables [
  User.bIsADS {type=NiagaraBool}
  User.bIsFrontEnd {type=NiagaraBool}
  User.bIsFrontEndPreview {type=NiagaraBool}
  User.DebugReactivity {type=NiagaraBool}
  User.EnableFrontendFakeAudioLoop {type=NiagaraFloat, defaultValue=0, description=""}
 ]
}
