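# Parent character material for the FareSpore body, adding an audio-reactive look through the IsAudioReactive static switch between the shared PreFX and PostFX functions.
# Audio amplitude drives fresnel and emissive detail, combined with scrolling pink/blue gradients, triplanar halftone and a shoulder effect.
# An FX mask confines the look, which blends in with amplitude plus custom thin film built on camera-space radial coordinates.
# DEBUG_ReactivityLook bypasses that blend and thin film.
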Material /BRCosmetics/Characters/Player/Male/Medium/Bodies/M_MED_FareSpore/Materials/M_FareSpore_Body_PARENT
graph {
 properties {
  bUseMaterialAttributes=true
  bUsedWithMeshDeformer=true
  bUsedWithMorphTargets=true
  bUsedWithSkeletalMesh=true
  blendMode=BLEND_Masked
  pixelDepthOffsetMode=PDOM_Legacy
  shadingModel=MSM_FromMaterialExpression
  usedShadingModels=Default Lit | Subsurface | Clear Coat | Subsurface Profile | Cloth
 }
 nodes [
  n1 = MaterialFunctionCall() {materialFunction=&MF_CharacterParent_2023_PreFX}
  MaterialFunctionCall(
   Material=StaticSwitchParameter(True=NamedRerouteUsage(), False=NamedRerouteDeclaration(None=@n1.Result) {name=OriginalMA, nodeColor=<rgba 1 0 0.1875 1>}) {group=Audio Reactivity, parameterName=IsAudioReactive}
   M=@n1.M
   Pre Skinned Local Position=NamedRerouteDeclaration(None=@n1.Pre skinned Local Position) {name=PreskinnedLocalPosition, nodeColor=<rgba 0 1 0.4375 1>}
   Distance Blend=@n1.M
   Fresnel=NamedRerouteDeclaration(None=@n1.Fresnel) {name=InFresnel, nodeColor=<rgba 0 1 0.8359 1>}
  ) {materialFunction=&MF_CharacterParent_2023_PostFX}
  n2 = PreSkinnedNormal()
  n3 = ShaderStageSwitch(PixelShader=VertexInterpolator(VS=@n2).PS, VertexShader=@n2)
  n4 = Time()
  n5 = Constant3Vector() {constant=<rgba 1 0 -30 0>}
  n6 = MaterialFunctionCall!(
   VertexNormal=@n3
   Basecolor Texture=TextureObject() {texture=/BRCosmetics/Characters/Player/Female/Medium/Bodies/F_MED_SpatialTravel_Watt/Skins/Superlevels/Textures/3d_VecNoise}
   Worldpos=Add(A=NamedRerouteUsage(), B=Multiply(A=@n4, B=@n5))
   TextureSize=Constant() {r=30}
  ) {materialFunction=/BRCosmetics/Characters/Player/Female/Medium/Bodies/F_MED_SpatialTravel_Watt/Skins/Superlevels/ParentMaterials/SimpleLowEndTriplanar_SL}
  Multiply(A=ComponentMask(None=@n6.Result) {g=true})
  n7 = MaterialFunctionCall!(InAmplitude, InSpectrum, CustomFakeSpectrum, CustomFakeAmplitude) {materialFunction=/BRCosmetics/Global/MaterialFunctions/AudioReactivity/MF_AudioReactivity_FrontendLoop}
  n8 = Saturate(None=@n7.Amplitude)
  n9 = Power(Base=@n8) {constExponent=3}
  n10 = Power(Base=NamedRerouteUsage(), Exp=ScalarParameter() {defaultValue=3, group=06_Thin Film, parameterName=CustomThinFilmExponent})
  n11 = ScalarParameter() {defaultValue=0.3, group=06_Thin Film, parameterName=CustomThinFilmIntensityExponent, sortPriority=3}
  n12 = Multiply(
   A=LinearInterpolate(
    A=ScalarParameter() {defaultValue=1, group=06_Thin Film, parameterName=CustomThinFilmIntensity, sortPriority=2}
    B=ScalarParameter() {defaultValue=2, group=06_Thin Film, parameterName=CustomThinFilmIntensity_Loud, sortPriority=2}
    Alpha=Multiply(A=@n9, B=@n10)
   )
   B=SmoothStep(Value=Power(Base=@n10, Exp=@n11))
  )
  n13 = NamedRerouteUsage()
  n14 = ScalarParameter() {defaultValue=0.5, group=06_Thin Film, parameterName=CustomThinFilmWarp, sortPriority=4}
  n15 = Add(
   A=Multiply(A=ScalarParameter() {defaultValue=-0.1, group=06_Thin Film, parameterName=CustomThinFilmPanSpeed}, B=Time())
   B=Multiply(A=ScalarParameter() {defaultValue=0.5, group=06_Thin Film, parameterName=CustomThinFilmScale, sortPriority=5}, B=Power(Base=@n13, Exp=@n14))
  )
  n16 = ScalarParameter() {defaultValue=0.3, group=06_Thin Film, parameterName=CustomThinFilmOffset}
  n17 = ComponentMask(None=NamedRerouteUsage()) {g=true}
  n18 = Multiply(A=Time(), B=ScalarParameter() {defaultValue=0.1, parameterName=CustomThinFilmCircularMaskRotateSpeed})
  n19 = TextureCoordinate()
  n20 = TextureSample!(
   UVs=AppendVector(A=Add(A=@n15, B=@n16), B=Subtract(A=@n17, B=@n18))
   Tex
   "DDX(UVs)"=DDX(Value=@n19)
   "DDY(UVs)"=DDY(Value=@n19)
   Apply View MipBias
  ) {
   mipValueMode=TMVM_Derivative
   samplerSource=SSM_Wrap_WorldGroupSettings
   samplerType=SAMPLERTYPE_LinearColor
   texture=/Game/Accessories/FORT_Backpacks/Backpack_M_MED_RelayStick/Textures/T_RelayStick_Pattern_A
  }
  n21 = Constant()
  n22 = Multiply(A=Time(), B=ScalarParameter() {defaultValue=0.1, parameterName=CustomThinFilmCircularPanSpeed})
  n23 = CurveAtlasRowParameter(CurveTime=Frac(None=Add(A=@n17, B=@n22))) {
   atlas=/Game/Weapons/FORT_Melee/Pickaxe_RelayStick/Textures/CA_RelayStick
   curve=/Game/Weapons/FORT_Melee/Pickaxe_RelayStick/Textures/C_RelayStick_CircularFresnel
   defaultValue=1
   parameterName=CustomThinFilmRadialGradient
  }
  n24 = MaterialFunctionCall(R=@n23.R, G=@n21, B=@n23.R) {materialFunction=&AppendMany}
  n25 = Multiply(A=Add(
    A=MaterialFunctionCall(R=@n20.R, G=@n21, B=@n20.R) {materialFunction=&AppendMany}.RGB
    B=@n24.RGB
   )) {constB=0.5}
  n26 = NamedRerouteDeclaration(None=Desaturation(None=QualitySwitch(Default=@n25, Low=@n24.RGB, Medium=@n24.RGB), Fraction=Constant() {r=0.5})) {name=SpectralThinFilm, nodeColor=<rgba 0.8516 1 0 1>}
  NamedRerouteDeclaration(None=Multiply(A=Multiply(A=@n12, B=@n26)) {constB=0.15}) {name=ColorfulThinFilm, nodeColor=<rgba 0 1 0.9531 1>}
  NamedRerouteDeclaration(None=@n7.Amplitude) {name=AudioAmplitude, nodeColor=<rgba 0.7578 0 1 1>}
  n27 = Constant3Vector() {constant=<rgba -17 -5 130 0>}
  n28 = Constant3Vector() {constant=<rgba -1 1 1 0>}
  n29 = NamedRerouteUsage()
  n30 = Constant() {r=18}
  n31 = Add(
   A=SphereMask(A=Multiply(A=@n27, B=@n28), B=@n29, Radius=@n30) {attenuationRadius=60, hardnessPercent=0}
   B=SphereMask(A=@n29, B=@n27, Radius=@n30) {attenuationRadius=60, hardnessPercent=0}
  )
  n32 = ComponentMask(None=NamedRerouteUsage()) {r=true}
  n33 = Multiply(A=LinearInterpolate(Alpha=Saturate(None=@n32)) {constA=-1}, B=Multiply(A=Time()) {constB=1.75})
  n34 = Divide(A=ComponentMask(None=@n29) {r=true})
  n35 = Multiply(A=Sine(None=Add(A=@n33, B=@n34))) {constB=0.5}
  NamedRerouteDeclaration(None=Multiply(A=Power(Base=@n31), B=Add(A=@n35) {constB=0.5})) {name=ShoulderEffect, nodeColor=<rgba 1 0.3984 0 1>}
  n36 = ComponentMask(None=Divide(A=NamedRerouteUsage()) {constB=100}) {b=true, r=true}
  n37 = Constant2Vector() {r=0.2}
  n38 = TextureSample!(UVs=Panner(Coordinate=@n36, Speed=@n37) {speedX=0.5}, Tex, Apply View MipBias) {
   samplerSource=SSM_Wrap_WorldGroupSettings
   texture=/BRCosmetics/Accessories/FORT_Backpacks/Backpack_M_MED_FareSpore/Textures/T_FareSpore_Backpack_AudioReactivityGradient
  }
  n39 = ScalarParameter() {defaultValue=3, parameterName=Pink/Blue_ScrollingSpeed}
  n40 = Time()
  n41 = TextureSample(UVs=Panner(Coordinate=@n36, Time=Add(A=@n39, B=@n40), Speed=Multiply(A=@n37) {constB=-1}) {speedX=-0.5}) {
   samplerSource=SSM_Wrap_WorldGroupSettings
   texture=&T_FareSpore_Backpack_AudioReactivityGradient
  }
  n42 = ComponentMask(None=NamedRerouteUsage()) {r=true}
  n43 = Saturate(None=Add(A=Multiply(A=@n42) {constB=0.1}) {constB=0.4})
  NamedRerouteDeclaration(None=Desaturation(None=LinearInterpolate(A=@n38.RGB, B=@n41.RGB, Alpha=@n43), Fraction=Constant() {r=-0.1})) {name="Pink&&BlueColors", nodeColor=<rgba 0.4531 0 1 1>}
  n44 = GetMaterialAttributes(None=NamedRerouteUsage()) {attributeGetTypes [#1]}
  n45 = Multiply(A=NamedRerouteUsage(), B=NamedRerouteUsage())
  n46 = Time()
  n47 = Constant3Vector() {constant=<rgba 0 0 -5 0>}
  n48 = MaterialFunctionCall(
   VertexNormal=@n3
   Basecolor Texture=TextureObject() {samplerType=SAMPLERTYPE_LinearGrayscale, texture=/Game/Characters/Player/Female/Medium/Bodies/F_MED_TNTina/Materials/HalftoneOffset}
   Worldpos=Add(A=NamedRerouteUsage(), B=Multiply(A=@n46, B=@n47))
   TextureSize=Constant() {r=1.5}
  ) {materialFunction=&SimpleLowEndTriplanar_SL}
  n49 = Add(A=ComponentMask(None=@n48.Result) {r=true}) {constB=0.2}
  n50 = Multiply(A=NamedRerouteUsage(), B=Multiply(A=Saturate(None=@n49)) {constB=1.2})
  n51 = NamedRerouteUsage()
  n52 = TextureSample() {
   samplerSource=SSM_Wrap_WorldGroupSettings
   samplerType=SAMPLERTYPE_Masks
   texture=/BRCosmetics/Characters/Player/Male/Medium/Bodies/M_MED_FareSpore/Skins/Spill/Textures/T_FareSpore_Spill_Body_FX
  }
  n53 = Power(Base=@n7.Amplitude) {constExponent=3}
  n54 = Multiply(A=@n52.G, B=@n53)
  n55 = Add(A=Multiply(A=Multiply(A=@n50, B=@n51)) {constB=3}, B=Multiply(A=NamedRerouteUsage(), B=Multiply(A=@n54) {constB=10}))
  n56 = Multiply(A=@n44.Base Color, B=LinearInterpolate(B=NamedRerouteUsage()) {constA=1, constAlpha=0.7})
  n57 = NamedRerouteUsage()
  n58 = BlendMaterialAttributes(
   A=@n44.MaterialAttributes
   B=SetMaterialAttributes!(MaterialAttributes=@n44.MaterialAttributes, Emissive Color=Add(A=@n45, B=@n55), Base Color=Add(A=@n56, B=@n57)) {attributeSetTypes [#2, #1]}
   Alpha=@n52.R
  )
  n59 = NamedRerouteUsage()
  n60 = StaticSwitchParameter(
   True=@n58
   False=BlendMaterialAttributes(
    A=NamedRerouteUsage()
    B=MaterialFunctionCall!(Material Attributes=@n58, Fresnel=@n59, Alpha=@n52.R) {materialFunction=&/MF_ThinFilm_FareSpore}.Result
    Alpha=@n8
   )
  ) {group=Audio Reactivity, parameterName=DEBUG_ReactivityLook}
  NamedRerouteDeclaration(None=@n60) {name=OutputMA, nodeColor=<rgba 0 0.5078 1 1>}
  n61 = Multiply(A=@n53) {constB=2.2}
  NamedRerouteDeclaration(None=Power(Base=NamedRerouteUsage(), Exp=Subtract(B=@n61) {constA=3.5}) {constExponent=3}) {name=AudioReactivityFresnel, nodeColor=<rgba 0.8281 1 0 1>}
  n62 = DotProduct(A=PixelNormalWS(), B=CameraVectorWS())
  n63 = Subtract(A=PixelNormalWS(), B=CameraVectorWS())
  n64 = Normalize(VectorInput=Multiply(
    A=Transform(None=@n63) {transformSourceType=TRANSFORMSOURCE_World, transformType=TRANSFORM_Camera}
    B=Constant3Vector() {constant=<rgba 1 1 0 0>}
   ))
  n65 = Constant3Vector() {constant=<rgba 0 1 0 0>}
  n66 = ComponentMask(None=CrossProduct(A=@n64, B=@n65)) {b=true}
  n67 = Abs(None=@n66)
  n68 = MaterialFunctionCall(In=@n67, Power=Constant() {r=0.75}) {
   desc="Kind of optional, but smoothes out\r\nsome nonlinear distortion"
   materialFunction=&SCurve
  }
  n69 = Multiply(A=QualitySwitch(Default=@n68.Result, Low=@n67, Medium=@n67), B=Sign(None=@n66))
  n70 = ConstantBiasScale(None=@n69) {bias=-3, scale=-0.25}
  n71 = ConstantBiasScale(None=@n69) {scale=0.25}
  n72 = Saturate(None=Sign(None=DotProduct(A=@n64, B=@n65)))
  NamedRerouteDeclaration(None=AppendVector(A=Power(Base=@n62) {desc=Radial coordinate distortion}, B=LinearInterpolate(A=@n70, B=@n71, Alpha=@n72) {constB=0}) {desc=Output UVs}) {name=RadialCoords, nodeColor=<rgba 0.7578 0 1 1>}
  SetMaterialAttributes!(MaterialAttributes=@n60, Emissive Color) {attributeSetTypes [#2]}
 ]
}
