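# Material function that dither-fades masked parts of the Fare Spore character's ears while the BP_Aim parameter is raised.
# A mask from the AudioWheel texture's inverted blue plus alpha is blended in with a short squared ease-out timed from BP_GetTime.
# BP_Aim blends that result against full opacity.
# DitherTemporalAA_Fortnite turns the value into the Opacity Mask on the incoming material attributes.
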MaterialFunction /BRCosmetics/Characters/Player/Male/Medium/Bodies/M_MED_FareSpore/Materials/MF_FareSpore_EarsDithering
graph {nodes [
  n1 = TextureSample() {
   samplerSource=SSM_Wrap_WorldGroupSettings
   texture=/BRCosmetics/Characters/Player/Male/Medium/Bodies/M_MED_FareSpore/Skins/Spill/Textures/T_FareSpore_Spill_AudioWheel
  }
  n2 = Add(A=OneMinus(None=@n1.B), B=@n1.A)
  n3 = Subtract(A=Time(), B=ScalarParameter() {defaultValue=9.999999680285692e+37, parameterName=BP_GetTime})
  n4 = OneMinus(None=Saturate(None=Multiply(A=@n3) {constB=3.5}))
  n5 = Multiply(A=@n4, B=@n4)
  n6 = LinearInterpolate(B=LinearInterpolate(B=Saturate(None=@n2), Alpha=OneMinus(None=@n5)) {constA=1}, Alpha=ScalarParameter() {parameterName=BP_Aim}) {constA=1}
  FunctionOutput(None=SetMaterialAttributes!(
    MaterialAttributes=FunctionInput() {inputName=MA, inputType=FunctionInput_MaterialAttributes}
    OpacityMask=MaterialFunctionCall(Alpha Threshold=@n6) {materialFunction=&DitherTemporalAA_Fortnite}.Result
    EmissiveColor
   ) {attributeSetTypes [#1, #2]})
 ]}
