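# Parent masked material for the FareSpore character's face accessory, with optional audio-reactive glow.
# Character PreFX and ear dithering feed base attributes, which PostFX finishes.
# Behind a static switch, audio amplitude, with a cosine-based fake frontend input, blends in masked base color and emissive changes.
# Named reroutes build four pulsing wheel overlays, scrolling pink/blue gradients, and a spectral thin film using view-space radial coordinates and a borrowed RelayStick curve atlas.
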Material /BRCosmetics/Characters/Player/Male/Medium/Bodies/M_MED_FareSpore/Materials/M_FareSpore_FaceAcc_PARENT
graph {
 properties {
  bUseMaterialAttributes=true
  bUsedWithMeshDeformer=true
  bUsedWithMorphTargets=true
  bUsedWithSkeletalMesh=true
  blendMode=BLEND_Masked
  pixelDepthOffsetMode=PDOM_Legacy
  shadingModel=MSM_FromMaterialExpression
  usedShadingModels=Default Lit | Subsurface | Clear Coat | Subsurface Profile | Cloth
 }
 nodes [
  Constant3Vector() {constant=<rgba 0.1312 0.1301 0.7726 1>}
  n1 = Constant3Vector() {constant=<rgba 0.1457 0.524 0.7726 1>}
  n2 = Constant3Vector() {constant=<rgba 0.5191 0.0809 0.4646 1>}
  n3 = Multiply(A=Time()) {constB=0.25}
  n4 = Add(A=Multiply(A=Sine(None=@n3)) {constB=0.5}) {constB=0.5}
  n5 = MaterialFunctionCall() {materialFunction=&MF_CharacterParent_2023_PreFX}
  n6 = Multiply(A=ComponentMask(None=@n5.Pre skinned Local Position) {r=true}) {constB=0.05}
  LinearInterpolate(A=LinearInterpolate(A=@n1, B=@n2, Alpha=@n4), B=LinearInterpolate(A=@n2, B=@n1, Alpha=@n4), Alpha=Saturate(None=Add(A=@n6) {constB=0.4}))
  n7 = MaterialFunctionCall!(MA=@n5.Result) {materialFunction=&/MF_FareSpore_EarsDithering}
  n8 = GetMaterialAttributes(None=@n7.Result) {attributeGetTypes [#1]}
  n9 = NamedRerouteUsage()
  n10 = NamedRerouteUsage()
  n11 = SphereMask(A=NamedRerouteUsage(), B=Constant3Vector() {constant=<rgba 0 1 157.5 0>}) {attenuationRadius=150, hardnessPercent=0}
  n12 = OneMinus(None=Saturate(None=Multiply(A=@n11) {constB=10}))
  n13 = Constant3Vector() {constant=<rgba 0.4089 0.0601 1 1>}
  n14 = NamedRerouteUsage()
  n15 = Add(A=LinearInterpolate(A=Multiply(A=@n10) {constB=0.35}, B=Multiply(A=@n12, B=@n13), Alpha=NamedRerouteUsage()), B=@n14)
  n16 = Multiply(A=@n10, B=OneMinus(None=NamedRerouteUsage()))
  n17 = Add(A=Multiply(A=@n8.Base Color) {constB=0.1}, B=@n14)
  n18 = OneMinus(None=ComponentMask(None=NamedRerouteUsage()) {b=true})
  n19 = BlendMaterialAttributes(
   A=@n8.MaterialAttributes
   B=SetMaterialAttributes!(MaterialAttributes=@n8.MaterialAttributes, Emissive Color=Add(A=@n9, B=@n15), Base Color=Add(A=@n16, B=@n17)) {attributeSetTypes [#2, #1]}
   Alpha=@n18
  )
  n20 = Cosine(None=Divide(A=Time(), B=ScalarParameter() {defaultValue=1.3, group=Audio Reactivity, parameterName=FrontendFakeAudio_LoopLength, sortPriority=200}))
  n21 = MaterialFunctionCall!(
   InAmplitude
   InSpectrum
   CustomFakeSpectrum
   CustomFakeAmplitude=Multiply(
    A=ConstantBiasScale(None=@n20)
    B=ScalarParameter() {defaultValue=0.5, group=Audio Reactivity, parameterName=FrontendFakeAudio_Amplitude, sortPriority=200}
   )
  ) {materialFunction=/BRCosmetics/Global/MaterialFunctions/AudioReactivity/MF_AudioReactivity_FrontendLoop}
  n22 = Saturate(None=Power(Base=@n21.Amplitude))
  MaterialFunctionCall(
   Material=StaticSwitchParameter(True=BlendMaterialAttributes(A=@n7.Result, B=@n19, Alpha=@n22), False=@n8.MaterialAttributes) {defaultValue=true, group=Audio Reactivity, parameterName=IsAudioReactive}
   M=NamedRerouteDeclaration(None=@n5.M) {name=InMask, nodeColor=<rgba 0.8047 0 1 1>}
   Pre Skinned Local Position=@n5.Pre skinned Local Position
   Distance Blend=@n5.Pre skinned Local Position
   Fresnel=NamedRerouteDeclaration(None=@n5.Fresnel) {name=InFresnel, nodeColor=<rgba 0.8984 1 0 1>}
  ) {materialFunction=&MF_CharacterParent_2023_PostFX}
  n23 = Time()
  n24 = TextureCoordinate()
  n25 = Constant2Vector() {g=0.153, r=-0.344}
  n26 = Constant() {r=0.3}
  n27 = Constant2Vector() {g=1.1, r=1}
  n28 = TextureSample(UVs=MaterialFunctionCall(UVs=Add(A=@n24, B=@n25), Texture Scale=Multiply(A=@n26, B=@n27)) {materialFunction=&ScaleUVsByCenter}.UVs) {
   samplerSource=SSM_Clamp_WorldGroupSettings
   texture=/BRCosmetics/Characters/Player/Male/Medium/Bodies/M_MED_FareSpore/Skins/Spill/Textures/T_FareSpore_Spill_AudioWheel
  }
  n29 = Subtract(A=Subtract(A=Add(A=@n23) {constB=0.35}, B=NamedRerouteUsage()), B=@n28.R)
  n30 = Add(A=Multiply(A=Sine(None=@n29)) {constB=0.5}) {constB=0.5}
  n31 = LinearInterpolate(A=NamedRerouteUsage(), B=NamedRerouteUsage(), Alpha=Saturate(None=Add(A=@n30) {constB=-0.1}))
  n32 = Saturate(None=Add(A=@n30) {constB=-0.15})
  n33 = LinearInterpolate(Alpha=@n30) {constA=0.45, constB=0.2}
  n34 = Time()
  n35 = TextureCoordinate()
  n36 = Constant2Vector() {g=0.33, r=0.34}
  n37 = Constant() {r=0.3}
  n38 = Constant2Vector() {g=1.1, r=1}
  n39 = TextureSample(UVs=MaterialFunctionCall(UVs=Add(A=@n35, B=@n36), Texture Scale=Multiply(A=@n37, B=@n38)) {materialFunction=&ScaleUVsByCenter}.UVs) {
   samplerSource=SSM_Clamp_WorldGroupSettings
   texture=&T_FareSpore_Spill_AudioWheel
  }
  n40 = Add(A=Subtract(A=Add(A=@n34) {constB=-0.1}, B=NamedRerouteUsage()), B=@n39.R)
  n41 = Add(A=Multiply(A=Sine(None=@n40)) {constB=0.5}) {constB=0.5}
  n42 = Add(A=@n41) {constB=-0.1}
  n43 = Multiply(A=LinearInterpolate(A=NamedRerouteUsage(), B=NamedRerouteUsage(), Alpha=Saturate(None=@n42)), B=Saturate(None=Add(A=@n41) {constB=-0.2}))
  n44 = SmoothStep(Min=LinearInterpolate(Alpha=@n41) {constA=0.45, constB=0.2}, Value=@n39.G) {constMax=0.5, constMin=0.6}
  n45 = NamedRerouteUsage()
  n46 = NamedRerouteUsage()
  n47 = Time()
  n48 = TextureCoordinate()
  n49 = Constant2Vector() {g=0.18, r=0.04}
  n50 = Constant() {r=0.3}
  n51 = Constant2Vector() {g=1.1, r=1}
  n52 = TextureSample(UVs=MaterialFunctionCall(UVs=Add(A=@n48, B=@n49), Texture Scale=Multiply(A=@n50, B=@n51)) {materialFunction=&ScaleUVsByCenter}.UVs) {
   samplerSource=SSM_Clamp_WorldGroupSettings
   texture=&T_FareSpore_Spill_AudioWheel
  }
  n53 = Subtract(A=Subtract(A=Add(A=@n47) {constB=0.5}, B=NamedRerouteUsage()), B=@n52.R)
  n54 = Add(A=Multiply(A=Sine(None=@n53)) {constB=0.5}) {constB=0.5}
  n55 = Saturate(None=Add(A=@n54) {constB=-0.1})
  n56 = Add(A=@n54) {constB=-0.2}
  n57 = Multiply(
   A=Multiply(A=LinearInterpolate(A=@n45, B=@n46, Alpha=@n55), B=Saturate(None=@n56))
   B=SmoothStep(Min=LinearInterpolate(Alpha=@n54) {constA=0.45, constB=0.2}, Value=@n52.G) {constMax=0.5, constMin=0.6}
  )
  n58 = NamedRerouteUsage()
  n59 = NamedRerouteUsage()
  n60 = TextureCoordinate()
  n61 = Constant2Vector() {g=0.04, r=0.29}
  n62 = Constant() {r=0.3}
  n63 = Constant2Vector() {g=1.1, r=1}
  n64 = TextureSample(UVs=MaterialFunctionCall(UVs=Add(A=@n60, B=@n61), Texture Scale=Multiply(A=@n62, B=@n63)) {materialFunction=&ScaleUVsByCenter}.UVs) {
   samplerSource=SSM_Clamp_WorldGroupSettings
   texture=&T_FareSpore_Spill_AudioWheel
  }
  n65 = Subtract(A=Add(A=Time(), B=@n64.R), B=NamedRerouteUsage())
  n66 = Add(A=Multiply(A=Sine(None=@n65)) {constB=0.5}) {constB=0.5}
  n67 = Saturate(None=Add(A=@n66) {constB=-0.1})
  n68 = Add(A=@n66) {constB=-0.2}
  n69 = Multiply(
   A=Multiply(A=LinearInterpolate(A=@n58, B=@n59, Alpha=@n67), B=Saturate(None=@n68))
   B=SmoothStep(Min=LinearInterpolate(Alpha=@n66) {constA=0.45, constB=0.2}, Value=@n64.G) {constMax=0.5, constMin=0.6}
  )
  n70 = Add(
   A=Multiply(A=Multiply(A=@n31, B=@n32), B=SmoothStep(Min=@n33, Value=@n28.G) {constMax=0.5, constMin=0.6})
   B=Add(A=Multiply(A=@n43, B=@n44), B=Add(A=@n57, B=@n69))
  )
  NamedRerouteDeclaration(None=Desaturation(None=Multiply(A=@n70) {constB=7}, Fraction=Constant() {r=-1})) {name=SpinningWheels, nodeColor=<rgba 1 0.5156 0 1>}
  n71 = Power(Base=NamedRerouteUsage(), Exp=ScalarParameter() {defaultValue=3, group=06_Thin Film, parameterName=CustomThinFilmExponent})
  n72 = LinearInterpolate(
   A=ScalarParameter() {defaultValue=1, group=06_Thin Film, parameterName=CustomThinFilmIntensity, sortPriority=2}
   B=ScalarParameter() {defaultValue=2, group=06_Thin Film, parameterName=CustomThinFilmIntensity_Loud, sortPriority=2}
   Alpha=Multiply(A=Power(Base=@n21.Amplitude) {constExponent=3}, B=@n71)
  )
  n73 = SmoothStep(Value=Power(Base=@n71, Exp=ScalarParameter() {defaultValue=0.3, group=06_Thin Film, parameterName=CustomThinFilmIntensityExponent, sortPriority=3}))
  n74 = NamedRerouteUsage()
  n75 = ScalarParameter() {defaultValue=0.5, group=06_Thin Film, parameterName=CustomThinFilmWarp, sortPriority=4}
  n76 = Add(
   A=Multiply(A=ScalarParameter() {defaultValue=-0.1, group=06_Thin Film, parameterName=CustomThinFilmPanSpeed}, B=Time())
   B=Multiply(A=ScalarParameter() {defaultValue=0.5, group=06_Thin Film, parameterName=CustomThinFilmScale, sortPriority=5}, B=Power(Base=@n74, Exp=@n75))
  )
  n77 = ScalarParameter() {defaultValue=0.3, group=06_Thin Film, parameterName=CustomThinFilmOffset}
  n78 = ComponentMask(None=NamedRerouteUsage()) {g=true}
  n79 = Multiply(A=Time(), B=ScalarParameter() {defaultValue=0.1, parameterName=CustomThinFilmCircularMaskRotateSpeed})
  n80 = TextureCoordinate()
  n81 = TextureSample!(
   UVs=AppendVector(A=Add(A=@n76, B=@n77), B=Subtract(A=@n78, B=@n79))
   Tex
   "DDX(UVs)"=DDX(Value=@n80)
   "DDY(UVs)"=DDY(Value=@n80)
   Apply View MipBias
  ) {
   mipValueMode=TMVM_Derivative
   samplerSource=SSM_Wrap_WorldGroupSettings
   texture=/BRCosmetics/Characters/Player/Male/Medium/Bodies/M_MED_FareSpore/Textures/T_FareSpore_FaceAcc_GradientMap
  }
  n82 = Multiply(A=Time(), B=ScalarParameter() {defaultValue=0.1, parameterName=CustomThinFilmCircularPanSpeed})
  n83 = CurveAtlasRowParameter(CurveTime=Frac(None=Add(B=@n82))) {
   atlas=/Game/Weapons/FORT_Melee/Pickaxe_RelayStick/Textures/CA_RelayStick
   curve=/Game/Weapons/FORT_Melee/Pickaxe_RelayStick/Textures/C_RelayStick_CircularFresnel
   defaultValue=1
   parameterName=CustomThinFilmRadialGradient
  }
  n84 = MaterialFunctionCall(R=@n83.R, G=Constant(), B=@n83.R) {materialFunction=&AppendMany}
  n85 = Desaturation(None=Multiply(A=Add(A=@n81.RGB, B=@n84.RGB)), Fraction=Constant() {r=0.3})
  NamedRerouteDeclaration(None=Multiply(A=Multiply(A=@n72, B=@n73), B=NamedRerouteDeclaration(None=@n85) {name=SpectralThinFilm, nodeColor=<rgba 0.8516 1 0 1>})) {name=ColorfulThinFilm, nodeColor=<rgba 0 1 0.9531 1>}
  n86 = ComponentMask(None=Divide(A=NamedRerouteUsage()) {constB=100}) {b=true, r=true}
  n87 = Constant2Vector() {r=0.2}
  n88 = TextureSample!(UVs=Panner(Coordinate=@n86, Speed=Multiply(A=@n87) {constB=-1}) {speedX=0.5}, Tex, Apply View MipBias) {
   samplerSource=SSM_Wrap_WorldGroupSettings
   texture=/BRCosmetics/Accessories/FORT_Backpacks/Backpack_M_MED_FareSpore/Textures/T_FareSpore_Backpack_AudioReactivityGradient
  }
  n89 = ScalarParameter() {defaultValue=3, parameterName=Pink/Blue_ScrollingSpeed}
  n90 = Time()
  n91 = TextureSample(UVs=Panner(Coordinate=@n86, Time=Add(A=@n89, B=@n90), Speed=@n87) {speedX=-0.5}) {
   samplerSource=SSM_Wrap_WorldGroupSettings
   texture=&T_FareSpore_Backpack_AudioReactivityGradient
  }
  n92 = ComponentMask(None=NamedRerouteUsage()) {r=true}
  n93 = Saturate(None=Add(A=Multiply(A=@n92) {constB=0.1}) {constB=0.4})
  NamedRerouteDeclaration(None=Desaturation(None=LinearInterpolate(A=@n88.RGB, B=@n91.RGB, Alpha=@n93), Fraction=Constant() {r=-0.1})) {name="Pink&&BlueColors", nodeColor=<rgba 0.4531 0 1 1>}
  NamedRerouteDeclaration(None=Multiply(A=@n22) {constB=-0.35}) {name=AudioAmplitude, nodeColor=<rgba 0 1 0.8125 1>}
  n94 = ComponentMask(None=@n5.Tech Art Mask) {g=true}
  NamedRerouteDeclaration(None=Multiply(A=OneMinus(None=@n94), B=@n18)) {name=MouthMask, nodeColor=<rgba 0.0078 0 1 1>}
  NamedRerouteDeclaration(None=@n5.Pre skinned Local Position) {name=PreSkinnedLocalPosition, nodeColor=<rgba 1 0 0.6562 1>}
  NamedRerouteDeclaration(None=Constant3Vector() {constant=<rgba 1 0 0.0029 1>}) {name=ColorWheelColor_Center, nodeColor=<rgba 0.1474 0.0469 1 1>}
  NamedRerouteDeclaration(None=Constant3Vector() {constant=<rgba 0.0073 0 1 1>}) {name=ColorWheelColor_Edge, nodeColor=<rgba 1 0.0938 0.6978 1>}
  n95 = DotProduct(A=PixelNormalWS(), B=CameraVectorWS())
  n96 = Subtract(A=PixelNormalWS(), B=CameraVectorWS())
  n97 = Normalize(VectorInput=Multiply(
    A=Transform(None=@n96) {transformSourceType=TRANSFORMSOURCE_World, transformType=TRANSFORM_Camera}
    B=Constant3Vector() {constant=<rgba 1 1 0 0>}
   ))
  n98 = Constant3Vector() {constant=<rgba 0 1 0 0>}
  n99 = ComponentMask(None=CrossProduct(A=@n97, B=@n98)) {b=true}
  n100 = Abs(None=@n99)
  n101 = MaterialFunctionCall(In=@n100, Power=Constant() {r=0.75}) {
   desc="Kind of optional, but smoothes out\r\nsome nonlinear distortion"
   materialFunction=&SCurve
  }
  n102 = Multiply(A=QualitySwitch(Default=@n101.Result, Low=@n100, Medium=@n100), B=Sign(None=@n99))
  n103 = ConstantBiasScale(None=@n102) {bias=-3, scale=-0.25}
  n104 = ConstantBiasScale(None=@n102) {scale=0.25}
  n105 = Saturate(None=Sign(None=DotProduct(A=@n97, B=@n98)))
  NamedRerouteDeclaration(None=AppendVector(A=Power(Base=@n95) {desc=Radial coordinate distortion}, B=LinearInterpolate(A=@n103, B=@n104, Alpha=@n105) {constB=0}) {desc=Output UVs}) {name=RadialCoords, nodeColor=<rgba 0.7578 0 1 1>}
  Power(Base=@n11) {constExponent=0.7}
 ]
}
