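# Physics asset for the FareSpore medium male back bling, setting up jacket panel bones to swing against simplified body shapes.
# Seven small base_jacket anchors lock angularly to spine_03 and root dyn_jacket chains, two links at the front and one at the back and sides.
# Dynamic links have narrow swing limits and zero twist limits.
# Mostly capsule bodies on pelvis, spine and legs give the jacket collision volumes, plus an extra right-thigh sphere.
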PhysicsAsset /BRCosmetics/Characters/Player/Male/Medium/Bodies/M_MED_FareSpore/Meshes/M_MED_FareSpore_BackBling_Physics
bodies {
 base_jacket_bck [{shapeType=Capsule, length=0.505, radius=0.505, shapeName=base_jacket_bck_capsule, rotation=<rot 0 -90 0>}]
 base_jacket_bck_l [{shapeType=Capsule, length=0.505, radius=0.505, shapeName=base_jacket_bck_l_capsule, rotation=<rot 0 -90 0>}]
 base_jacket_bck_r [{shapeType=Capsule, length=0.505, radius=0.505, shapeName=base_jacket_bck_r_capsule, rotation=<rot 0 -90 0>}]
 base_jacket_fr_l [{shapeType=Capsule, length=0.505, radius=0.505, shapeName=base_jacket_fr_l_capsule, rotation=<rot 0 -90 0>}]
 base_jacket_fr_r [{shapeType=Capsule, length=0.505, radius=0.505, shapeName=base_jacket_fr_r_capsule, rotation=<rot 0 -90 0>}]
 base_jacket_l [{shapeType=Capsule, length=0.505, radius=0.505, shapeName=base_jacket_l_capsule, rotation=<rot 0 -90 0>}]
 base_jacket_r [{shapeType=Capsule, length=0.505, radius=0.505, shapeName=base_jacket_r_capsule, rotation=<rot 0 -90 0>}]
 calf_l [{shapeType=Capsule, length=22.0941, radius=7.3764, center=<-17.8349 0 0>, rotation=<rot 90 0 0>}]
 calf_r [{shapeType=Capsule, length=21.4529, radius=7.2181, rotation=<rot 90 0 0>}]
 dyn_jacket_01_bck [{shapeType=Capsule, length=13.5918, radius=4.4815, shapeName=dyn_jacket_01_bck_capsule, center=<9.4958 0 3.1375>, rotation=<rot 90 0 0>}]
 dyn_jacket_01_bck_l [{shapeType=Capsule, length=13.5918, radius=4.9747, shapeName=dyn_jacket_01_bck_l_capsule, center=<9.4958 0 3.1375>, rotation=<rot 90 0 0>}]
 dyn_jacket_01_bck_r [{shapeType=Capsule, length=13.5918, radius=4.9747, shapeName=dyn_jacket_01_bck_r_capsule, center=<-9.4958 0 -3.1375>, rotation=<rot 90 180 0>}]
 dyn_jacket_01_fr_l [
  {shapeType=Capsule, length=6.6036, radius=5.3336, shapeName=dyn_jacket_01_fr_l_capsule, center=<5.4783 0 4.3005>, rotation=<rot 90 0 0>}
  {shapeType=Capsule, length=8.8615, radius=3.0758, shapeName=dyn_jacket_01_fr_l_capsule, center=<5.4783 4.9562 0.7359>, rotation=<rot 90 0 0>}
 ]
 dyn_jacket_01_fr_r [{shapeType=Capsule, length=6.6036, radius=5.3336, shapeName=dyn_jacket_01_fr_r_capsule, center=<-5.4783 0 -4.3005>, rotation=<rot 90 0 180>}]
 dyn_jacket_01_l [{shapeType=Capsule, length=11.7617, radius=5.7982, shapeName=dyn_jacket_01_l_capsule, center=<12.9028 0 3.1375>, rotation=<rot 90 0 0>}]
 dyn_jacket_01_r [{shapeType=Capsule, length=11.7617, radius=5.7982, shapeName=dyn_jacket_01_r_capsule, center=<-9.6652 0 -3.1375>, rotation=<rot 90 0 180>}]
 dyn_jacket_02_fr_l [
  {shapeType=Capsule, length=7.5985, radius=6.8664, shapeName=dyn_jacket_02_fr_l_capsule, center=<6.4982 0 5.7545>, rotation=<rot 90 0 0>}
  {shapeType=Capsule, length=11.1464, radius=3.3185, shapeName=dyn_jacket_02_fr_l_capsule, center=<6.4982 3.8169 1.9671>, rotation=<rot 90 0 0>}
 ]
 dyn_jacket_02_fr_r [{shapeType=Capsule, length=9.9264, radius=6.8664, shapeName=dyn_jacket_02_fr_r_capsule, center=<-8.848 0 -5.7545>, rotation=<rot 90 0 180>}]
 foot_l [{shapeType=Capsule, length=15.292, radius=8.2466, rotation=<rot 90 0 0>}]
 foot_r [{shapeType=Capsule, length=15.2469, radius=8.5054, rotation=<rot 90 0 0>}]
 pelvis [{shapeType=Capsule, length=11.6313, radius=9.4468, center=<2.1282 -1.304 0.305>, rotation=<rot -1.1901 -90 1.0408>}]
 spine_01 [{shapeType=Capsule, length=12.3452, radius=8.7157, shapeName=spine_01_capsule, rotation=<rot 0 -90 0>}]
 spine_02 [{shapeType=Capsule, length=9.6091, radius=9.4609}]
 spine_03 [{shapeType=Capsule, length=6.2727, radius=10.6692}]
 spine_04 [{shapeType=Capsule, length=15.3142, radius=4.8856}]
 thigh_l [{shapeType=Capsule, length=24.3018, radius=8.7008, center=<-19.9867 -0.5297 0>, rotation=<rot 90 0 0>}]
 thigh_r [
  {shapeType=Sphere, radius=11.2781, shapeName=thigh_r_sphere, center=<0 0.873 -4.2367>}
  {shapeType=Capsule, length=24.3018, radius=8.7008, center=<19.9867 0.5297 0>, rotation=<rot 90 0 180>}
 ]
}
constraints [
 {bone1Name=spine_02, bone2Name=spine_01, swing1LimitDegrees=45, swing2LimitDegrees=45, twistLimitDegrees=45}
 {bone1Name=spine_03, bone2Name=spine_02, swing1LimitDegrees=45, swing2LimitDegrees=45, twistLimitDegrees=45}
 {bone1Name=spine_04, bone2Name=spine_03, swing1LimitDegrees=45, swing2LimitDegrees=45, twistLimitDegrees=45}
 {bone1Name=thigh_l, bone2Name=pelvis, swing1LimitDegrees=45, swing2LimitDegrees=45, twistLimitDegrees=45}
 {bone1Name=calf_l, bone2Name=thigh_l, swing1LimitDegrees=45, swing2LimitDegrees=45, twistLimitDegrees=45}
 {bone1Name=foot_l, bone2Name=calf_l, swing1LimitDegrees=45, swing2LimitDegrees=45, twistLimitDegrees=45}
 {bone1Name=thigh_r, bone2Name=pelvis, swing1LimitDegrees=45, swing2LimitDegrees=45, twistLimitDegrees=45}
 {bone1Name=calf_r, bone2Name=thigh_r, swing1LimitDegrees=45, swing2LimitDegrees=45, twistLimitDegrees=45}
 {bone1Name=foot_r, bone2Name=calf_r, swing1LimitDegrees=45, swing2LimitDegrees=45, twistLimitDegrees=45}
 {
  bone1Name=base_jacket_fr_l
  bone2Name=spine_03
  swing1LimitDegrees=45
  swing1Motion=Locked
  swing2LimitDegrees=45
  swing2Motion=Locked
  twistLimitDegrees=45
  twistMotion=Locked
 }
 {bone1Name=dyn_jacket_01_fr_l, bone2Name=base_jacket_fr_l, swing1LimitDegrees=10, swing2LimitDegrees=20, twistLimitDegrees=0}
 {bone1Name=dyn_jacket_02_fr_l, bone2Name=dyn_jacket_01_fr_l, swing1LimitDegrees=10, swing2LimitDegrees=20, twistLimitDegrees=0}
 {
  bone1Name=base_jacket_fr_r
  bone2Name=spine_03
  swing1LimitDegrees=45
  swing1Motion=Locked
  swing2LimitDegrees=45
  swing2Motion=Locked
  twistLimitDegrees=45
  twistMotion=Locked
 }
 {bone1Name=dyn_jacket_01_fr_r, bone2Name=base_jacket_fr_r, swing1LimitDegrees=10, swing2LimitDegrees=20, twistLimitDegrees=0}
 {bone1Name=dyn_jacket_02_fr_r, bone2Name=dyn_jacket_01_fr_r, swing1LimitDegrees=10, swing2LimitDegrees=20, twistLimitDegrees=0}
 {
  bone1Name=base_jacket_bck
  bone2Name=spine_03
  swing1LimitDegrees=45
  swing1Motion=Locked
  swing2LimitDegrees=45
  swing2Motion=Locked
  twistLimitDegrees=45
  twistMotion=Locked
 }
 {bone1Name=dyn_jacket_01_bck, bone2Name=base_jacket_bck, swing1LimitDegrees=10, swing2LimitDegrees=5, twistLimitDegrees=0}
 {
  bone1Name=base_jacket_bck_l
  bone2Name=spine_03
  swing1LimitDegrees=45
  swing1Motion=Locked
  swing2LimitDegrees=45
  swing2Motion=Locked
  twistLimitDegrees=45
  twistMotion=Locked
 }
 {bone1Name=dyn_jacket_01_bck_l, bone2Name=base_jacket_bck_l, swing1LimitDegrees=10, swing2LimitDegrees=5, twistLimitDegrees=0}
 {
  bone1Name=base_jacket_bck_r
  bone2Name=spine_03
  swing1LimitDegrees=45
  swing1Motion=Locked
  swing2LimitDegrees=45
  swing2Motion=Locked
  twistLimitDegrees=45
  twistMotion=Locked
 }
 {bone1Name=dyn_jacket_01_bck_r, bone2Name=base_jacket_bck_r, swing1LimitDegrees=10, swing2LimitDegrees=0, twistLimitDegrees=0}
 {
  bone1Name=base_jacket_l
  bone2Name=spine_03
  swing1LimitDegrees=45
  swing1Motion=Locked
  swing2LimitDegrees=45
  swing2Motion=Locked
  twistLimitDegrees=45
  twistMotion=Locked
 }
 {bone1Name=dyn_jacket_01_l, bone2Name=base_jacket_l, swing1LimitDegrees=10, swing2LimitDegrees=20, twistLimitDegrees=0}
 {
  bone1Name=base_jacket_r
  bone2Name=spine_03
  swing1LimitDegrees=45
  swing1Motion=Locked
  swing2LimitDegrees=45
  swing2Motion=Locked
  twistLimitDegrees=45
  twistMotion=Locked
 }
 {bone1Name=dyn_jacket_01_r, bone2Name=base_jacket_r, swing1LimitDegrees=10, swing2LimitDegrees=20, twistLimitDegrees=0}
]
