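# Control Rig for the male medium FareSpore player body, adding corrective deformation to the body, hood and jacket bones.
# A shared deform-rig function runs first, with upper-arm lift switched by parachute state.
# Head and thigh rotations, thigh attributes and a spherical arm pose reader then weight additive bone offsets, adjusted for DBNO, parachute, crouch and RBANoff states.
# Pelvis height also drives a CrouchWalking curve.
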ControlRigBlueprint /BRCosmetics/Characters/Player/Male/Medium/Bodies/M_MED_FareSpore/Meshes/M_MED_FareSpore_CtrlRig
graphs [
 New Function_ContainedGraph {nodes [
   n1 = RigVMFunctionEntryNode()
   RigVMFunction_MathQuaternionToEuler = &RigVMFunction_MathQuaternionToEuler(
    RotationOrder=XYZ
    Value=&RigUnit_GetRelativeTransformForItem(
     Child=(Type=Bone,Name="hand_l")
     Parent=(Type=Null,Name="lowerarm_l")
     bChildInitial=False
     bParentInitial=False
     Child.Name=@n1.Child
     Parent.Name=@n1.Parent
    ).RelativeTransform.Rotation
   )
   Remap = &RigVMFunction_MathDoubleRemap(
    TargetMaximum=1
    TargetMinimum=0
    Value=@RigVMFunction_MathQuaternionToEuler.Result.Z
    SourceMaximum=@n1.Source Maximum
    SourceMinimum=@n1.Source Minimum
   )
   > RigVMFunctionReturnNode(Delta Rotation=@Remap.Result)
  ]}
 New Function_ContainedGraph {nodes [
   n1 = RigVMFunctionEntryNode()
   n2 = &RigVMFunction_ControlFlowBranch(ExecuteContext=@n1.ExecuteContext, Condition=RigVMVariableNode() {variable=isParachuteOpen}.Value)
   RigVMFunctionReturnNode(ExecuteContext=@n2.Completed)
   RigVMVariableNode(Value=(X=0,Y=0,Z=8), ExecuteContext=@n2.False) {variable=SetTranslate}
   RigVMVariableNode(Value=(X=3,Y=0,Z=4), ExecuteContext=@n2.True) {variable=SetTranslate}
  ]}
 New Function_ContainedGraph {nodes [
   n1 = RigVMFunctionEntryNode()
   RigVMFunction_ControlFlowBranch = &RigVMFunction_ControlFlowBranch(Condition=False, ExecuteContext=@n1.ExecuteContext)
   RigVMFunctionReturnNode(ExecuteContext=@RigVMFunction_ControlFlowBranch.Completed)
   &RigUnit_ParentConstraint(
    AdvancedSettings=(RotationOrderForFilter=XZY)
    Child=(Type=Bone,Name="base_jacket_bck")
    Parents="((Item=(Type=Bone,Name=\"clavicle_l\"),Weight=0.0012),(Item=(Type=Bone,Name=\"spine_04\"),Weight=0.0886),(Item=(Type=Bone,Name=\"pelvis\"),Weight=0.1038),(Item=(Type=Bone,Name=\"spine_01\"),Weight=0.14),(Item=(Type=Bone,Name=\"spine_02\"),Weight=0.3003),(Item=(Type= ...(truncated)"
    Weight=1
    ExecutePin=@RigVMFunction_ControlFlowBranch.True
   )
   &RigUnit_ParentConstraint(
    AdvancedSettings=(RotationOrderForFilter=XZY)
    Child=(Type=Bone,Name="base_jacket_bck_l")
    ExecutePin=()
    Parents="((Item=(Type=Bone,Name=\"deform_upperarm_bk_l\"),Weight=0.0003),(Item=(Type=Bone,Name=\"clavicle_l\"),Weight=0.0295),(Item=(Type=Bone,Name=\"spine_04\"),Weight=0.081),(Item=(Type=Bone,Name=\"pelvis\"),Weight=0.1228),(Item=(Type=Bone,Name=\"spine_01\"),Weight=0.1556) ...(truncated)"
    Weight=1
   )
   &RigUnit_ParentConstraint(
    AdvancedSettings=(RotationOrderForFilter=XZY)
    Child=(Type=Bone,Name="base_jacket_bck_r")
    ExecutePin=()
    Parents="((Item=(Type=Bone,Name=\"deform_upperarm_bk_r\"),Weight=0.0004),(Item=(Type=Bone,Name=\"clavicle_r\"),Weight=0.0311),(Item=(Type=Bone,Name=\"spine_04\"),Weight=0.0827),(Item=(Type=Bone,Name=\"pelvis\"),Weight=0.1185),(Item=(Type=Bone,Name=\"spine_01\"),Weight=0.1534 ...(truncated)"
    Weight=1
   )
  ]}
 New Function_ContainedGraph {nodes [
   n1 = RigVMFunctionEntryNode()
   RigVMFunction_MathQuaternionToEuler = &RigVMFunction_MathQuaternionToEuler(
    RotationOrder=XYZ
    Value=&RigUnit_GetRelativeTransformForItem(
     Child=(Type=Bone,Name="hand_l")
     Parent=(Type=Null,Name="lowerarm_l")
     bChildInitial=False
     bParentInitial=False
     Child.Name=@n1.Child
     Parent.Name=@n1.Parent
    ).RelativeTransform.Rotation
   )
   Remap = &RigVMFunction_MathDoubleRemap(
    TargetMaximum=1
    TargetMinimum=0
    Value=@RigVMFunction_MathQuaternionToEuler.Result.Y
    SourceMaximum=@n1.Source Maximum
    SourceMinimum=@n1.Source Minimum
   )
   > RigVMFunctionReturnNode(Delta Rotation=@Remap.Result)
  ]}
 RigVMFunctionLibrary {nodes [
   BackBlingConstraints = RigVMCollapseNode(ExecuteContext=())
   Delta Rotation Y = RigVMCollapseNode(ExecuteContext=())
   Delta Rotation Z = RigVMCollapseNode(ExecuteContext=())
   ShoulderVolume = RigVMCollapseNode(ExecuteContext=())
  ]}
 RigVMFunction_Sequence_1_ContainedGraph {nodes [
   n1 = RigVMFunctionEntryNode()
   RigVMFunction_Sequence = &RigVMFunction_Sequence(ExecuteContext=@n1.ExecuteContext)
   n2 = &RigVMFunction_Sequence(ExecuteContext=@RigVMFunction_Sequence.B)
   RigVMFunctionReturnNode(A=@RigVMFunction_Sequence.A, B=@n2.A, C=@n2.B)
  ]}
 RigVMModel {nodes [
   RigUnit_BeginExecution = &RigUnit_BeginExecution()
   ShoulderVolume = RigVMFunctionReferenceNode(ExecuteContext=@RigUnit_BeginExecution.ExecutePin)
   > FN_DeformRig_FNC = RigVMFunctionReferenceNode(
    DeformRig_ Curve="(deformRig_l_upperarm_up_alpha_72_#1=False,deformRig_r_upperarm_up_alpha_73_#2=False,deformRig_l_upperarm_fr_alpha_74_#3=False,deformRig_r_upperarm_fr_alpha_75_CB51B8 ...(truncated)"
    DeformRig_Bool="(upperarm_1_#4=True,elbow_3_#5=True,wrist_7_#6=True,groin_8_#7=True,glute_17_#8=True,knee_9_40C408324CA261 ...(truncated)"
    DeformRig_Debug=(upperarm_debug_18_#4=False,wrist_debug_20_#6=False,lat_debug_26_#9=False,groinGlute_debug_27_#7=false)
    DeformRig_Value="(upperarm_up_18_#10=(Rotation=(X=0.0,Y=-0.0,Z=0.0,W=1.0),Translation=(X=0.0,Y=0.0,Z=8.0),Scale3D=(X=0.70,Y=1.0,Z=1.0)),upperarm_fwd_fr_19_#11=(Rotation=(X=0.0,Y=-0.0,Z=0.0,W=1.0),Translation=(X=0.0, ...(truncated)"
    DeformRig_Value.upperarm_up_18_#10.Translation=RigVMVariableNode() {variable=SetTranslate}.Value
   )
   > BackBlingConstraints = RigVMFunctionReferenceNode()
   > RigVMFunction_Sequence = RigVMAggregateNode()
   Delta Rotation Z = RigVMFunctionReferenceNode(Child=head, Parent=head_null, Source Maximum=-55, Source Minimum=-35, ExecuteContext=@RigVMFunction_Sequence.A)
   > RigVMVariableNode(Value=@"Delta Rotation Z".Delta Rotation) {variable=head_lookUp}
   > Delta Rotation Z_1 = RigVMFunctionReferenceNode(Child=thigh_l, Parent=thigh_l_null, Source Maximum=-94, Source Minimum=0)
   > RigVMVariableNode(Value=@"Delta Rotation Z_1".Delta Rotation) {variable=thigh_l_up}
   RigUnit_GetRelativeTransformForItem = &RigUnit_GetRelativeTransformForItem(Child=(Type=Bone,Name="pelvis"), Parent=(Type=Bone,Name="root"), bChildInitial=False, bParentInitial=False)
   Remap = &RigVMFunction_MathDoubleRemap(
    SourceMaximum=60
    SourceMinimum=45
    TargetMaximum=1.2
    TargetMinimum=0
    bClamp=False
    Value=@RigUnit_GetRelativeTransformForItem.RelativeTransform.Translation.Z
   )
   n1 = &RigVMFunction_AnimEvalRichCurve(
    Curve=(EditorCurveData=(Keys=((InterpMode=RCIM_Cubic,TangentMode=RCTM_Break),(InterpMode=RCIM_Cubic,TangentMode=RCTM_Break,Time=1,Value=1))))
    SourceMinimum=0
    TargetMinimum=0
    Value=@Remap.Result
   )
   > /Script/ControlRig.RigUnit_SetCurveValue(Curve=CrouchWalking, Value=@n1.Result) {is=struct RigVMUnitNode}
   &RigUnit_ModifyTransforms(
    ItemToModify=((Item=(Type=Bone,Name="dyn_hoodtop_m"),Transform=(Rotation=(X=0,Y=0,Z=-0.2974,W=-0.9548),Translation=(X=16.2137,Y=-6.0707,Z=11.3293),Scale3D=(X=1,Y=1,Z=1))))
    WeightMinimum=0
    ExecutePin=@RigVMFunction_Sequence.B
    Weight=RigVMVariableNode() {variable=head_lookUp}.Value
   )
   RigVMFunction_MathBoolToFloat = &RigVMFunction_MathBoolToFloat(Value=RigVMVariableNode() {variable=isDBNO}.Value)
   n2 = &RigVMFunction_AlphaInterp(
    InterpSpeedDecreasing=5
    InterpSpeedIncreasing=10
    OutRange=(Min=1,Max=0)
    Scale=1
    bClampResult=False
    bInterpResult=False
    bMapRange=true
    Value=@RigVMFunction_MathBoolToFloat.Result
   )
   Multiply = &RigVMFunction_MathDoubleMul(B=@n2.Result, A=RigVMVariableNode() {variable=thigh_l_up}.Value)
   > &RigUnit_ModifyTransforms(
    ItemToModify="((Item=(Type=Bone,Name=\"base_jacket_fr_l\"),Transform=(Rotation=(X=-0.4445,Y=0.3003,Z=0.1236,W=-0.8348),Translation=(X=-3.0271,Y=-6.967,Z=-2.3238),Scale3D=(X=1,Y=1,Z=1))),(Item=(Type=Bone,Name=\"dyn_jacket_02_fr_l\"),Transform=(Rotation=(X=0.2665,Y=0.0838,Z=0 ...(truncated)"
    WeightMinimum=0
    Weight=@Multiply.Result
   )
   n3 = &RigVMFunction_AlphaInterp(
    InterpSpeedDecreasing=5
    InterpSpeedIncreasing=10
    OutRange=(Min=1,Max=0)
    Scale=1
    bClampResult=False
    bInterpResult=False
    bMapRange=true
    Value=&RigVMFunction_MathBoolToFloat(Value=RigVMVariableNode() {variable=isDBNO}.Value).Result
   )
   > &RigUnit_ModifyTransforms(
    ItemToModify=((Item=(Type=Bone,Name="base_jacket_fr_r"),Transform=(Rotation=(X=0.1667,Y=-0.1466,Z=0.07,W=0.9725),Translation=(X=0,Y=0,Z=0),Scale3D=(X=1,Y=1,Z=1))))
    WeightMinimum=0
    Weight=&RigVMFunction_MathDoubleMul(
     B=@n3.Result
     A=RigVMDispatchNode(BoneName=root, Default=1, Name=deformRig_r_thigh_up_alpha) {template="DISPATCH_RigDispatch_GetAnimAttribute(in Name,in BoneName,in Default,out Value,out Found)"}.Value
    ).Result
   )
   > n4 = &RigUnit_ModifyTransforms(
    ItemToModify=((Item=(Type=Bone,Name="deform_glute_l"),Transform=(Rotation=(X=0,Y=0,Z=0,W=1),Translation=(X=-3.7488,Y=0,Z=0),Scale3D=(X=1,Y=1,Z=1))))
    WeightMinimum=0
    Weight=RigVMDispatchNode(BoneName=root, Default=1, Name=deformRig_r_thigh_up_alpha) {template="DISPATCH_RigDispatch_GetAnimAttribute(in Name,in BoneName,in Default,out Value,out Found)"}.Value
   )
   RigVMFunction_ControlFlowBranch = &RigVMFunction_ControlFlowBranch(ExecuteContext=@n4.ExecutePin, Condition=RigVMVariableNode() {variable=isDBNO}.Value)
   AlphaInterp = &RigVMFunction_AlphaInterp(
    InterpSpeedDecreasing=10
    InterpSpeedIncreasing=10
    Scale=1
    bClampResult=false
    bMapRange=true
    Value=&RigVMFunction_MathBoolToFloat(Value=RigVMVariableNode() {variable=isParachuteOpen}.Value).Result
   )
   &RigUnit_ModifyTransforms(
    ItemToModify=((Item=(Type=Bone,Name="dyn_hoodtop_m"),Transform=(Rotation=(X=0,Y=0,Z=0,W=1),Translation=(X=0,Y=3.1963,Z=4.5191),Scale3D=(X=1,Y=1.5,Z=1))))
    WeightMinimum=0
    Weight=@AlphaInterp.Result
    ExecutePin=@RigVMFunction_ControlFlowBranch.Completed
   )
   > n5 = &RigUnit_SphericalPoseReader(
    ActiveRegionScaleFactors=(PositiveWidth=1,NegativeWidth=1,PositiveHeight=1,NegativeHeight=0.8)
    ActiveRegionSize=0.1
    Debug=(bDrawDebug=false,bDraw2D=False,bDrawLocalAxes=False,DebugScale=25,DebugSegments=20,DebugThickness=0.25)
    DriverAxis=(X=1,Y=0,Z=0)
    DriverItem=(Type=Bone,Name="upperarm_r")
    FalloffRegionScaleFactors=(PositiveWidth=1,NegativeWidth=1,PositiveHeight=1.7,NegativeHeight=0.5)
    FalloffSize=0.2
    FlipHeightScaling=False
    FlipWidthScaling=False
    OptionalParentItem=(Type=Bone,Name="None")
    RotationOffset=(X=51.9395,Y=8.7023,Z=22.4256)
   )
   > &RigUnit_ModifyTransforms(
    ItemToModify="((Item=(Type=Bone,Name=\"base_jacket_fr_l\"),Transform=(Rotation=(X=0,Y=0,Z=0,W=-1),Translation=(X=0,Y=0,Z=0.3688),Scale3D=(X=1,Y=1,Z=1))),(Item=(Type=Bone,Name=\"base_jacket_l\"),Transform=(Rotation=(X=0,Y=0,Z=0,W=1),Translation=(X=0,Y=0,Z=1.2111),Scale3D=(X= ...(truncated)"
    WeightMinimum=0
    Weight=@n5.OutputParam
   )
   &RigUnit_ModifyTransforms(
    ItemToModify="((Item=(Type=Bone,Name=\"base_jacket_fr_l\"),Transform=(Rotation=(X=0.3589,Y=-0.1751,Z=0.0172,W=0.9166),Translation=(X=-4.4799,Y=-0.057,Z=-2.0867),Scale3D=(X=1,Y=1,Z=1))),(Item=(Type=Bone,Name=\"base_jacket_l\"),Transform=(Rotation=(X=0,Y=0.0517,Z=0,W=-0.9987) ...(truncated)"
    Weight=1
    WeightMinimum=0
    ExecutePin=@RigVMFunction_ControlFlowBranch.True
   )
   n6 = &RigVMFunction_AlphaInterp(
    InterpSpeedDecreasing=5
    InterpSpeedIncreasing=10
    Scale=1
    bClampResult=False
    bInterpResult=False
    bMapRange=true
    Value=&RigVMFunction_MathBoolToFloat(Value=RigVMVariableNode() {variable=RBANoff}.Value).Result
   )
   n7 = RigVMDispatchNode(BoneName=root, Default=1, Name=deformRig_r_thigh_up_alpha) {template="DISPATCH_RigDispatch_GetAnimAttribute(in Name,in BoneName,in Default,out Value,out Found)"}
   &RigUnit_ModifyTransforms(
    ItemToModify="((Item=(Type=Bone,Name=\"dyn_jacket_01_fr_r\"),Transform=(Rotation=(X=0,Y=-0.116,Z=0,W=0.9932),Translation=(X=0,Y=0,Z=0),Scale3D=(X=1,Y=1,Z=1))),(Item=(Type=Bone,Name=\"dyn_jacket_01_r\"),Transform=(Rotation=(X=0,Y=-0.1212,Z=0,W=0.9926),Translation=(X=0,Y=0,Z= ...(truncated)"
    WeightMinimum=0
    Weight=&RigVMFunction_MathDoubleMul(
     B=@n6.Result
     A=&RigVMFunction_MathDoubleRemap(SourceMaximum=0.4, SourceMinimum=0, TargetMaximum=1, TargetMinimum=0, Value=@n7.Value).Result
    ).Result
    ExecutePin=@RigVMFunction_Sequence.C
   )
   > n8 = &RigUnit_ModifyTransforms(
    ItemToModify=((Item=(Type=Bone,Name="base_jacket_fr_l"),Transform=(Rotation=(X=0,Y=0,Z=0,W=-1),Translation=(X=0,Y=0,Z=-6.2454),Scale3D=(X=1,Y=1,Z=1))))
    WeightMinimum=0
    Weight=&RigVMFunction_MathDoubleMul(B=@n6.Result, A=RigVMVariableNode() {variable=thigh_l_up}.Value).Result
   )
   n9 = &RigVMFunction_ControlFlowBranch(ExecuteContext=@n8.ExecutePin, Condition=RigVMVariableNode() {variable=crouch}.Value)
   Delta Rotation Y = RigVMFunctionReferenceNode(Child=thigh_l, Parent=thigh_l_null, Source Maximum=-10, Source Minimum=5, ExecuteContext=@n9.False)
   > &RigUnit_ModifyTransforms(
    ItemToModify=((Item=(Type=Bone,Name="dyn_jacket_01_fr_l"),Transform=(Rotation=(X=0,Y=-0.0436,Z=0,W=0.999),Translation=(X=0,Y=0,Z=0),Scale3D=(X=1,Y=1,Z=1))))
    WeightMinimum=0
    Weight=@"Delta Rotation Y".Delta Rotation
   )
   setFKCtrl = RigVMFunctionReferenceNode(ExecuteContext=(), setFK_Ctrl=true)
  ]}
 RigVMModel Construction Event Graph {nodes [
   &RigUnit_PrepareForExecution()
   addFKCtrl = RigVMFunctionReferenceNode(
    Custom_FKCtrl=()
    ExecuteContext=()
    Options_FKCtrl="(pelvis_1_#12=false,spine_3_#13=false,clavicle_5_#14=false,arm_7_#15=false,leg_9_#16=true,head_11_46DC15EF465CD ...(truncated)"
    Shape_Settings="(pelvis_2_#17=(bVisible=True,Name=\"Hexagon_Thin\",Color=(R=0.0117,G=0.8281,B=1,A=1),Transform=(Rotation=(X=0,Y=-0.7071,Z=0,W=0.7071),Translation=(X=2,Y=0,Z=0),Scale3D=(X=5,Y=5,Z=7))),spine_4_#18=(bVi ...(truncated)"
   )
  ]}
]
hierarchy [
 root {loc=<0 0 0>, rot=<0 0 0 1>}
 pelvis {parent=root, loc=<0 -0.7071 93.1191>, rot=<0.0127 0.707 -0.0127 -0.707>}
 spine_01 {parent=pelvis, loc=<6.7494 -0.6388 0>, rot=<0 0 0.034 -0.9994>}
 spine_02 {parent=spine_01, loc=<7.2518 -0.035 0>, rot=<0 0 0.0238 -0.9997>}
 spine_03 {parent=spine_02, loc=<6.0104 0 0>, rot=<0 0 0.0149 -0.9999>}
 spine_04 {parent=spine_03, loc=<8.1024 0 0>, rot=<0 0 -0.0132 -0.9999>}
 spine_05 {parent=spine_04, loc=<10.0146 0 0>, rot=<0 0 -0.0015 -1>}
 neck_01 {parent=spine_05, loc=<11.3138 0.1529 0>, rot=<0 0 -0.2003 -0.9797>}
 neck_02 {parent=neck_01, loc=<4.5627 0 0>, rot=<0 0 0.0504 -0.9987>}
 head {parent=neck_02, loc=<5.3914 0 0>, rot=<0 0 0.115 -0.9934>}
 attach_cape {parent=spine_05, loc=<9.0195 -11.7103 0.001>, rot=<0.0616 -0.7044 -0.0616 -0.7044>}
 cape {parent=attach_cape, type=Null, loc=<0 -2.2339 1.8937>, rot=<0 0 0 1>}
 attach_backpack {parent=spine_05, loc=<-0.1187 -13.4666 0.0004>, rot=<-0.0093 -0.707 0.0093 -0.707>}
 Backpack_BR {parent=attach_backpack, type=Null, loc=<0 -1.2063 0>, rot=<0 0 0 1>}
 clavicle_l {parent=spine_05, loc=<4.3575 -0.1466 -2.3223>, rot=<0.0465 -0.7289 -0.009 -0.6829>}
 upperarm_l {parent=clavicle_l, loc=<15.5111 0 0>, rot=<0.0226 -0.3234 -0.0088 -0.9459>}
 lowerarm_l {parent=upperarm_l, loc=<26.6514 0 0>, rot=<0.0391 -0.0099 -0.2453 -0.9686>}
 hand_l {parent=lowerarm_l, loc=<22.3491 0 0>, rot=<-0.7081 -0.0507 0.0975 -0.6975>}
 thumb_01_l {parent=hand_l, loc=<1.3118 -0.3191 -2.459>, rot=<0.6546 -0.4042 0.0664 -0.6354>}
 thumb_02_l {parent=thumb_01_l, loc=<5.9336 -0.0448 0.0401>, rot=<-0.0037 0.0213 0.298 -0.9543>}
 thumb_03_l {parent=thumb_02_l, loc=<2.2223 0.3682 -0.0907>, rot=<0.0118 -0.013 0.0066 -0.9998>}
 index_metacarpal_l {parent=hand_l, loc=<1.6535 0.0183 -2.96>, rot=<0.0158 -0.0124 -0.0173 -0.9996>}
 index_01_l {parent=index_metacarpal_l, loc=<8.3008 0.0357 -0.0524>, rot=<-0.0044 0.0485 0.0924 -0.9945>}
 index_02_l {parent=index_01_l, loc=<3.3587 -0.3004 0.0452>, rot=<-0.002 0.0027 0.3225 -0.9466>}
 index_03_l {parent=index_02_l, loc=<2.5578 -0.123 -0.0252>, rot=<-0.002 0.0067 0.2558 -0.9667>}
 middle_metacarpal_l {parent=hand_l, loc=<2.1856 -0.1725 -0.5984>, rot=<0.006 0.0168 -0.0329 -0.9993>}
 middle_01_l {parent=middle_metacarpal_l, loc=<7.3275 -0.2447 0.0129>, rot=<-0.0208 0.0042 0.0858 -0.9961>}
 middle_02_l {parent=middle_01_l, loc=<4.0123 0.1401 0.1175>, rot=<-0.0218 0.0258 0.3512 -0.9357>}
 middle_03_l {parent=middle_02_l, loc=<2.6077 0.0534 0.1373>, rot=<0.0082 0.0188 0.1708 -0.9851>}
 ring_metacarpal_l {parent=hand_l, loc=<2.2074 -0.3625 1.3638>, rot=<-0.0284 0.0722 -0.0458 -0.9959>}
 ring_01_l {parent=ring_metacarpal_l, loc=<7.0295 -0.6197 -0.1446>, rot=<-0.0489 -0.0252 0.1892 -0.9804>}
 ring_02_l {parent=ring_01_l, loc=<3.6174 0.4647 -0.1027>, rot=<0 0 0.2806 -0.9598>}
 ring_03_l {parent=ring_02_l, loc=<2.4152 0.114 0.0199>, rot=<0 0 0.1189 -0.9929>}
 pinky_metacarpal_l {parent=hand_l, loc=<1.7616 -0.9753 3.3814>, rot=<-0.0397 0.0814 -0.069 -0.9935>}
 pinky_01_l {parent=pinky_metacarpal_l, loc=<7.1709 -0.4831 -0.1094>, rot=<-0.0654 -0.0035 0.2242 -0.9723>}
 pinky_02_l {parent=pinky_01_l, loc=<2.8274 0.0893 -0.1043>, rot=<0.0037 0.0011 0.2854 -0.9584>}
 pinky_03_l {parent=pinky_02_l, loc=<2.2046 0.0307 -0.0078>, rot=<-0.0002 0.0098 0.1445 -0.9895>}
 deform_wrist_dn_base_l {parent=hand_l, loc=<0 -2.2342 0>, rot=<0 0 0 -1>}
 deform_wrist_dn_l {parent=deform_wrist_dn_base_l, loc=<0 0 0>, rot=<0 0 0 1>}
 deform_wrist_up_base_l {parent=hand_l, loc=<0 2.2682 0>, rot=<0 0 0 -1>}
 deform_wrist_up_l {parent=deform_wrist_up_base_l, loc=<0 0 0>, rot=<0 0 0 1>}
 lowerarm_twist_02_l {parent=lowerarm_l, loc=<9.6199 0 0>, rot=<0 0 0 -1>}
 lowerarm_twist_01_l {parent=lowerarm_l, loc=<17.8549 0 0>, rot=<0 0 0 -1>}
 deform_elbow_in_base_l {parent=lowerarm_l, loc=<0.2112 5.259 0.8274>, rot=<0 0 0 -1>}
 deform_elbow_in_l {parent=deform_elbow_in_base_l, loc=<0 0 0>, rot=<0 0 0 1>}
 deform_elbow_out_base_l {parent=lowerarm_l, loc=<0 -4 0>, rot=<0 0 0 -1>}
 deform_elbow_out_l {parent=deform_elbow_out_base_l, loc=<0 0 0>, rot=<0 0 0 1>}
 upperarm_twist_01_l {parent=upperarm_l, loc=<0.0041 0 0>, rot=<0 0 0 -1>}
 deform_upperarm_bk_base_l {parent=upperarm_twist_01_l, loc=<1.0813 -7.2193 -1.9453>, rot=<0 0 0 -1>}
 deform_upperarm_bk_l {parent=deform_upperarm_bk_base_l, loc=<0 0 0>, rot=<0 0 0 1>}
 deform_upperarm_fr_base_l {parent=upperarm_twist_01_l, loc=<-0.3691 7.0506 -2.765>, rot=<0 0 0 -1>}
 deform_upperarm_fr_l {parent=deform_upperarm_fr_base_l, loc=<0 0 0>, rot=<0 0 0 1>}
 deform_upperarm_up_base_l {parent=upperarm_twist_01_l, loc=<-3.2216 -1.1391 3.3311>, rot=<0 0 0 1>}
 deform_upperarm_up_l {parent=deform_upperarm_up_base_l, loc=<0 0 0>, rot=<0 0 0 1>}
 upperarm_twist_02_l {parent=upperarm_l, loc=<13.7941 0 0>, rot=<0 0 0 -1>}
 clavicle_r {parent=spine_05, loc=<4.3575 -0.1466 2.3223>, rot=<0.7289 0.0465 -0.6829 0.009>}
 upperarm_r {parent=clavicle_r, loc=<-15.5111 0 0>, rot=<0.0226 -0.3234 -0.0088 -0.9459>}
 lowerarm_r {parent=upperarm_r, loc=<-26.6514 0 0>, rot=<0.0391 -0.0099 -0.2453 -0.9686>}
 hand_r {parent=lowerarm_r, loc=<-22.3491 0 0>, rot=<-0.7081 -0.0507 0.0975 -0.6975>}
 thumb_01_r {parent=hand_r, loc=<-1.3118 0.3191 2.459>, rot=<0.6546 -0.4042 0.0664 -0.6354>}
 thumb_02_r {parent=thumb_01_r, loc=<-5.9336 0.0448 -0.0401>, rot=<-0.0037 0.0213 0.298 -0.9543>}
 thumb_03_r {parent=thumb_02_r, loc=<-2.2223 -0.3682 0.0907>, rot=<0.0118 -0.013 0.0066 -0.9998>}
 index_metacarpal_r {parent=hand_r, loc=<-1.6535 -0.0183 2.96>, rot=<0.0158 -0.0124 -0.0173 -0.9996>}
 index_01_r {parent=index_metacarpal_r, loc=<-8.3008 -0.0357 0.0524>, rot=<-0.0044 0.0485 0.0924 -0.9945>}
 index_02_r {parent=index_01_r, loc=<-3.3587 0.3004 -0.0452>, rot=<-0.002 0.0027 0.3225 -0.9466>}
 index_03_r {parent=index_02_r, loc=<-2.5578 0.123 0.0252>, rot=<-0.002 0.0067 0.2558 -0.9667>}
 middle_metacarpal_r {parent=hand_r, loc=<-2.1856 0.1725 0.5984>, rot=<0.006 0.0168 -0.0329 -0.9993>}
 middle_01_r {parent=middle_metacarpal_r, loc=<-7.3275 0.2447 -0.0129>, rot=<-0.0208 0.0042 0.0858 -0.9961>}
 middle_02_r {parent=middle_01_r, loc=<-4.0123 -0.1401 -0.1175>, rot=<-0.0218 0.0258 0.3512 -0.9357>}
 middle_03_r {parent=middle_02_r, loc=<-2.6077 -0.0534 -0.1373>, rot=<0.0082 0.0188 0.1708 -0.9851>}
 ring_metacarpal_r {parent=hand_r, loc=<-2.2074 0.3625 -1.3638>, rot=<-0.0284 0.0722 -0.0458 -0.9959>}
 ring_01_r {parent=ring_metacarpal_r, loc=<-7.0295 0.6197 0.1446>, rot=<-0.0489 -0.0252 0.1892 -0.9804>}
 ring_02_r {parent=ring_01_r, loc=<-3.6174 -0.4647 0.1027>, rot=<0 0 0.2806 -0.9598>}
 ring_03_r {parent=ring_02_r, loc=<-2.4152 -0.114 -0.0199>, rot=<0 0 0.1189 -0.9929>}
 pinky_metacarpal_r {parent=hand_r, loc=<-1.7616 0.9753 -3.3814>, rot=<-0.0397 0.0814 -0.069 -0.9935>}
 pinky_01_r {parent=pinky_metacarpal_r, loc=<-7.1709 0.4831 0.1094>, rot=<-0.0654 -0.0035 0.2242 -0.9723>}
 pinky_02_r {parent=pinky_01_r, loc=<-2.8274 -0.0893 0.1043>, rot=<0.0037 0.0011 0.2854 -0.9584>}
 pinky_03_r {parent=pinky_02_r, loc=<-2.2046 -0.0307 0.0078>, rot=<-0.0002 0.0098 0.1445 -0.9895>}
 deform_wrist_dn_base_r {parent=hand_r, loc=<-0.0004 2.2345 -0.0001>, rot=<0 0 0 -1>}
 deform_wrist_dn_r {parent=deform_wrist_dn_base_r, loc=<0 0 0>, rot=<0 0 0 1>}
 deform_wrist_up_base_r {parent=hand_r, loc=<0.0003 -2.2684 0.0001>, rot=<0 0 0 -1>}
 deform_wrist_up_r {parent=deform_wrist_up_base_r, loc=<0 0 0>, rot=<0 0 0 1>}
 lowerarm_twist_02_r {parent=lowerarm_r, loc=<-9.6199 0 0>, rot=<0 0 0 -1>}
 lowerarm_twist_01_r {parent=lowerarm_r, loc=<-17.8549 0 0>, rot=<0 0 0 -1>}
 deform_elbow_in_base_r {parent=lowerarm_r, loc=<-0.2111 -5.2591 -0.8275>, rot=<0 0 0 -1>}
 deform_elbow_in_r {parent=deform_elbow_in_base_r, loc=<-0.0001 0 0.0002>, rot=<0 0 0 1>}
 deform_elbow_out_base_r {parent=lowerarm_r, loc=<0 4 -0.0001>, rot=<0 0 0 -1>}
 deform_elbow_out_r {parent=deform_elbow_out_base_r, loc=<0 0 0>, rot=<0 0 0 1>}
 upperarm_twist_01_r {parent=upperarm_r, loc=<-0.0041 0 0>, rot=<0 0 0 -1>}
 deform_upperarm_bk_base_r {parent=upperarm_twist_01_r, loc=<-1.0814 7.2192 1.9455>, rot=<0 0 0 -1>}
 deform_upperarm_bk_r {parent=deform_upperarm_bk_base_r, loc=<0 0 0>, rot=<0 0 0 1>}
 deform_upperarm_fr_base_r {parent=upperarm_twist_01_r, loc=<0.3689 -7.0506 2.7652>, rot=<0 0 0 -1>}
 deform_upperarm_fr_r {parent=deform_upperarm_fr_base_r, loc=<0 0 0>, rot=<0 0 0 1>}
 deform_upperarm_up_base_r {parent=upperarm_twist_01_r, loc=<3.2214 1.1391 -3.3308>, rot=<0 0 0 1>}
 deform_upperarm_up_r {parent=deform_upperarm_up_base_r, loc=<0 0 0>, rot=<0 0 0 -1>}
 upperarm_twist_02_r {parent=upperarm_r, loc=<-13.7941 0 0>, rot=<0 0 0 -1>}
 base_hood_m {parent=spine_05, loc=<11.787 -7.8379 -1.297>, rot=<0.0284 -0.7065 0.7065 -0.0284>}
 dyn_hood_m {parent=base_hood_m, loc=<0 0 0>, rot=<0 0 0 1>}
 base_hoodtop_m {parent=spine_05, loc=<17.1633 -13.1881 -2.163>, rot=<0.0284 -0.7065 0.7065 -0.0284>}
 dyn_hoodtop_m {parent=base_hoodtop_m, loc=<0 0 0>, rot=<0 0 0 1>}
 neck_01_null {parent=spine_05, type=Null, loc=<11.3138 0.1529 0>, rot=<0 0 -0.2003 -0.9797>}
 head_null {parent=spine_05, type=Null, loc=<20.6573 3.5489 0>, rot=<0 0 -0.036 -0.9994>}
 deform_pec_aim_base_l {parent=spine_04, loc=<7.8732 12.6388 -2.4123>, rot=<0.0102 -0.4602 0.3979 -0.7936>}
 deform_pec_aim_l {parent=deform_pec_aim_base_l, loc=<0 0 0>, rot=<0 0 0 1>}
 deform_pec_base_l {parent=deform_pec_aim_l, loc=<10.766 0 0>, rot=<0 0 0 1>}
 deform_pec_l {parent=deform_pec_base_l, loc=<0.055 0 0>, rot=<0 0 0 1>}
 deform_pec_aim_base_r {parent=spine_04, loc=<7.8727 12.6388 2.4123>, rot=<-0.4602 -0.0102 0.7936 0.3979>}
 deform_pec_aim_r {parent=deform_pec_aim_base_r, loc=<0 0 0>, rot=<0 0 0 1>}
 deform_pec_base_r {parent=deform_pec_aim_r, loc=<-10.7662 -0.0006 -0.0007>, rot=<0 0 0 1>}
 deform_pec_r {parent=deform_pec_base_r, loc=<0 0 0>, rot=<0 0 0 1>}
 base_jacket_fr_l {parent=spine_03, loc=<3.624 8.9257 -11.8907>, rot=<0.051 -0.9297 -0.3641 0.02>}
 dyn_jacket_01_fr_l {parent=base_jacket_fr_l, loc=<0 0 0>, rot=<0 0 0 1>}
 dyn_jacket_02_fr_l {parent=dyn_jacket_01_fr_l, loc=<11.4457 -0.0871 0>, rot=<0 0 0 1>}
 base_jacket_fr_r {parent=spine_03, loc=<1.6478 8.7084 11.8907>, rot=<0.9297 0.051 0.02 0.3641>}
 dyn_jacket_01_fr_r {parent=base_jacket_fr_r, loc=<0 0 0>, rot=<0 0 0 1>}
 dyn_jacket_02_fr_r {parent=dyn_jacket_01_fr_r, loc=<-11.446 0.0871 0>, rot=<0 0 0 1>}
 base_jacket_bck {parent=spine_03, loc=<3.1497 -12.1872 -1.3076>, rot=<0.0387 -0.706 0.706 -0.0387>}
 dyn_jacket_01_bck {parent=base_jacket_bck, loc=<0 0 0>, rot=<0 0 0 1>}
 dyn_jacket_02_bck {parent=dyn_jacket_01_bck, loc=<9.6504 -0.0871 0>, rot=<0 0 0 1>}
 base_jacket_bck_l {parent=spine_03, loc=<2.985 -10.6895 -10.245>, rot=<0.0485 -0.8848 0.4627 -0.0254>}
 dyn_jacket_01_bck_l {parent=base_jacket_bck_l, loc=<0 0 0>, rot=<0 0 0 1>}
 dyn_jacket_02_bck_l {parent=dyn_jacket_01_bck_l, loc=<9.6504 -0.0871 0>, rot=<0 0 0 1>}
 base_jacket_l {parent=spine_03, loc=<1.9835 -1.5824 -16.019>, rot=<0.0547 -0.9985 0 0>}
 dyn_jacket_01_l {parent=base_jacket_l, loc=<0 0 0>, rot=<0 0 0 1>}
 dyn_jacket_02_l {parent=dyn_jacket_01_l, loc=<9.6504 -0.0871 0>, rot=<0 0 0 1>}
 base_jacket_bck_r {parent=spine_03, loc=<2.9851 -10.6895 10.245>, rot=<0.8848 0.0485 -0.0254 -0.4627>}
 dyn_jacket_01_bck_r {parent=base_jacket_bck_r, loc=<0 0 0>, rot=<0 0 0 1>}
 dyn_jacket_02_bck_r {parent=dyn_jacket_01_bck_r, loc=<-9.651 0.0872 0>, rot=<0 0 0 1>}
 base_jacket_r {parent=spine_03, loc=<1.9836 -1.5824 16.019>, rot=<0.9985 0.0547 0 0>}
 dyn_jacket_01_r {parent=base_jacket_r, loc=<0 0 0>, rot=<0 0 0 1>}
 dyn_jacket_02_r {parent=dyn_jacket_01_r, loc=<-9.651 0.0871 0>, rot=<0 0 0 1>}
 base_zipper_01_m {parent=spine_02, loc=<4.2992 6.7919 -14.423>, rot=<0.0385 -0.9647 -0.2604 0.0104>}
 dyn_zipper_01_m {parent=base_zipper_01_m, loc=<0 0 0>, rot=<0 0 0 1>}
 base_zipper_03_m {parent=spine_02, loc=<4.2745 6.7304 15.4033>, rot=<-0.0076 0.1894 0.9811 -0.0392>}
 dyn_zipper_03_m {parent=base_zipper_03_m, loc=<0 0 0>, rot=<0 0 0 1>}
 base_zipper_02_m {parent=spine_01, loc=<-2.9495 10.9636 -7.7697>, rot=<0.0139 -0.8623 -0.5061 0.0082>}
 dyn_zipper_02_m {parent=base_zipper_02_m, loc=<0 0 0>, rot=<0 0 0 1>}
 thigh_l {parent=pelvis, loc=<-2.5693 -0.0109 -7.7501>, rot=<0.0012 0.0572 0.0155 -0.9982>}
 calf_l {parent=thigh_l, loc=<-39.0484 0 0>, rot=<0.0006 -0.0161 -0.0391 -0.9991>}
 foot_l {parent=calf_l, loc=<-42.561 0.1077 0.4292>, rot=<0.0418 -0.0455 0.0343 -0.9975>}
 ball_l {parent=foot_l, loc=<-6.1599 13.8264 0.6806>, rot=<0.0001 -0.0001 -0.7063 -0.7079>}
 calf_twist_02_l {parent=calf_l, loc=<-18.8064 -0.0967 0.0397>, rot=<0.0237 0.0095 -0.0035 -0.9997>}
 calf_twist_01_l {parent=calf_l, loc=<-35.813 -0.0967 0.0397>, rot=<0.0237 0.0095 -0.0035 -0.9997>}
 deform_knee_in_base_l {parent=calf_l, loc=<1.298 -5.2957 -0.0496>, rot=<0 0 0 -1>}
 deform_knee_in_l {parent=deform_knee_in_base_l, loc=<0 0 0>, rot=<0 0 0 1>}
 deform_knee_out_base_l {parent=calf_l, loc=<0 4 0>, rot=<0 0 0 -1>}
 deform_knee_out_l {parent=deform_knee_out_base_l, loc=<0 0 0>, rot=<0 0 0 1>}
 thigh_twist_01_l {parent=thigh_l, loc=<-21.6303 0 0>, rot=<0.0013 0 0 -1>}
 deform_groin_base_l {parent=thigh_l, loc=<-2.1908 8.1142 -1.7442>, rot=<0 0 0 -1>}
 deform_groin_l {parent=deform_groin_base_l, loc=<0 0 0>, rot=<0 0 0 1>}
 thigh_r {parent=pelvis, loc=<-2.5693 -0.0109 7.7501>, rot=<0.0572 -0.0012 0.9982 0.0155>}
 calf_r {parent=thigh_r, loc=<39.0484 0 0>, rot=<0.0006 -0.0161 -0.0391 -0.9991>}
 foot_r {parent=calf_r, loc=<42.561 -0.1077 -0.4292>, rot=<0.0418 -0.0455 0.0343 -0.9975>}
 ball_r {parent=foot_r, loc=<6.1599 -13.8264 -0.6806>, rot=<0.0001 -0.0001 -0.7063 -0.7079>}
 calf_twist_02_r {parent=calf_r, loc=<18.8064 0.0967 -0.0397>, rot=<0.0237 0.0095 -0.0035 -0.9997>}
 calf_twist_01_r {parent=calf_r, loc=<35.813 0.0967 -0.0397>, rot=<0.0237 0.0095 -0.0035 -0.9997>}
 deform_knee_in_base_r {parent=calf_r, loc=<-1.298 5.2957 0.0496>, rot=<0 0 0 1>}
 deform_knee_in_r {parent=deform_knee_in_base_r, loc=<0 0 0>, rot=<0 0 0 1>}
 deform_knee_out_base_r {parent=calf_r, loc=<0 -4 0>, rot=<0 0 0 -1>}
 deform_knee_out_r {parent=deform_knee_out_base_r, loc=<0 0 0>, rot=<0 0 0 1>}
 thigh_twist_01_r {parent=thigh_r, loc=<21.6303 0 0>, rot=<0.0013 0 0 -1>}
 deform_groin_base_r {parent=thigh_r, loc=<2.1908 -8.1142 1.7442>, rot=<0 0 0 -1>}
 deform_groin_r {parent=deform_groin_base_r, loc=<0 0 0>, rot=<0 0 0 1>}
 deform_glute_base_l {parent=pelvis, loc=<-8.8064 -8.8222 -8.4287>, rot=<0.0275 0.0501 -0.4472 -0.8926>}
 deform_glute_l {parent=deform_glute_base_l, loc=<0 0 0>, rot=<0 0 0 1>}
 deform_glute_base_r {parent=pelvis, loc=<-8.8062 -8.8222 8.4287>, rot=<0.0501 -0.0275 0.8926 -0.4472>}
 deform_glute_r {parent=deform_glute_base_r, loc=<0 0 0>, rot=<0 0 0 1>}
 thigh_l_null {parent=pelvis, type=Null, loc=<-2.5693 -0.0109 -7.7501>, rot=<0.0012 0.0572 0.0155 -0.9982>}
 CrouchWalking {type=Curve, loc=<0 0 0>, rot=<0 0 0 1>}
]
variables [
 CrouchWalking {type=double, default=0}
 RBANoff {type=bool, default=False}
 SetTranslate {type=FVector, default=(X=0,Y=0,Z=0)}
 crouch {type=bool, default=False}
 head_lookUp {type=double, default=0}
 isBackbling {type=bool, default=False}
 isDBNO {type=bool, default=False}
 isParachuteOpen {type=bool, default=False}
 thigh_l_up {type=double, default=0}
]
