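# Post-process-style AnimBlueprint for the M_MED_FareSpore player body, adding jacket, hood and zipper motion to the attached parent's copied pose.
# A Control Rig feeds three RigidBody branches (default with hood AnimDynamics, DBNO, backbling), blended by bools.
# Cached poses then pass through jacket corrections, weighted zipper constraints, zipper AnimDynamics and final bone offsets.
# The EventGraph sets state flags and builds skydive gravity from sine waves and Perlin noise, simulating wind.
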AnimBlueprint /BRCosmetics/Characters/Player/Male/Medium/Bodies/M_MED_FareSpore/Meshes/M_MED_FareSpore_AnimBP
target_skeleton=&/M_MED_FareSpore_Skeleton
graphs [
 AnimGraph {nodes [
   Copy Pose From Mesh() {bUseAttachedParent=true}
   Look At - Bone: dyn_jacket_01_fr_l!(ComponentPose, bAlphaBoolEnabled:bool=True) -> (Pose) {
    alphaInputType=Bool
    alphaBoolBlend {blendInTime=0.2, blendOption=Linear, blendOutTime=3}
    boneToModify {boneName=dyn_jacket_01_fr_l}
    lookAtLocation=<0 0 -13.2744>
    lookAtTarget {boneReference {boneName=calf_l}}
    lookAt_Axis {axis=<1 0>}
    lookUp_Axis {axis {x=0, z=1}}
   }
   n1 = "Transform (Modify) Bone - Bone: dyn_jacket_01_bck"(
    ComponentPose=Local To Component(LocalPose=Use cached pose 'CRCache_zipperDyn'())
    Alpha=1
    Translation="5,0,0"
    Rotation="-7,0,0"
    Scale="1,1,1"
   ) {
    rotationMode=BMM_Additive
    rotationSpace=BCS_BoneSpace
    translationMode=BMM_Additive
    translationSpace=BCS_BoneSpace
    alphaBoolBlend {blendOption=Linear}
    boneToModify {boneName=dyn_jacket_01_bck}
    rotation {pitch=-7}
    translation {x=5}
   }
   n2 = "Transform (Modify) Bone - Bone: dyn_jacket_01_bck_l"(ComponentPose=@n1, Alpha=1, Translation="5,0,0", Rotation="-5,0,0", Scale="1,1,1") {
    rotationMode=BMM_Additive
    rotationSpace=BCS_BoneSpace
    translationMode=BMM_Additive
    translationSpace=BCS_BoneSpace
    alphaBoolBlend {blendOption=Linear}
    boneToModify {boneName=dyn_jacket_01_bck_l}
    rotation {pitch=-5}
    translation {x=5}
   }
   n3 = "Transform (Modify) Bone - Bone: dyn_jacket_01_bck_r"(ComponentPose=@n2, Alpha=1, Translation="-3,0,0", Rotation="-5,8,0", Scale="1,1,1") {
    rotationMode=BMM_Additive
    rotationSpace=BCS_BoneSpace
    translationMode=BMM_Additive
    translationSpace=BCS_BoneSpace
    alphaBoolBlend {blendOption=Linear}
    boneToModify {boneName=dyn_jacket_01_bck_r}
    rotation {pitch=-5, yaw=8}
    translation {x=-3}
   }
   n4 = "Transform (Modify) Bone - Bone: dyn_hood_m"(ComponentPose=@n3, Alpha=1, Translation="0,0,0", Rotation="0,0,0", Scale="1,1.2,1") {scaleMode=BMM_Additive, scaleSpace=BCS_BoneSpace, alphaBoolBlend {blendOption=Linear}, boneToModify {boneName=dyn_hood_m}, scale {y=1.2}}
   n5 = "Transform (Modify) Bone - Bone: dyn_jacket_01_r"!(ComponentPose=@n4, bAlphaBoolEnabled:bool=True, Translation="0,0,0", Rotation="-0.000,-16.326,-33.403", Scale="1,1,1") -> (Pose) {
    alphaInputType=Bool
    rotationMode=BMM_Additive
    rotationSpace=BCS_BoneSpace
    alphaBoolBlend {blendInTime=0.2, blendOption=Linear, blendOutTime=0.2}
    boneToModify {boneName=dyn_jacket_01_r}
    rotation=<rot 0 -16.326 -33.403>
   }
   n6 = "Transform (Modify) Bone - Bone: dyn_hood_m"!(ComponentPose=@n5, bAlphaBoolEnabled:bool=True, Translation="0.000,0.000,1.223", Rotation="0,0,0", Scale="1.000,1.000,0.660") -> (Pose) {
    alphaInputType=Bool
    rotationMode=BMM_Additive
    rotationSpace=BCS_BoneSpace
    scaleMode=BMM_Additive
    scaleSpace=BCS_BoneSpace
    translationMode=BMM_Additive
    translationSpace=BCS_BoneSpace
    alphaBoolBlend {blendInTime=0.2, blendOption=Linear, blendOutTime=0.2}
    boneToModify {boneName=dyn_hood_m}
    scale {z=0.66}
    translation {z=1.223}
   }
   n7 = "Transform (Modify) Bone - Bone: dyn_jacket_01_bck_r"!(ComponentPose=@n6, bAlphaBoolEnabled:bool=True, Translation="0,0,0", Rotation="7.648,0.181,0.001", Scale="1,1,1") -> (Pose) {
    alphaInputType=Bool
    rotationMode=BMM_Additive
    rotationSpace=BCS_BoneSpace
    alphaBoolBlend {blendOption=Linear}
    boneToModify {boneName=dyn_jacket_01_bck_r}
    rotation=<rot 7.6479 0.1811 0.001>
   }
   n8 = "Transform (Modify) Bone - Bone: dyn_jacket_01_bck"!(ComponentPose=@n7, bAlphaBoolEnabled:bool=True, Translation="0,0,0", Rotation="4.760,-0.016,0.001", Scale="1,1,1") -> (Pose) {
    alphaInputType=Bool
    rotationMode=BMM_Additive
    rotationSpace=BCS_BoneSpace
    alphaBoolBlend {blendOption=Linear}
    boneToModify {boneName=dyn_jacket_01_bck}
    rotation=<rot 4.7599 -0.016 0.001>
   }
   Output Pose(Result=Component To Local(ComponentPose=@n8)) {layerGroup=None}
   n9 = Control Rig!(
    Source=Copy Pose From Mesh() {bUseAttachedParent=true}
    Alpha
    isDBNO:bool=Get bIsDBNO()
    isParachuteOpen:bool=Get bIsParachuteOpened()
    isBackbling:bool=Get isBackbling() -> (isBackbling:bool)
    RBANoff:bool=Get rbanOff() -> (rbanOff:bool)
    crouch:bool=Get bIsCrouching()
   ) -> (Pose) {
    alphaBoolBlend {blendOption=Linear}
    controlRigAssetReference {blueprintRigClass=&/M_MED_FareSpore_CtrlRig.M_MED_FareSpore_CtrlRig_C}
   }
   Save cached pose 'CRCache_Input'(Pose=@n9) {cacheName=CRCache_Input}
   n10 = "Transform (Modify) Bone - Bone: base_jacket_fr_r"!(
    ComponentPose=Local To Component(LocalPose=Use cached pose 'CRCache_Input'())
    bAlphaBoolEnabled:bool=True
    Translation="0,0,0"
    Rotation="16.766,-18.578,-66.217"
    Scale="1,1,1"
   ) -> (Pose) {
    alphaInputType=Bool
    rotationMode=BMM_Additive
    rotationSpace=BCS_BoneSpace
    alphaBoolBlend {blendInTime=0.5, blendOption=Linear, blendOutTime=0.2}
    boneToModify {boneName=base_jacket_fr_r}
    rotation=<rot 16.766 -18.578 -66.217>
   }
   n11 = RigidBody!(ComponentPose=@n10, Alpha, OverrideWorldGravity:Vector="0,0,-980", ExternalForce) -> (Pose) {
    bOverrideWorldGravity=true
    overridePhysicsAsset=&/M_MED_FareSpore_Physics
    simulationSpace=BaseBoneSpace
    alphaBoolBlend {blendOption=Linear}
    baseBoneRef {boneName=spine_03}
    overrideWorldGravity {z=-980}
    simSpaceSettings {
     maxAngularAcceleration=3
     maxAngularVelocity=1
     maxLinearAcceleration=500
     maxLinearVelocity=200
     worldAlpha=1
     externalLinearDragV=<0.2 0.2 0.2>
    }
   }
   n12 = Anim Dynamics - Bone: dyn_hood_m(ComponentPose=@n11, Alpha=1, ExternalForce="0,0,0") {
    angularDampingOverride=1
    angularSpringConstant=20000
    bAngularSpring=true
    bOverrideAngularDamping=true
    bUseGravityOverride=true
    gravityScale=0.35
    alphaBoolBlend {blendOption=Linear}
    boundBone {boneName=dyn_hood_m}
    componentAppliedLinearAccClamp=<5000 5000 5000>
    componentLinearAccScale=<0.2 0.5 1>
    componentLinearVelScale=<0.2 0.5 1>
    gravityOverride {z=-980}
    physicsBodyDefinitions [{
      boundBone {boneName=dyn_hood_m}
      boxExtents=<20 30 30>
      constraintSetup {angularLimitsMax=<0 3 5>, angularLimitsMin=<0 -80 -5>, angularTarget=<1 0 0>}
      localJointOffset=<5 0 0>
     }]
   }
   n13 = Component To Local(ComponentPose=@n12)
   n14 = RigidBody(ComponentPose=Local To Component(LocalPose=Use cached pose 'CRCache_Input'()), OverrideWorldGravity="0,0,-980") {
    bOverrideWorldGravity=true
    overridePhysicsAsset=&/M_MED_FareSpore_DBNO_Physics
    simulationSpace=BaseBoneSpace
    alphaBoolBlend {blendOption=Linear}
    baseBoneRef {boneName=spine_03}
    overrideWorldGravity {z=-980}
    simSpaceSettings {
     maxAngularAcceleration=3
     maxAngularVelocity=1
     maxLinearAcceleration=500
     maxLinearVelocity=200
     worldAlpha=1
     externalLinearDragV=<0.2 0.2 0.2>
    }
   }
   n15 = Component To Local(ComponentPose=@n14)
   n16 = Get bIsDBNO()
   n17 = "Transform (Modify) Bone - Bone: attach_backpack"!(
    ComponentPose=Local To Component(LocalPose=Use cached pose 'CRCache_Input'())
    bAlphaBoolEnabled:bool=True
    Translation="0,-2,0"
    Rotation="0,0,8"
    Scale="1,1,1"
   ) -> (Pose) {
    alphaInputType=Bool
    rotationMode=BMM_Additive
    rotationSpace=BCS_BoneSpace
    translationMode=BMM_Additive
    translationSpace=BCS_BoneSpace
    alphaBoolBlend {blendInTime=0.5, blendOption=Linear, blendOutTime=0.5}
    boneToModify {boneName=attach_backpack}
    rotation {roll=8}
    translation {y=-2}
   }
   n18 = "Transform (Modify) Bone - Bone: dyn_jacket_01_bck_r"(ComponentPose=@n17, bAlphaBoolEnabled=True, Translation="0,0,0", Rotation="10,0,0", Scale="1,1,1") {
    alphaInputType=Bool
    rotationMode=BMM_Additive
    rotationSpace=BCS_BoneSpace
    translationMode=BMM_Additive
    translationSpace=BCS_BoneSpace
    alphaBoolBlend {blendInTime=0.5, blendOption=Linear, blendOutTime=0.5}
    boneToModify {boneName=dyn_jacket_01_bck_r}
    rotation {pitch=10}
   }
   n19 = RigidBody(ComponentPose=@n18, OverrideWorldGravity="0,0,-980") {
    bOverrideWorldGravity=true
    overridePhysicsAsset=&/M_MED_FareSpore_BackBling_Physics
    simulationSpace=BaseBoneSpace
    alphaBoolBlend {blendOption=Linear}
    baseBoneRef {boneName=spine_03}
    overrideWorldGravity {z=-980}
    simSpaceSettings {
     maxAngularAcceleration=3
     maxAngularVelocity=1
     maxLinearAcceleration=500
     maxLinearVelocity=200
     worldAlpha=1
     externalLinearDragV=<0.2 0.2 0.2>
    }
   }
   Save cached pose 'CRCache_Dynamics'(Pose=Blend Poses by bool(
     BlendPose_0=Component To Local(ComponentPose=@n19)
     BlendPose_1=Blend Poses by bool(BlendPose_0=@n15, BlendPose_1=@n13, bActiveValue=@n16) {blendType=Linear}
     BlendTime_0=0.2
     BlendTime_1=0.5
     bActiveValue=Get isBackbling()
    ) {blendType=Linear}) {cacheName=CRCache_Dynamics}
   n20 = "Transform (Modify) Bone - Bone: dyn_jacket_01_r"(
    ComponentPose=Local To Component(LocalPose=Use cached pose 'CRCache_Dynamics'())
    bAlphaBoolEnabled=True
    Translation="0,0,0"
    Rotation="5,0,0"
    Scale="1,1,1"
   ) {
    alphaInputType=Bool
    rotationMode=BMM_Additive
    rotationSpace=BCS_BoneSpace
    translationMode=BMM_Additive
    translationSpace=BCS_BoneSpace
    alphaBoolBlend {blendOption=Linear}
    boneToModify {boneName=dyn_jacket_01_r}
    rotation {pitch=5}
   }
   n21 = "Transform (Modify) Bone - Bone: dyn_jacket_01_fr_r"!(ComponentPose=@n20, bAlphaBoolEnabled:bool=True, Translation="0,0,0", Rotation="2,0,0", Scale="1,1,1") -> (Pose) {
    alphaInputType=Bool
    rotationMode=BMM_Additive
    rotationSpace=BCS_BoneSpace
    translationMode=BMM_Additive
    translationSpace=BCS_BoneSpace
    alphaBoolBlend {blendOption=Linear}
    boneToModify {boneName=dyn_jacket_01_fr_r}
    rotation {pitch=2}
   }
   n22 = "Constraint - base_jacket_fr_r (RR)"!(ComponentPose=@n21, bAlphaBoolEnabled:bool=True, ConstraintWeights_0=0.4, ConstraintWeights_1:float=0.6) -> (Pose) {
    alphaInputType=Bool
    alphaBoolBlend {blendInTime=0.2, blendOption=Linear, blendOutTime=3}
    boneToModify {boneName=base_jacket_fr_r}
    constraintSetup [{transformType=Rotation, targetBone {boneName=thigh_r}}, {transformType=Rotation, targetBone {boneName=spine_02}}]
    constraintWeights [1,1]
   }
   n23 = "Transform (Modify) Bone - Bone: dyn_jacket_02_fr_l"!(ComponentPose=@n22, bAlphaBoolEnabled:bool=True, Translation="0,0,0", Rotation="15,0,0", Scale="1,1,1") -> (Pose) {
    alphaInputType=Bool
    rotationMode=BMM_Additive
    rotationSpace=BCS_BoneSpace
    translationMode=BMM_Additive
    translationSpace=BCS_BoneSpace
    alphaBoolBlend {blendInTime=0.2, blendOption=Linear, blendOutTime=3}
    boneToModify {boneName=dyn_jacket_02_fr_l}
    rotation {pitch=15}
   }
   n24 = "Constraint - base_zipper_01_m (PPPPPPPPP)"!(
    ComponentPose=@n23
    Alpha=1
    ConstraintWeights_0=0.3349
    ConstraintWeights_1:float=0.2853
    ConstraintWeights_2:float=0.1124
    ConstraintWeights_3:float=0.0986
    ConstraintWeights_4:float=0.0703
    ConstraintWeights_5:float=0.0497
    ConstraintWeights_6:float=0.0401
    ConstraintWeights_7:float=0.0056
    ConstraintWeights_8:float=0.0031
   ) -> (Pose) {
    alphaBoolBlend {blendOption=Linear}
    boneToModify {boneName=base_zipper_01_m}
    constraintSetup [
     {targetBone {boneName=dyn_jacket_01_fr_l}}
     {targetBone {boneName=dyn_jacket_01_l}}
     {targetBone {boneName=dyn_jacket_02_fr_l}}
     {targetBone {boneName=spine_03}}
     {targetBone {boneName=dyn_jacket_02_l}}
     {targetBone {boneName=spine_02}}
     {targetBone {boneName=spine_04}}
     {targetBone {boneName=dyn_jacket_01_bck_l}}
     {targetBone {boneName=deform_pec_l}}
    ]
    constraintWeights [1,1,1,1,1,1,1,1,1]
   }
   n25 = "Constraint - base_zipper_03_m (PPPPPPPPPP)"!(
    ComponentPose=@n24
    Alpha=1
    ConstraintWeights_0=0.38
    ConstraintWeights_1:float=0.2351
    ConstraintWeights_2:float=0.1476
    ConstraintWeights_3:float=0.1058
    ConstraintWeights_4:float=0.0638
    ConstraintWeights_5:float=0.0533
    ConstraintWeights_6:float=0.0069
    ConstraintWeights_7:float=0.0046
    ConstraintWeights_8:float=0.0016
    ConstraintWeights_9:float=0.0013
   ) -> (Pose) {
    alphaBoolBlend {blendOption=Linear}
    boneToModify {boneName=base_zipper_03_m}
    constraintSetup [
     {targetBone {boneName=dyn_jacket_01_fr_r}}
     {targetBone {boneName=dyn_jacket_01_r}}
     {targetBone {boneName=spine_03}}
     {targetBone {boneName=spine_02}}
     {targetBone {boneName=spine_04}}
     {targetBone {boneName=dyn_jacket_02_fr_r}}
     {targetBone {boneName=dyn_jacket_02_r}}
     {targetBone {boneName=clavicle_r}}
     {targetBone {boneName=deform_pec_r}}
     {targetBone {boneName=spine_01}}
    ]
    constraintWeights [1,1,1,1,1,1,1,1,1,1]
   }
   n26 = "Constraint - base_zipper_02_m (PPPP)"!(
    ComponentPose=@n25
    Alpha=1
    ConstraintWeights_0=0.9745
    ConstraintWeights_1:float=0.0242
    ConstraintWeights_2:float=0.001
    ConstraintWeights_3:float=0.0003
   ) -> (Pose) {
    alphaBoolBlend {blendOption=Linear}
    boneToModify {boneName=base_zipper_02_m}
    constraintSetup [
     {targetBone {boneName=dyn_jacket_02_fr_l}}
     {targetBone {boneName=dyn_jacket_01_fr_l}}
     {targetBone {boneName=dyn_jacket_02_l}}
     {targetBone {boneName=dyn_jacket_01_l}}
    ]
    constraintWeights [1,1,1,1]
   }
   n27 = "Constraint - dyn_zipper_01_m (R)"(ComponentPose=@n26, Alpha=1, ConstraintWeights_0=1) {
    alphaBoolBlend {blendOption=Linear}
    boneToModify {boneName=dyn_zipper_01_m}
    constraintSetup [{transformType=Rotation, targetBone {boneName=dyn_jacket_01_fr_l}}]
    constraintWeights [1]
   }
   n28 = "Constraint - dyn_zipper_03_m (R)"(ComponentPose=@n27, Alpha=1, ConstraintWeights_0=1) {
    alphaBoolBlend {blendOption=Linear}
    boneToModify {boneName=dyn_zipper_03_m}
    constraintSetup [{transformType=Rotation, targetBone {boneName=dyn_jacket_01_fr_r}}]
    constraintWeights [1]
   }
   n29 = Anim Dynamics - Bone: dyn_zipper_02_m(ComponentPose=@n28, Alpha=1, ExternalForce="0,0,0") {
    angularDampingOverride=1
    angularSpringConstant=20000
    bAngularSpring=true
    bOverrideAngularDamping=true
    bUseGravityOverride=true
    gravityScale=0.35
    simulationSpace=RootRelative
    alphaBoolBlend {blendOption=Linear}
    boundBone {boneName=dyn_zipper_02_m}
    componentLinearAccScale=<1 1 1>
    componentLinearVelScale=<1 1 1>
    gravityOverride {z=-980}
    physicsBodyDefinitions [{
      boundBone {boneName=dyn_zipper_02_m}
      boxExtents=<20 30 30>
      constraintSetup {angularLimitsMax=<15 10 15>, angularLimitsMin=<-15 -30 -15>, angularTarget=<1 0 0>}
      localJointOffset=<4 0 0>
     }]
   }
   n30 = Anim Dynamics - Bone: dyn_zipper_03_m(ComponentPose=@n29, Alpha=1, ExternalForce="0,0,0") {
    angularDampingOverride=1
    angularSpringConstant=20000
    bAngularSpring=true
    bOverrideAngularDamping=true
    bUseGravityOverride=true
    gravityScale=0.35
    simulationSpace=RootRelative
    alphaBoolBlend {blendOption=Linear}
    boundBone {boneName=dyn_zipper_03_m}
    componentLinearAccScale=<1 1 1>
    componentLinearVelScale=<1 1 1>
    gravityOverride {z=-980}
    physicsBodyDefinitions [{
      boundBone {boneName=dyn_zipper_03_m}
      boxExtents=<20 30 30>
      constraintSetup {angularLimitsMax=<15 1 15>, angularLimitsMin=<-15 -30 -15>, angularTarget=<1 0 0>}
      localJointOffset=<4 0 0>
     }]
   }
   n31 = Anim Dynamics - Bone: dyn_zipper_01_m(ComponentPose=@n30, Alpha=1, ExternalForce="0,0,0") {
    angularDampingOverride=1
    angularSpringConstant=20000
    bAngularSpring=true
    bOverrideAngularDamping=true
    bUseGravityOverride=true
    gravityScale=0.35
    simulationSpace=RootRelative
    alphaBoolBlend {blendOption=Linear}
    boundBone {boneName=dyn_zipper_01_m}
    componentLinearAccScale=<1 1 1>
    componentLinearVelScale=<1 1 1>
    gravityOverride {z=-980}
    physicsBodyDefinitions [{
      boundBone {boneName=dyn_zipper_01_m}
      boxExtents=<20 30 30>
      constraintSetup {angularLimitsMax=<15 1 15>, angularLimitsMin=<-15 -30 -15>, angularTarget=<1 0 0>}
      localJointOffset=<4 0 0>
     }]
   }
   Save cached pose 'CRCache_zipperDyn'(Pose=Component To Local(ComponentPose=@n31)) {cacheName=CRCache_zipperDyn}
  ]}
 EventGraph {nodes [
   n1 = Event BlueprintUpdateAnimation()
   > n2 = Sequence!(execute) -> (then_0, then_1, then_2:Exec, then_3:Exec)
   n3 = Branch(execute=@n2.then_1, Condition=OR Boolean(A=Get bIsOnGrindRail(), B=Get bIsSliding()))
   > Set jacketTransform(jacketTransform:bool=true) -> (Output_Get:bool)
   Set jacketTransform(execute=@n3.else, jacketTransform=false)
   n4 = Get bIsParachuteOpened()
   > Set isskydive(isskydive:bool=AND Boolean(A=Get bIsSkyDiving(), B=NOT Boolean(A=@n4))) -> (Output_Get:bool)
   n5 = Sequence!(execute=@n2.then_0) -> (then_0, then_1)
   > n6 = Branch(Condition=Get bIsSkydiveFloating())
   > Set Set Gravity(Set Gravity:Vector="0,0,-200") -> (Output_Get:Vector)
   > Set SetAmplitud(SetAmplitud:Vector="50,50,300") -> (Output_Get:Vector)
   > Set SetFrequency(SetFrequency:Rotator="4,13,5") -> (Output_Get:Rotator)
   Set Set Gravity(execute=@n6.else, Set Gravity="0,0,-500")
   > Branch(Condition=Get bIsSkydiveDiveMode())
   > Set Set Gravity(Set Gravity="0,0,200")
   > Set SetAmplitud(SetAmplitud="100,400,100")
   > Set SetFrequency(SetFrequency="12,10,1.5")
   n7 = Branch(execute=@n5.then_1, Condition=Get bIsSkyDiving())
   n8 = BreakVector(InVec=Get SetAmplitud() -> (SetAmplitud:Vector))
   n9 = GetGameTimeInSeconds()
   n10 = BreakRotator(InRot=Get SetFrequency() -> (SetFrequency:Rotator))
   n11 = "Sin (Radians)"(A=float + float(A=float * float(A=@n10.Yaw, B=@n9) {is=title K2Node_PromotableOperator}, B=@n1.DeltaTimeX))
   n12 = float * float(A=PerlinNoise1D!(Value:float=@n9) -> (ReturnValue:float), B=1) {is=title K2Node_CommutativeAssociativeBinaryOperator}
   n13 = float * float(A=@n11, B=@n12)
   n14 = float + float(A=float * float(A=float + float(A=@n11, B=@n13) {is=title K2Node_CommutativeAssociativeBinaryOperator}, B=@n8.Z))
   n15 = "Sin (Radians)"(A=float + float(A=float * float(A=@n10, B=@n9) {is=title K2Node_PromotableOperator}, B=@n1.DeltaTimeX) {is=title K2Node_PromotableOperator})
   n16 = float * float(A=@n15, B=@n12) {is=title K2Node_CommutativeAssociativeBinaryOperator}
   n17 = float + float(A=float * float(A=float + float(A=@n15, B=@n16) {is=title K2Node_CommutativeAssociativeBinaryOperator}, B=@n8))
   n18 = "Sin (Radians)"(A=float + float(A=float * float(A=@n10.Pitch, B=@n9) {is=title K2Node_PromotableOperator}, B=@n1.DeltaTimeX) {is=title K2Node_PromotableOperator})
   n19 = float * float(A=@n18, B=@n12) {is=title K2Node_CommutativeAssociativeBinaryOperator}
   n20 = float + float(A=float * float(A=float + float(A=@n18, B=@n19) {is=title K2Node_CommutativeAssociativeBinaryOperator}, B=@n8.Y))
   > Set GravityOverride(GravityOverride=vector + vector(A=MakeVector(X=@n17, Y=@n20, Z=@n14), B=Get Set Gravity() -> (Set Gravity:Vector)))
   Set GravityOverride(execute=@n7.else, GravityOverride="0,0,-980")
   n21 = Get bIsSliding()
   n22 = Branch(execute=@n2.then_2, Condition=AND Boolean(A=Get bIsCrouching(), B=NOT Boolean(A=@n21)))
   > Set Crouchwalking(Crouchwalking:double=GetCurveValue(CurveName=CrouchWalking)) -> (Output_Get:double)
   Set Crouchwalking(execute=@n22.else, Crouchwalking=0)
   n23 = CheckForBackbling()
   n24 = OR Boolean(A=@n23.bIsBackblingACape, B=@n23.bIsBackblingLong)
   Set isBackbling(execute=@n2.then_3, isBackbling:bool=AND Boolean(A=@n23, B=NOT Boolean(A=@n24))) -> (Output_Get:bool)
   > Set isWearingPackBool(isWearingPackBool:bool=OR Boolean!(A=@n23, B=@n23.bIsBackblingACape, C:bool=@n23.bIsBackblingLong) -> (ReturnValue)) -> (Output_Get:bool)
   > Set rbanOff(rbanOff:bool=NOT Boolean(A=Get bRigidBodyEnabled())) -> (Output_Get:bool)
   n25 = Get bIsCrouching()
   n26 = Get bRigidBodyEnabled()
   n27 = Get bIsSkyDiving()
   > Set jacket constraints(jacket constraints:bool=AND Boolean!(A=Get bIsMoving2D(), B=NOT Boolean(A=@n25), C:bool=NOT Boolean(A=@n26), D:bool=NOT Boolean(A=@n27)) -> (ReturnValue)) -> (Output_Get:bool)
  ]}
]
variables [
 Character Part Anim Instance {type=CustomCharacterPartAnimInstance, default=null}
 Crouchwalking {type=number}
 Gravity Override Params {
  type=GravityOverrideParamsStruct
  default {
   onGround_Standing {windAmplitude=<500 200 100>, windFrequency=<500 1 1>}
   skydive_FloatingDown {gravityOverride=<0 0 0>, windAmplitude=<100 100 200>, windOffset=<0 -600 -200>}
   skydive_GoingUp {gravityOverride=<0 0 0>, windAmplitude=<100 100 200>, windFrequency=<1.5 7 10>, windOffset=<0 -600 -200>}
   skydive_Parachuting {gravityOverride=<0 0 -200>, windAmplitude=<100 100 200>, windOffset=<0 -600 -500>}
  }
 }
 GravityOverride {type=Vector}
 isBackbling {type=boolean}
 isskydive {type=boolean}
 isWearingPackBool {type=boolean}
 jacket constraints {type=boolean}
 jacketTransform {type=boolean}
 rbanOff {type=boolean}
 Set Gravity {type=Vector}
 SetAmplitud {type=Vector}
 SetFrequency {type=Rotator, default=<rot 4 13 5>}
]
